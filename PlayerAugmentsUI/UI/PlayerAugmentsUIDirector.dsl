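; Dynamic UI director that adds the player augments UI scene for players or live-broadcast spectators once playlist data is ready.
; Players in a zone get the mobile or standard scene, gated by a hotfix flag and chosen by touch detection.
; For broadcast controllers, replay HUD visibility events and the saved augments toggle feed an update function.
; It adds or removes the spectator scene only when the computed visibility changes, requiring the augment system enabled.
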(event Custom|On%20HUD%20VIsibility%20Changed
  (params (IsVisible :type EHudVisibilityState%20Enum :default "FullyVisible"))
  (Variables|Default|SetSpectatorHudState IsVisible)
  (CallFunction|UpdateAugmentsSpectatorUIVisibility))

(event Custom|CustomEvent
  (params (bIsVisible :type Boolean))
  (Variables|Default|SetbAugmentsHudToggle bIsVisible)
  (CallFunction|UpdateAugmentsSpectatorUIVisibility))

(event EventBeginPlay
  (bind _returnvalue (DynamicUIDirector|GetOwningLocalPlayerController))
  (bind _returnvalue_2 (BlueprintContext|GetFortReplayContext))
  (Class|FortAsyncActionCurrentPlaylistReady|PlaylistDataReadyAsync
    (:OnReady
      (bind _asfort_live_broadcast_controller (Utilities|Casting|CastToFortLiveBroadcastController _returnvalue)
        (:then
          (Class|FortLiveBroadcastController|SetupClientBindingsforAugmentsHUD _asfort_live_broadcast_controller)
          (CallFunction|InitializeAugmentsSpectatorUIVisibility _asfort_live_broadcast_controller)
          (ReplayContext|BindEventtoOnReplayHudVisibilityChanged _returnvalue_2 (delegate Custom|On%20HUD%20VIsibility%20Changed))
          (ReplayContext|BindEventtoOnReplayAugmentsHudVisibilityChanged _returnvalue_2 (delegate Custom|CustomEvent))
          (CallFunction|UpdateAugmentsSpectatorUIVisibility))
        (:CastFailed
          (if (and (GlobalUIContext|IsinZone (PlayerController|LocalPlayerSubsystems|GetFortGlobalUIContext _returnvalue)) (ScalableFloat|GetValueasBool (Variables|Default|GetAugmentsEnabledViaHotfix)))
            (if (or (TouchPlatform|IsUsingTouch (DynamicUIDirector|GetOwningLocalPlayerController)) (and (Editor|InEditororPIE) (TouchPlatform|IsMobileGame)))
              (DynamicUIDirector|AddScene "/PlayerAugmentsUI/MobileUI/Scenes/PlayerAugmentsUIMobileScene.PlayerAugmentsUIMobileScene")
              (else
                (DynamicUIDirector|AddScene "/PlayerAugmentsUI/UI/PlayerAugmentsUIScene.PlayerAugmentsUIScene")))))))))

(fn InitializeAugmentsSpectatorUIVisibility
  (params (Player%20Controller :type Fort%20Player%20Controller%20Object%20Reference))
  (bind (_starttime _endtime _timenow _bispaused _playbackspeedmultiplier _hudvisibility) (Utilities|Struct|BreakFortReplayPlaybackState (Variables|/Script/FortniteGame.FortReplayContext|GetPlaybackState (BlueprintContext|GetFortReplayContext))))
  (Variables|Default|SetbAugmentsHudToggle (Class|FortClientSettingsRecord|GetAugmentsHUDVisibility (Class|FortPlayerController|GetClientSettings Player%20Controller)))
  (Variables|Default|SetSpectatorHudState _hudvisibility))

(fn UpdateAugmentsSpectatorUIVisibility
  (bind _spectatorhudstate (Variables|Default|GetSpectatorHudState))
  (bind _returnvalue (Utilities|Select :Index _spectatorhudstate :FullyVisible true :FullyHidden false :GameOnly true :ReplayOnly false))
  (bind _baugmentshudtoggle (Variables|Default|GetbAugmentsHudToggle))
  (bind _returnvalue_2 (Class|FortBattleRoyalePlayerAugmentSystem|Get (DynamicUIDirector|GetOwningLocalPlayerController)))
  (if (!= (select (and (Class|FortBattleRoyalePlayerAugmentSystem|IsBRAugmentSystemEnabled _returnvalue_2) _baugmentshudtoggle) _returnvalue false) (Variables|Default|GetbShowAugmentsUI))
    (bind _output_get (Variables|Default|SetbShowAugmentsUI (select (and (Class|FortBattleRoyalePlayerAugmentSystem|IsBRAugmentSystemEnabled _returnvalue_2) _baugmentshudtoggle) _returnvalue false)))
    (if _output_get
      (DynamicUIDirector|AddScene "/PlayerAugmentsUI/UI/PlayerAugmentsSpectatorUIScene.PlayerAugmentsSpectatorUIScene")
      (else
        (DynamicUIDirector|RemoveScene "/PlayerAugmentsUI/UI/PlayerAugmentsSpectatorUIScene.PlayerAugmentsSpectatorUIScene")))))

(fn ConstructionScript)
