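# Sound cue for the Player Augments UI selector, named for showing the active list.
# A single wave player feeds the output directly with one sound asset, with no randomization or modulation.
# It uses the non-critical in-game UI sound class and overrides concurrency to one instance, stopping the oldest when retriggered.
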SoundCue /PlayerAugmentsUI/Sound/PlayerAugments_UI_Selector_Show_List_Active_Cue
root=WavePlayer_0
nodes [{id=WavePlayer_0, properties {soundWaveAssetPtr=&/PlayerAugments_UI_Selector_Show_List_Active_01}}]
properties {
 bDebug=true
 bOverrideConcurrency=true
 maxDistance=2097152
 soundClassObject=&Fort_UI_InGame_NonCritical
 volumeMultiplier=1
 concurrencyOverrides {maxCount=1, resolutionRule=StopOldest}
}
