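# Sound cue for the Player Augments UI that, by its name, plays when an augments window is swiped away.
# A single wave player feeds one swipe-away sample straight to the output, with a raised volume and no randomization or layering.
# It routes to the non-critical in-game UI sound class, overrides concurrency to stop the oldest of a few overlapping instances, and disables virtualization.
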SoundCue /PlayerAugmentsUI/Sound/PlayerAugments_UI_Window_Swipe_Away_Cue
root=WavePlayer_0
nodes [{id=WavePlayer_0, properties {soundWaveAssetPtr=&/PlayerAugments_UI_Select_HUD_Swipe_Away_01}}]
properties {
 bOverrideConcurrency=true
 maxDistance=2097152
 soundClassObject=&Fort_UI_InGame_NonCritical
 virtualizationMode=Disabled
 volumeMultiplier=1.3
 concurrencyOverrides {maxCount=2, resolutionRule=StopOldest}
}
