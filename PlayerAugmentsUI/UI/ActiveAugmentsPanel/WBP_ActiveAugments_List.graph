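# HUD quickbar widget showing the player's active augments as a scrollable, focusable list of slots.
# PreConstruct builds empty slots from the playlist's grant-round count, or preview entries at design time, and HandleAugmentsUpdated fills them.
# Opening resets focus, fades in the glow and staggers entry spawn animations.
# ScrollUp/ScrollDown move focus, scroll it into view and refresh arrows and input-specific keybind, wrapping downward only on mobile touch.
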WidgetBlueprint /PlayerAugmentsUI/UI/ActiveAugmentsPanel/WBP_ActiveAugments_List
animations [Anim_ListFocused {
  binding_count=1
  length_seconds=0.25
  object=&self:Anim_ListFocused
  sequence {
   bindings [Backing_Glow {
     id=#1
     possessed_class=Image
     tracks [{
       type=MovieSceneWidgetMaterialTrack
       sections [{type=MovieSceneParameterSection, channels [Anim_BackingAlpha {type=MovieSceneScriptingFloatChannel, keys [[0,0],[15,1]]}]}]
      }]
    }]
   display_rate {numerator=60}
  }
 }]
graphs [
 CreateActiveAugmentList {nodes [
   n1 = CreateActiveAugmentList!() -> (then, Count:int) {is=title K2Node_FunctionEntry}
   > ClearChildren(self=Get ScrollBox_ActiveList() -> (ScrollBox_ActiveList:Fort Child Navigatable Scroll Box Object Reference))
   > For Loop(LastIndex=int - int(A=@n1))
   > n2 = Create Widget!(
    execute:Exec
    Class:User Widget Class Reference=&/WBP_ActiveAugments_ListEntry.WBP_ActiveAugments_ListEntry_C
    OwningPlayer:Player Controller Object Reference
   ) -> (then:Exec, ReturnValue:WBP Active Augments List Entry Object Reference)
   > AddChild(self=Get ScrollBox_ActiveList(), Content=@n2)
   > SetAsEmptySlot!(execute, self:WBP Active Augments List Entry Object Reference=@n2, InIsEmptySlot:bool=true) -> (then)
  ]}
 EventGraph {nodes [
   OnActiveListOpened() {is=title K2Node_CustomEvent}
   > RefreshArrows()
   > ResetFocusedEntry()
   > PlayAnimation(InAnimation=Get Anim_ListFocused() -> (Anim_ListFocused:Widget Animation Object Reference))
   > n1 = For Each Loop with Delay!(Exec:Exec, Array:Array of Widget Object References=GetAllChildren(self=Get ScrollBox_ActiveList()), Delay:double=0.05) -> (Loop Body:Exec, Array Element:Widget Object Reference, Array Index:int, Completed:Exec) {is=title K2Node_MacroInstance}
   > n2 = Cast To WBP_ActiveAugments_ListEntry!(execute, Object=@n1) -> (then, CastFailed, AsWBP Active Augments List Entry:WBP Active Augments List Entry Object Reference) {is=title K2Node_DynamicCast}
   > PlaySpawnEntryAnim(self:WBP Active Augments List Entry Object Reference=@n2)
   n3 = Event PreConstruct()
   > n4 = Branch(Condition=@n3.IsDesignTime)
   n5 = Get Anim_ListFocused()
   > PlayAnimation(InAnimation=@n5, StartAtTime=GetEndTime(self=@n5))
   > ClearChildren(self=Get ScrollBox_ActiveList())
   > n6 = For Loop(LastIndex=Get MaxAugmentIndex_DesignTime() -> (MaxAugmentIndex_DesignTime:int=0))
   > n7 = Create Widget(Class=&WBP_ActiveAugments_ListEntry_C)
   > AddChild(self=Get ScrollBox_ActiveList(), Content=@n7)
   > n8 = Branch(Condition=OR Boolean(A="Equal (Integer)"(A=@n6), B="Equal (Integer)"(A=@n6, B=1)))
   SetAsEmptySlot(execute=@n8.else, self=@n7, InIsEmptySlot=true)
   CreateActiveAugmentList!(execute=@n4.else, self, Count:int=GetNumAugmentGrantRounds!(self) -> (NumAugmentsForRound:int=0)) -> (then) {is=title K2Node_CallFunction}
   n9 = Event HandleAugmentsUpdated!() -> (OutputDelegate, then, ActiveAugments:Array of Fort Player Augment Item Definition Object References) {is=title K2Node_Event}
   > Set Active Augments(Active Augments:Array of Fort Player Augment Item Definition Object References=@n9.ActiveAugments) -> (Output_Get:Array of Fort Player Augment Item Definition Object References)
   > RefreshCurrentAugmentList()
  ]}
 GetNumAugmentGrantRounds {nodes [
   GetNumAugmentGrantRounds() {is=title K2Node_FunctionEntry}
   n1 = Get NumAugmentGrantRoundsForPlaylist() -> (NumAugmentGrantRoundsForPlaylist:Scalable Float Structure)
   > Return Node!(
    execute
    NumAugmentsForRound:int=Truncate(A=GetValueAtLevel!(Input:"Scalable Float Structure (by ref)"=@n1, Level:float=0) -> (ReturnValue:float))
   )
  ]}
 RefreshArrows {nodes [
   RefreshArrows() {is=title K2Node_FunctionEntry}
   > UpdateScrollBoundAction!(execute, self, InputType:ECommonInputType Enum=Get InputType() -> (InputType:ECommonInputType Enum)) -> (then)
   n1 = Get ActiveAugmentsEntries() -> (ActiveAugmentsEntries:Array of WBP Active Augments List Entry Object References)
   > n2 = Branch(Condition="integer >= integer"(A=Length(TargetArray:Array of WBP Active Augments List Entry Object References=@n1), B=5))
   n3 = Get ActiveAugmentsEntries()
   > n4 = Branch(Condition="integer >= integer"(A=Length(TargetArray=@n3), B=2))
   > SetVisibility(self=Get VBox_CycleInput() -> (VBox_CycleInput:Vertical Box Object Reference), InVisibility=HitTestInvisible)
   > SetRenderOpacity(self=Get VBox_CycleInput(), InOpacity=1)
   n5 = Get FocusedEntryIndex() -> (FocusedEntryIndex:int=0)
   > SetRenderOpacity(self=Get Image_ArrowUp() -> (Image_ArrowUp:Image Object Reference), InOpacity=SelectFloat(A=0.25, B=1, bPickA="Equal (Integer)"(A=@n5)))
   n6 = Last Index!(TargetArray:Array of WBP Active Augments List Entry Object References=Get ActiveAugmentsEntries()) -> (ReturnValue=0)
   n7 = Get FocusedEntryIndex() -> (FocusedEntryIndex=0)
   > SetRenderOpacity(
    self=Get Image_ArrowDown() -> (Image_ArrowDown:Image Object Reference)
    InOpacity=SelectFloat(A=0.25, B=1, bPickA="Equal (Integer)"(A=@n7, B=@n6))
   )
   > SpawnSound2D!(
    execute
    Sound:Sound Base Object Reference=Get HighlightSound() -> (HighlightSound:Sound Base Object Reference)
    VolumeMultiplier:float=1
    PitchMultiplier:float=1
    StartTime:float=0
    ConcurrencySettings:Sound Concurrency Object Reference
    bPersistAcrossLevelTransition:bool=false
    bAutoDestroy:bool=true
   ) -> (then, ReturnValue:Audio Component Object Reference)
   SetRenderOpacity(execute=@n4.else, self=Get VBox_CycleInput(), InOpacity=0.25)
   SetVisibility(execute=@n2.else, self=Get VBox_CycleInput(), InVisibility=Collapsed)
  ]}
 RefreshCurrentAugmentList {nodes [
   RefreshCurrentAugmentList() {is=title K2Node_FunctionEntry}
   > Clear(TargetArray:Array of WBP Active Augments List Entry Object References=Get ActiveAugmentsEntries())
   > n1 = For Each Loop(Array:Array of Widget Object References=GetAllChildren(self=Get ScrollBox_ActiveList())) -> (Array Element:Widget Object Reference)
   > n2 = Cast To WBP_ActiveAugments_ListEntry(Object=@n1)
   > Set AugmentEntry(AugmentEntry:WBP Active Augments List Entry Object Reference=@n2) -> (Output_Get:WBP Active Augments List Entry Object Reference)
   > n3 = Set CurrentAugmentIndex(CurrentAugmentIndex:int=@n1.Array Index) -> (Output_Get:int=0)
   > n4 = Branch(Condition=Is Valid Index!(
     TargetArray:Array of Fort Player Augment Item Definition Object References=Get Active Augments() -> (Active Augments:Array of Fort Player Augment Item Definition Object References)
     IndexToTest:int=@n3
    ) -> (ReturnValue:bool))
   > n5 = SetAugmentInfo!(
    execute
    self:WBP Active Augments List Entry Object Reference=Get AugmentEntry() -> (AugmentEntry:WBP Active Augments List Entry Object Reference)
    InAugmentItemDefinition:Fort Player Augment Item Definition Object Reference="Get (a copy)"(
     Array:Array of Fort Player Augment Item Definition Object References=Get Active Augments()
     Dimension 1=Get CurrentAugmentIndex() -> (CurrentAugmentIndex:int=0)
    ) -> (Output:Fort Player Augment Item Definition Object Reference)
   ) -> (then)
   Increment Int(" "=[@n5.then, @n6.then], Value=Get CurrentAugmentIndex() -> (CurrentAugmentIndex=0))
   n6 = SetAsEmptySlot(execute=@n4.else, self=Get AugmentEntry(), InIsEmptySlot=true)
  ]}
 ResetFocusedEntry {nodes [
   ResetFocusedEntry() {is=title K2Node_FunctionEntry}
   > n1 = For Each Loop(Array:Array of WBP Active Augments List Entry Object References=Get ActiveAugmentsEntries()) -> (Array Element:WBP Active Augments List Entry Object Reference)
   > n2 = Branch(Condition="Equal (Integer)"(A=@n1.Array Index))
   > n3 = Set FocusedEntryIndex(FocusedEntryIndex:int=0) -> (Output_Get:int=0)
   > n4 = Set FocusedEntry(FocusedEntry:WBP Active Augments List Entry Object Reference="Get (a copy)"(Array:Array of WBP Active Augments List Entry Object References=Get ActiveAugmentsEntries(), Dimension 1=@n3) -> (Output:WBP Active Augments List Entry Object Reference)) -> (Output_Get:WBP Active Augments List Entry Object Reference)
   > SetEntryFocused!(execute, self:WBP Active Augments List Entry Object Reference=@n4, IsFocused:bool=true) -> (then)
   SetEntryFocused(execute=@n2.else, self=@n1, IsFocused=false)
   RefreshArrows(execute=@n1.Completed) {is=title K2Node_CallFunction}
  ]}
 ScrollDown {nodes [
   ScrollDown() {is=title K2Node_FunctionEntry}
   n1 = Get ActiveAugmentsEntries()
   > n2 = Branch(Condition="Equal (Integer)"(A=Get FocusedEntryIndex() -> (FocusedEntryIndex=0), B=Last Index(TargetArray=@n1) -> (ReturnValue=0)))
   n3 = GetOwningPlayer()
   > Branch(Condition=AND Boolean(A=IsMobileGame(), B=IsUsingTouch(PlayerController=@n3)))
   > n4 = SetFocusedEntry!(execute, self, NewFocusedEntryIndex:int=0) -> (then)
   RefreshArrows(execute=[@n4.then, @n5.then])
   > SpawnSound2D(
    Sound=Get ScrollTickSound() -> (ScrollTickSound:Sound Base Object Reference)
    VolumeMultiplier=1
    PitchMultiplier=1
    StartTime=0
    bPersistAcrossLevelTransition=false
    bAutoDestroy=true
   )
   n5 = SetFocusedEntry(execute=@n2.else, NewFocusedEntryIndex=int + int(A=Get FocusedEntryIndex() -> (FocusedEntryIndex=0)))
  ]}
 ScrollUp {nodes [
   ScrollUp() {is=title K2Node_FunctionEntry}
   > n1 = Branch(Condition="Equal (Integer)"(A=Get FocusedEntryIndex() -> (FocusedEntryIndex=0)))
   SetFocusedEntry(execute=@n1.else, NewFocusedEntryIndex=int - int(A=Get FocusedEntryIndex() -> (FocusedEntryIndex=0)))
   > RefreshArrows()
   > SpawnSound2D(Sound=Get ScrollTickSound(), VolumeMultiplier=1, PitchMultiplier=1, StartTime=0, bPersistAcrossLevelTransition=false, bAutoDestroy=true)
  ]}
 SetFocusedEntry {nodes [
   SetFocusedEntry!() -> (then, NewFocusedEntryIndex:int) {is=title K2Node_FunctionEntry}
   n1 = "Get (a copy)"(Array=Get ActiveAugmentsEntries(), Dimension 1=Get FocusedEntryIndex() -> (FocusedEntryIndex=0))
   > n2 = Is Valid(InputObject=@n1)
   > n3 = SetEntryFocused(self=@n1, IsFocused=false)
   n4 = Set FocusedEntryIndex(execute=[@n3.then, @n2.Is Not Valid], FocusedEntryIndex=Get NewFocusedEntryIndex() -> (NewFocusedEntryIndex:int=0)) -> (Output_Get=0)
   > n5 = Set FocusedEntry(FocusedEntry="Get (a copy)"(Array=Get ActiveAugmentsEntries(), Dimension 1=@n4))
   > SetEntryFocused(self=@n5, IsFocused=true)
   > ScrollWidgetIntoView!(
    execute
    self:Scroll Box Object Reference=Get ScrollBox_ActiveList()
    WidgetToFind:Widget Object Reference=Get FocusedEntry() -> (FocusedEntry:WBP Active Augments List Entry Object Reference)
    AnimateScroll:bool=true
    ScrollDestination:EDescendantScrollDestination Enum=IntoView
    Padding:float=0
   ) -> (then)
  ]}
 UpdateScrollBoundAction {nodes [
   n1 = UpdateScrollBoundAction!() -> (then, InputType:ECommonInputType Enum) {is=title K2Node_FunctionEntry}
   > n2 = Set InputType(InputType:ECommonInputType Enum=@n1) -> (Output_Get:ECommonInputType Enum)
   n3 = Get ActiveAugmentsEntries()
   > n4 = Branch(Condition="Equal (Integer)"(A=Get FocusedEntryIndex() -> (FocusedEntryIndex=0), B=Last Index(TargetArray=@n3) -> (ReturnValue=0)))
   n5 = "Equal (Enum)"(A=@n2, B=MouseAndKeyboard)
   > SetBoundAction!(
    execute
    self:Fort Keybind Widget Object Reference=Get Keybind_CycleAction() -> (Keybind_CycleAction:Keybind Widget Object Reference)
    NewBoundAction:Name=Select(Option 0:Name=AugmentNextAlternative_Gamepad, Option 1:Name=AugmentNextAlternative, Index=@n5) -> (ReturnValue:Name)
   ) -> (then)
   Branch(execute=@n4.else, Condition="Equal (Integer)"(A=Get FocusedEntryIndex() -> (FocusedEntryIndex=0)))
   > SetBoundAction(
    self=Get Keybind_CycleAction()
    NewBoundAction=Select(Option 0=AugmentPreviousAlternative_Gamepad, Option 1=AugmentPreviousAlternative, Index=@n5)
   )
  ]}
]
variables [
 Active Augments {type="array<FortPlayerAugmentItemDefinition>", default []}
 ActiveAugmentsEntries {type="array<WBP_ActiveAugments_ListEntry_C>", default []}
 FocusedEntry {type=WBP_ActiveAugments_ListEntry_C, default=null}
 FocusedEntryIndex {type=integer}
 HighlightSound {type=SoundBase, default=/PlayerAugmentsUI/Sound/PlayerAugments_UI_Selector_Highlight_Cue}
 InputType {type=ECommonInputType, default=MouseAndKeyboard}
 MaxAugmentIndex_DesignTime {type=integer, default=5}
 ScrollTickSound {type=SoundBase, default=/PlayerAugmentsUI/Sound/PlayerAugments_UI_Selector_Tick_Cue}
]
widgets {
 info {
  inheritedWidgetCount=0
  parentClass=/Script/PlayerAugmentsUI.FortPlayerAugmentHUDQuickbarWidget
  rootWidgetClass=&VerticalBox
  widgetCount=14
 }
 widgets [
  VerticalBox_1 {widgetClassPath=&VerticalBox, widget=VerticalBox_1}
  Text_ActiveList_Hdr {
   widgetClassPath=&CommonTextBlock
   parent=VerticalBox_1
   widget=Text_ActiveList_Hdr
   properties {
    lineHeightPercentage=0.9
    mobileFontSizeMultiplier=1.2
    scrollStyle=&TextScrollStyle-Base_C
    text=Active Augments
    textTransformPolicy=ToUpper
    font {
     bForceMonospaced=false
     bMaterialIsStencil=false
     fontMaterial=null
     fontObject=&BurbankSmall
     letterSpacing=0
     monospacedWidth=1
     size=13
     skewAmount=0
     typefaceFontName=Black
     outlineSettings {
      bApplyOutlineToDropShadows=false
      bMiteredCorners=false
      bSeparateFillAlpha=true
      outlineMaterial=&M-FontShadow
      outlineSize=2
      outlineColor=<rgba 0 0 0 0.5>
     }
    }
    margin=<ltrb 0 -2 0 0>
    renderTransform {angle=0, scale=<1 1>, shear=<-12 0>, translation=<0 0>}
    shadowOffset=<0 0>
   }
   slot {class=VerticalBoxSlot, properties {horizontalAlignment=HAlign_Center, verticalAlignment=VAlign_Top, padding=<ltrb 0 0 0 15>}}
  }
  Overlay_AugScrollList {widgetClassPath=&Overlay, parent=VerticalBox_1, widget=Overlay_AugScrollList, slot {class=VerticalBoxSlot}}
  Backing_Vignette {
   widgetClassPath=&Image
   parent=Overlay_AugScrollList
   widget=Backing_Vignette
   properties {
    visibility=SelfHitTestInvisible
    brush {
     bIsDynamicallyLoaded=false
     drawAs=Box
     imageType=NoImage
     mirroring=NoMirror
     resourceName=None
     resourceObject=/PlayerAugmentsUI/Materials/MI_UI_PlayerAugment_Panel_Darken
     tiling=NoTile
     imageSize=<640 256>
     margin=<ltrb 0.5 0.5 0.5 0.5>
     outlineSettings {
      bUseBrushTransparency=false
      roundingType=HalfHeightRadius
      width=0
      color {colorUseRule=UseColor_Specified, specifiedColor=<rgba 0 0 0 0>}
      cornerRadii=<0 0 0 0>
     }
     tintColor {colorUseRule=UseColor_Specified, specifiedColor=<rgba 1 1 1 1>}
     uVRegion {bIsValid=false, max=<0 0>, min=<0 0>}
    }
   }
   slot {class=OverlaySlot, properties {horizontalAlignment=HAlign_Fill, verticalAlignment=VAlign_Fill, padding=<ltrb -64 -64 -64 -64>}}
  }
  Backing_Glow {
   widgetClassPath=&Image
   parent=Overlay_AugScrollList
   widget=Backing_Glow
   properties {
    visibility=SelfHitTestInvisible
    brush {
     bIsDynamicallyLoaded=false
     drawAs=Box
     imageType=NoImage
     mirroring=NoMirror
     resourceName=None
     resourceObject=/PlayerAugmentsUI/Materials/MI_UI_PlayerAugment_Panel_Glow
     tiling=NoTile
     imageSize=<640 256>
     margin=<ltrb 0.5 0.5 0.5 0.5>
     outlineSettings {
      bUseBrushTransparency=false
      roundingType=HalfHeightRadius
      width=0
      color {colorUseRule=UseColor_Specified, specifiedColor=<rgba 0 0 0 0>}
      cornerRadii=<0 0 0 0>
     }
     tintColor {colorUseRule=UseColor_Specified, specifiedColor=<rgba 1 1 1 1>}
     uVRegion {bIsValid=false, max=<0 0>, min=<0 0>}
    }
   }
   slot {class=OverlaySlot, properties {horizontalAlignment=HAlign_Fill, verticalAlignment=VAlign_Fill, padding=<ltrb -96 -96 -96 -96>}}
  }
  HorizontalBox_0 {
   widgetClassPath=&HorizontalBox
   parent=Overlay_AugScrollList
   widget=HorizontalBox_0
   slot {class=OverlaySlot, properties {horizontalAlignment=HAlign_Fill, verticalAlignment=VAlign_Fill}}
  }
  HBox_ScrollInput {
   widgetClassPath=&HorizontalBox
   parent=HorizontalBox_0
   widget=HBox_ScrollInput
   slot {class=HorizontalBoxSlot, properties {horizontalAlignment=HAlign_Left, verticalAlignment=VAlign_Center, padding=<ltrb 0 0 10 0>}}
  }
  VBox_CycleInput {
   widgetClassPath=&VerticalBox
   bIsVariable=true
   parent=HBox_ScrollInput
   widget=VBox_CycleInput
   properties {visibility=Collapsed}
   slot {class=HorizontalBoxSlot}
  }
  Image_ArrowUp {
   widgetClassPath=&Image
   bIsVariable=true
   parent=VBox_CycleInput
   widget=Image_ArrowUp
   properties {
    visibility=SelfHitTestInvisible
    brush {
     bIsDynamicallyLoaded=false
     drawAs=Image
     imageType=NoImage
     mirroring=NoMirror
     resourceName=None
     resourceObject=/PlayerAugmentsUI/Materials/MI_UI_AugmentArrow
     tiling=NoTile
     imageSize=<32 32>
     margin=<ltrb 0 0 0 0>
     outlineSettings {
      bUseBrushTransparency=false
      roundingType=HalfHeightRadius
      width=0
      color {colorUseRule=UseColor_Specified, specifiedColor=<rgba 0 0 0 0>}
      cornerRadii=<0 0 0 0>
     }
     tintColor {colorUseRule=UseColor_Specified, specifiedColor=<rgba 1 1 1 1>}
     uVRegion {bIsValid=false, max=<0 0>, min=<0 0>}
    }
    renderTransform {angle=180, scale=<1 1>, shear=<0 0>, translation=<0 0>}
   }
   slot {class=VerticalBoxSlot, properties {horizontalAlignment=HAlign_Center, verticalAlignment=VAlign_Center, padding=<ltrb 0 0 0 1>}}
  }
  SizeBox_0 {
   widgetClassPath=&SizeBox
   parent=VBox_CycleInput
   widget=SizeBox_0
   properties {bOverride_HeightOverride=true, bOverride_WidthOverride=true, heightOverride=24, widthOverride=24}
   slot {class=VerticalBoxSlot, properties {horizontalAlignment=HAlign_Center, verticalAlignment=VAlign_Center}}
  }
  Keybind_CycleAction {
   widgetClassPath=/FortUILibrary/Input/KeybindWidget
   bIsVariable=true
   parent=SizeBox_0
   widget=Keybind_CycleAction
   properties {boundAction=AugmentPreviousAlternative, boundActionFallbacks [AugmentPreviousAlternative_Gamepad]}
   slot {class=SizeBoxSlot}
  }
  Image_ArrowDown {
   widgetClassPath=&Image
   bIsVariable=true
   parent=VBox_CycleInput
   widget=Image_ArrowDown
   properties {
    visibility=SelfHitTestInvisible
    brush {
     bIsDynamicallyLoaded=false
     drawAs=Image
     imageType=NoImage
     mirroring=NoMirror
     resourceName=None
     resourceObject=&MI_UI_AugmentArrow
     tiling=NoTile
     imageSize=<32 32>
     margin=<ltrb 0 0 0 0>
     outlineSettings {
      bUseBrushTransparency=false
      roundingType=HalfHeightRadius
      width=0
      color {colorUseRule=UseColor_Specified, specifiedColor=<rgba 0 0 0 0>}
      cornerRadii=<0 0 0 0>
     }
     tintColor {colorUseRule=UseColor_Specified, specifiedColor=<rgba 1 1 1 1>}
     uVRegion {bIsValid=false, max=<0 0>, min=<0 0>}
    }
   }
   slot {class=VerticalBoxSlot, properties {horizontalAlignment=HAlign_Center, verticalAlignment=VAlign_Center}}
  }
  SizeBox_ActiveList {
   widgetClassPath=&SizeBox
   parent=HorizontalBox_0
   widget=SizeBox_ActiveList
   properties {bOverride_HeightOverride=true, clipping=ClipToBounds, heightOverride=575}
   slot {class=HorizontalBoxSlot}
  }
  ScrollBox_ActiveList {
   widgetClassPath=/Script/FortniteUI.FortChildNavigatableScrollBox
   bIsVariable=true
   parent=SizeBox_ActiveList
   widget=ScrollBox_ActiveList
   properties {
    scrollbarPadding=<ltrb 10 2 2 2>
    widgetBarStyle {
     thickness=16
     draggedThumbImage {
      bIsDynamicallyLoaded=false
      drawAs=Image
      imageType=FullColor
      mirroring=NoMirror
      resourceName=../../../Engine/Content/Slate/Common/Scrollbar_Thumb.png
      resourceObject=/Game/Athena/UI/ReplayBrowser/M_Ui_Skewed_Vfor9slices_BG_Separator_Inst
      tiling=NoTile
      imageSize=<0 0>
      margin=<ltrb 0.25 0.25 0.25 0.25>
      outlineSettings {
       bUseBrushTransparency=false
       roundingType=HalfHeightRadius
       width=0
       color {colorUseRule=UseColor_Specified, specifiedColor=<rgba 0 0 0 0>}
       cornerRadii=<0 0 0 0>
      }
      tintColor {colorUseRule=UseColor_Specified, specifiedColor=<rgba 1 1 1 1>}
      uVRegion {bIsValid=false, max=<0 0>, min=<0 0>}
     }
     horizontalBackgroundImage {
      bIsDynamicallyLoaded=false
      drawAs=Image
      imageType=NoImage
      mirroring=NoMirror
      resourceName=None
      resourceObject=&M_Ui_Skewed_Vfor9slices_BG_Separator_Inst
      tiling=NoTile
      imageSize=<0 0>
      margin=<ltrb 0 0 0 0>
      outlineSettings {
       bUseBrushTransparency=false
       roundingType=HalfHeightRadius
       width=0
       color {colorUseRule=UseColor_Specified, specifiedColor=<rgba 0 0 0 0>}
       cornerRadii=<0 0 0 0>
      }
      tintColor {colorUseRule=UseColor_Specified, specifiedColor=<rgba 0 0.2159 1 0.498>}
      uVRegion {bIsValid=false, max=<0 0>, min=<0 0>}
     }
     horizontalBottomSlotImage {
      bIsDynamicallyLoaded=false
      drawAs=NoDrawType
      imageType=FullColor
      mirroring=NoMirror
      resourceName=../../../Engine/Content/Slate/Common/Scrollbar_Background_Horizontal.png
      resourceObject=null
      tiling=NoTile
      imageSize=<8 8>
      margin=<ltrb 0 0 0 0>
      outlineSettings {
       bUseBrushTransparency=false
       roundingType=HalfHeightRadius
       width=0
       color {colorUseRule=UseColor_Specified, specifiedColor=<rgba 0 0 0 0>}
       cornerRadii=<0 0 0 0>
      }
      tintColor {colorUseRule=UseColor_Specified, specifiedColor=<rgba 1 1 1 1>}
      uVRegion {bIsValid=false, max=<0 0>, min=<0 0>}
     }
     horizontalTopSlotImage {
      bIsDynamicallyLoaded=false
      drawAs=NoDrawType
      imageType=FullColor
      mirroring=NoMirror
      resourceName=../../../Engine/Content/Slate/Common/Scrollbar_Background_Horizontal.png
      resourceObject=null
      tiling=NoTile
      imageSize=<8 8>
      margin=<ltrb 0 0 0 0>
      outlineSettings {
       bUseBrushTransparency=false
       roundingType=HalfHeightRadius
       width=0
       color {colorUseRule=UseColor_Specified, specifiedColor=<rgba 0 0 0 0>}
       cornerRadii=<0 0 0 0>
      }
      tintColor {colorUseRule=UseColor_Specified, specifiedColor=<rgba 1 1 1 1>}
      uVRegion {bIsValid=false, max=<0 0>, min=<0 0>}
     }
     hoveredThumbImage {
      bIsDynamicallyLoaded=false
      drawAs=Image
      imageType=FullColor
      mirroring=NoMirror
      resourceName=../../../Engine/Content/Slate/Common/Scrollbar_Thumb.png
      resourceObject=&M_Ui_Skewed_Vfor9slices_BG_Separator_Inst
      tiling=NoTile
      imageSize=<0 0>
      margin=<ltrb 0.25 0.25 0.25 0.25>
      outlineSettings {
       bUseBrushTransparency=false
       roundingType=HalfHeightRadius
       width=0
       color {colorUseRule=UseColor_Specified, specifiedColor=<rgba 0 0 0 0>}
       cornerRadii=<0 0 0 0>
      }
      tintColor {colorUseRule=UseColor_Specified, specifiedColor=<rgba 0.5573 0.808 1 1>}
      uVRegion {bIsValid=false, max=<0 0>, min=<0 0>}
     }
     normalThumbImage {
      bIsDynamicallyLoaded=false
      drawAs=Image
      imageType=FullColor
      mirroring=NoMirror
      resourceName=../../../Engine/Content/Slate/Common/Scrollbar_Thumb.png
      resourceObject=&M_Ui_Skewed_Vfor9slices_BG_Separator_Inst
      tiling=NoTile
      imageSize=<12 12>
      margin=<ltrb 0.25 0.25 0.25 0.25>
      outlineSettings {
       bUseBrushTransparency=false
       roundingType=HalfHeightRadius
       width=0
       color {colorUseRule=UseColor_Specified, specifiedColor=<rgba 0 0 0 0>}
       cornerRadii=<0 0 0 0>
      }
      tintColor {colorUseRule=UseColor_Specified, specifiedColor=<rgba 0.2122 0.6584 1 1>}
      uVRegion {bIsValid=false, max=<0 0>, min=<0 0>}
     }
     verticalBackgroundImage {
      bIsDynamicallyLoaded=false
      drawAs=NoDrawType
      imageType=NoImage
      mirroring=NoMirror
      resourceName=None
      resourceObject=null
      tiling=NoTile
      imageSize=<0 0>
      margin=<ltrb 0 0 0 0>
      outlineSettings {
       bUseBrushTransparency=false
       roundingType=HalfHeightRadius
       width=0
       color {colorUseRule=UseColor_Specified, specifiedColor=<rgba 0 0 0 0>}
       cornerRadii=<0 0 0 0>
      }
      tintColor {colorUseRule=UseColor_Specified, specifiedColor=<rgba 1 1 1 1>}
      uVRegion {bIsValid=false, max=<0 0>, min=<0 0>}
     }
     verticalBottomSlotImage {
      bIsDynamicallyLoaded=false
      drawAs=Image
      imageType=FullColor
      mirroring=NoMirror
      resourceName=../../../Engine/Content/Slate/Common/Scrollbar_Background_Vertical.png
      resourceObject=null
      tiling=NoTile
      imageSize=<8 8>
      margin=<ltrb 0 0 0 0>
      outlineSettings {
       bUseBrushTransparency=false
       roundingType=HalfHeightRadius
       width=0
       color {colorUseRule=UseColor_Specified, specifiedColor=<rgba 0 0 0 0>}
       cornerRadii=<0 0 0 0>
      }
      tintColor {colorUseRule=UseColor_Specified, specifiedColor=<rgba 1 1 1 1>}
      uVRegion {bIsValid=false, max=<0 0>, min=<0 0>}
     }
     verticalTopSlotImage {
      bIsDynamicallyLoaded=false
      drawAs=Image
      imageType=FullColor
      mirroring=NoMirror
      resourceName=../../../Engine/Content/Slate/Common/Scrollbar_Background_Vertical.png
      resourceObject=null
      tiling=NoTile
      imageSize=<8 8>
      margin=<ltrb 0 0 0 0>
      outlineSettings {
       bUseBrushTransparency=false
       roundingType=HalfHeightRadius
       width=0
       color {colorUseRule=UseColor_Specified, specifiedColor=<rgba 0 0 0 0>}
       cornerRadii=<0 0 0 0>
      }
      tintColor {colorUseRule=UseColor_Specified, specifiedColor=<rgba 1 1 1 1>}
      uVRegion {bIsValid=false, max=<0 0>, min=<0 0>}
     }
    }
    widgetStyle {
     barThickness=9
     bottomShadowBrush {
      bIsDynamicallyLoaded=false
      drawAs=Box
      imageType=FullColor
      mirroring=NoMirror
      resourceName=../../../Engine/Content/Slate/Common/ScrollBoxShadowBottom.png
      resourceObject=null
      tiling=NoTile
      imageSize=<16 8>
      margin=<ltrb 0.5 0 0.5 1>
      outlineSettings {
       bUseBrushTransparency=false
       roundingType=HalfHeightRadius
       width=0
       color {colorUseRule=UseColor_Specified, specifiedColor=<rgba 0 0 0 0>}
       cornerRadii=<0 0 0 0>
      }
      tintColor {colorUseRule=UseColor_Specified, specifiedColor=<rgba 1 1 1 1>}
      uVRegion {bIsValid=false, max=<0 0>, min=<0 0>}
     }
     horizontalScrolledContentPadding=<ltrb 0 0 1 0>
     leftShadowBrush {
      bIsDynamicallyLoaded=false
      drawAs=NoDrawType
      imageType=FullColor
      mirroring=NoMirror
      resourceName=../../../Engine/Content/Slate/Common/ScrollBoxShadowLeft.png
      resourceObject=null
      tiling=NoTile
      imageSize=<8 16>
      margin=<ltrb 1 0.5 0 0.5>
      outlineSettings {
       bUseBrushTransparency=false
       roundingType=HalfHeightRadius
       width=0
       color {colorUseRule=UseColor_Specified, specifiedColor=<rgba 0 0 0 0>}
       cornerRadii=<0 0 0 0>
      }
      tintColor {colorUseRule=UseColor_Specified, specifiedColor=<rgba 1 1 1 1>}
      uVRegion {bIsValid=false, max=<0 0>, min=<0 0>}
     }
     rightShadowBrush {
      bIsDynamicallyLoaded=false
      drawAs=NoDrawType
      imageType=FullColor
      mirroring=NoMirror
      resourceName=../../../Engine/Content/Slate/Common/ScrollBoxShadowRight.png
      resourceObject=null
      tiling=NoTile
      imageSize=<8 16>
      margin=<ltrb 0 0.5 1 0.5>
      outlineSettings {
       bUseBrushTransparency=false
       roundingType=HalfHeightRadius
       width=0
       color {colorUseRule=UseColor_Specified, specifiedColor=<rgba 0 0 0 0>}
       cornerRadii=<0 0 0 0>
      }
      tintColor {colorUseRule=UseColor_Specified, specifiedColor=<rgba 1 1 1 1>}
      uVRegion {bIsValid=false, max=<0 0>, min=<0 0>}
     }
     topShadowBrush {
      bIsDynamicallyLoaded=false
      drawAs=Box
      imageType=FullColor
      mirroring=NoMirror
      resourceName=../../../Engine/Content/Slate/Common/ScrollBoxShadowTop.png
      resourceObject=null
      tiling=NoTile
      imageSize=<16 8>
      margin=<ltrb 0.5 1 0.5 0>
      outlineSettings {
       bUseBrushTransparency=false
       roundingType=HalfHeightRadius
       width=0
       color {colorUseRule=UseColor_Specified, specifiedColor=<rgba 0 0 0 0>}
       cornerRadii=<0 0 0 0>
      }
      tintColor {colorUseRule=UseColor_Specified, specifiedColor=<rgba 1 1 1 1>}
      uVRegion {bIsValid=false, max=<0 0>, min=<0 0>}
     }
     verticalScrolledContentPadding=<ltrb 0 0 0 1>
    }
   }
   slot {class=SizeBoxSlot}
  }
 ]
}
