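# HUD timer widget for player augments that shows a countdown to the next augment selection, or "Selector Offline".
# Parent-class events update the formatted time, pending-augment count and finished-for-match flag, which drive a visibility switcher between active and offline overlays.
# Once timers finish, the switcher collapses while augments are pending.
# Dispatchers broadcast pending-count changes, and when none remain after timers finish.
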WidgetBlueprint /PlayerAugmentsUI/UI/ActiveAugmentsPanel/WBP_ActiveAugments_Timer
graphs [
 EventGraph {nodes [
   n1 = Event OnNumPendingAugmentsToGrantUpdated!() -> (OutputDelegate, then, NumPendingAugmentsToGrant:int=0) {is=title K2Node_Event}
   > n2 = Set Num Pending Augments to Grant(Num Pending Augments to Grant:int=@n1.NumPendingAugmentsToGrant) -> (Output_Get:int=0)
   > Call On Num Aug Pending Updated!(execute:Exec, self:WBP Active Augments Timer Object Reference, NumPendingToGrant:int=@n2) -> (then:Exec) {is=title K2Node_CallDelegate}
   > n3 = Branch(Condition=Get TimersFinishedForMatch() -> (TimersFinishedForMatch:bool))
   > SetActiveWidget(
    self=Get VisSwitcher_TimerState() -> (VisSwitcher_TimerState:Common Visibility Switcher Object Reference)
    Widget=Get Overlay_Offline() -> (Overlay_Offline:Overlay Object Reference)
   )
   n4 = "integer >= integer"(A=Get Num Pending Augments to Grant() -> (Num Pending Augments to Grant:int=0), B=1)
   > SetVisibility(self=Get VisSwitcher_TimerState(), InVisibility=Select(Option 0=SelfHitTestInvisible, Option 1=Collapsed, Index=@n4))
   > n5 = Branch(Condition=@n4)
   Call On No Aug Left for Selection!(execute:Exec=@n5.else, self:WBP Active Augments Timer Object Reference) -> (then:Exec) {is=title K2Node_CallDelegate}
   SetActiveWidget(execute=@n3.else, self=Get VisSwitcher_TimerState(), Widget=Get Overlay_Active() -> (Overlay_Active:Overlay Object Reference))
   Event OnAugmentGrantingTimersFinishedForMatch() {is=title K2Node_Event}
   > Set TimersFinishedForMatch(TimersFinishedForMatch:bool=true) -> (Output_Get:bool)
   > n6 = Branch(Condition="integer >= integer"(A=Get Num Pending Augments to Grant() -> (Num Pending Augments to Grant=0), B=1))
   > SetVisibility(self=Get VisSwitcher_TimerState(), InVisibility=Collapsed)
   SetActiveWidget(execute=@n6.else, self=Get VisSwitcher_TimerState(), Widget=Get Overlay_Offline())
   > Call On No Aug Left for Selection()
   n7 = Event OnTimerUpdated!() -> (OutputDelegate, then, TimeRemaining:float, TimePercentage:float) {is=title K2Node_Event}
   > Set Text(
    Text=FormatTime!(self:Fort Player Augment HUDTimer Widget Object Reference, TotalSeconds:float=@n7.TimeRemaining) -> (ReturnValue:Text)
    self=Get Text_TimeRemaining() -> (Text_TimeRemaining:Common Text Object Reference)
   )
   Event OnAugmentGrantingTimersResetForMatch() {is=title K2Node_Event}
   > Set TimersFinishedForMatch(TimersFinishedForMatch=false)
   > SetActiveWidget(self=Get VisSwitcher_TimerState(), Widget=Get Overlay_Active())
   > SetVisibility(self=Get VisSwitcher_TimerState(), InVisibility=SelfHitTestInvisible)
  ]}
 OnNoAugLeftForSelection {nodes [OnNoAugLeftForSelection() {is=title K2Node_FunctionEntry}]}
 OnNumAugPendingUpdated {nodes [OnNumAugPendingUpdated!() -> (then, NumPendingToGrant:int) {is=title K2Node_FunctionEntry}]}
]
variables [
 Can Select Augment {type=boolean}
 Num Pending Augments to Grant {type=integer}
 OnNoAugLeftForSelection {}
 OnNumAugPendingUpdated {}
 TimersFinishedForMatch {type=boolean}
]
widgets {
 info {
  inheritedWidgetCount=1
  parentClass=/Script/PlayerAugmentsUI.FortPlayerAugmentHUDTimerWidget
  rootWidgetClass=&Overlay
  widgetCount=11
 }
 widgets [
  Overlay_Root {widgetClassPath=&Overlay, widget=Overlay_Root}
  SizeBox_0 {
   widgetClassPath=&SizeBox
   parent=Overlay_Root
   widget=SizeBox_0
   properties {bOverride_WidthOverride=true, widthOverride=350}
   slot {class=OverlaySlot, properties {horizontalAlignment=HAlign_Center, verticalAlignment=VAlign_Center}}
  }
  VisSwitcher_TimerState {
   widgetClassPath=&CommonVisibilitySwitcher
   bIsVariable=true
   parent=SizeBox_0
   widget=VisSwitcher_TimerState
   slot {class=SizeBoxSlot}
  }
  Overlay_Active {
   widgetClassPath=&Overlay
   bIsVariable=true
   parent=VisSwitcher_TimerState
   widget=Overlay_Active
   slot {class=CommonVisibilitySwitcherSlot}
  }
  HBox_Timer {widgetClassPath=&HorizontalBox, parent=Overlay_Active, widget=HBox_Timer, slot {class=OverlaySlot}}
  Text_AugSelection {
   widgetClassPath=&CommonTextBlock
   parent=HBox_Timer
   widget=Text_AugSelection
   properties {
    lineHeightPercentage=0.9
    mobileFontSizeMultiplier=1.2
    scrollStyle=&TextScrollStyle-Base_C
    text=Next Augment Selection:
    textTransformPolicy=ToUpper
    font {
     bForceMonospaced=false
     bMaterialIsStencil=false
     fontMaterial=null
     fontObject=&BurbankSmall
     letterSpacing=0
     monospacedWidth=1
     size=13
     skewAmount=0
     typefaceFontName=Black
     outlineSettings {
      bApplyOutlineToDropShadows=false
      bMiteredCorners=false
      bSeparateFillAlpha=true
      outlineMaterial=&M-FontShadow
      outlineSize=2
      outlineColor=<rgba 0 0 0 0.5>
     }
    }
    margin=<ltrb 0 -2 0 0>
    renderTransform {angle=0, scale=<1 1>, shear=<-12 0>, translation=<0 0>}
    shadowOffset=<0 0>
   }
   slot {class=HorizontalBoxSlot, properties {horizontalAlignment=HAlign_Center, verticalAlignment=VAlign_Center}}
  }
  Spacer_88 {widgetClassPath=&Spacer, parent=HBox_Timer, widget=Spacer_88, properties {size=<10 1>}, slot {class=HorizontalBoxSlot}}
  Text_TimeRemaining {
   widgetClassPath=&CommonTextBlock
   bIsVariable=true
   parent=HBox_Timer
   widget=Text_TimeRemaining
   properties {
    bAutoCollapseWithEmptyText=true
    mobileFontSizeMultiplier=1.2
    scrollStyle=&TextScrollStyle-Base_C
    text=00:00
    textTransformPolicy=ToUpper
    font {
     bForceMonospaced=false
     bMaterialIsStencil=false
     fontMaterial=null
     fontObject=&Burbank
     letterSpacing=0
     monospacedWidth=1
     size=16
     skewAmount=0
     typefaceFontName=BurbankRegular-Black
     outlineSettings {
      bApplyOutlineToDropShadows=false
      bMiteredCorners=false
      bSeparateFillAlpha=false
      outlineMaterial=null
      outlineSize=0
      outlineColor=<rgba 0 0 0 1>
     }
    }
    renderTransform {angle=0, scale=<1 1>, shear=<-12 0>, translation=<0 0>}
    shadowOffset=<0 0>
   }
   slot {class=HorizontalBoxSlot, properties {horizontalAlignment=HAlign_Center, verticalAlignment=VAlign_Center}}
  }
  Overlay_Offline {
   widgetClassPath=&Overlay
   bIsVariable=true
   parent=VisSwitcher_TimerState
   widget=Overlay_Offline
   slot {class=CommonVisibilitySwitcherSlot}
  }
  Text_SelectorOffline {
   widgetClassPath=&CommonTextBlock
   parent=Overlay_Offline
   widget=Text_SelectorOffline
   properties {
    lineHeightPercentage=0.9
    mobileFontSizeMultiplier=1.2
    scrollStyle=&TextScrollStyle-Base_C
    text=Selector Offline
    textTransformPolicy=ToUpper
    font {
     bForceMonospaced=false
     bMaterialIsStencil=false
     fontMaterial=/Game/Athena/HUD/Inventory/Art/M_UI_Inventory_CatergoryHeader_FontColor
     fontObject=&BurbankSmall
     letterSpacing=0
     monospacedWidth=1
     size=13
     skewAmount=0
     typefaceFontName=Black
     outlineSettings {
      bApplyOutlineToDropShadows=false
      bMiteredCorners=false
      bSeparateFillAlpha=true
      outlineMaterial=&M-FontShadow
      outlineSize=2
      outlineColor=<rgba 0 0 0 0.5>
     }
    }
    margin=<ltrb 0 -2 0 0>
    renderTransform {angle=0, scale=<1 1>, shear=<-12 0>, translation=<0 0>}
    shadowOffset=<0 0>
   }
   slot {class=OverlaySlot, properties {horizontalAlignment=HAlign_Center, verticalAlignment=VAlign_Center}}
  }
  TutorialHighlight {widgetClassPath=&TutorialHighlight}
 ]
}
