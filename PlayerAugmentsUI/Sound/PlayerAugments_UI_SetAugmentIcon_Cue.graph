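# Sound cue in the PlayerAugmentsUI content, named for the UI sound of setting an augment icon.
# It has no root or sound nodes, so as shown it plays no audio.
# It is assigned to the critical in-game UI sound class at reduced volume, with the debug flag enabled.
# A concurrency override allows two simultaneous instances and stops the oldest beyond that.
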SoundCue /PlayerAugmentsUI/Sound/PlayerAugments_UI_SetAugmentIcon_Cue
root=null
nodes []
properties {
 bDebug=true
 bOverrideConcurrency=true
 maxDistance=2097152
 soundClassObject=&Fort_UI_InGame_Critical
 volumeMultiplier=0.4
 concurrencyOverrides {maxCount=2, resolutionRule=StopOldest}
}
