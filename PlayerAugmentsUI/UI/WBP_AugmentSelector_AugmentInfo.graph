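# Reality augment selection card, built on FortPlayerAugmentInfoEntryWidget, showing one augment's icon, text, activate button and new or borrowed badges.
# SetAugmentInfo updates the entry and async-loads the HUD effect texture into the icon's dynamic material.
# Focus, pulse, reroll and activation-confirmation animations drive material parameters and transforms; SetFocused plays forward or reversed, optionally skipping to the end.
# Input method changes switch compact mobile touch layout; pin color tints teammate icons.
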WidgetBlueprint /PlayerAugmentsUI/UI/WBP_AugmentSelector_AugmentInfo
animations [
 Anim_Focused {
  binding_count=19
  length_seconds=0.5
  object=&self:Anim_Focused
  sequence {
   bindings [
    Icon_BackingPanel {
     id=#1
     possessed_class=Image
     tracks [{
       type=MovieSceneWidgetMaterialTrack
       sections [{
         type=MovieSceneParameterSection
         channels [
          Anim_BackingAlpha {type=MovieSceneScriptingFloatChannel, keys [[0,0],[15,1]]}
          Anim_Color {type=MovieSceneScriptingFloatChannel, keys [[0,0],[15,1]]}
         ]
        }]
      }]
    }
    Backing_Glow {
     id=#2
     possessed_class=Image
     tracks [{
       type=MovieSceneWidgetMaterialTrack
       sections [{type=MovieSceneParameterSection, channels [Anim_BackingAlpha {type=MovieSceneScriptingFloatChannel, keys [[0,0],[15,1]]}]}]
      }]
    }
    Icon_BackingPanel_Reflection {
     id=#3
     possessed_class=Image
     tracks [
      {
       type=MovieSceneWidgetMaterialTrack
       sections [{type=MovieSceneParameterSection, channels [Anim_BackingAlpha {type=MovieSceneScriptingFloatChannel, keys [[0,0],[15,1]]}]}]
      }
      {
       type=MovieScene2DTransformTrack
       sections [{
         type=MovieScene2DTransformSection
         channels [
          Translation.X {type=MovieSceneScriptingFloatChannel, keys [[0,10],[15,12]]}
          Translation.Y {type=MovieSceneScriptingFloatChannel, keys [[0,16],[15,24]]}
         ]
        }]
      }
     ]
    }
    Backing_EdgeGlow {
     id=#4
     possessed_class=Image
     tracks [{
       type=MovieSceneWidgetMaterialTrack
       sections [{type=MovieSceneParameterSection, channels [Anim_BorderAlpha {type=MovieSceneScriptingFloatChannel, keys [[0,0],[15,1]]}]}]
      }]
    }
    Backing_Panel {
     id=#5
     possessed_class=Image
     tracks [{
       type=MovieSceneWidgetMaterialTrack
       sections [{
         type=MovieSceneParameterSection
         channels [
          Anim_BackingAlpha {type=MovieSceneScriptingFloatChannel, keys [[0,0],[15,1]]}
          Anim_ScanLines_Amount {type=MovieSceneScriptingFloatChannel, keys [[0,0],[6,16],[15,0]]}
          Anim_Color {type=MovieSceneScriptingFloatChannel, keys [[0,0],[15,1]]}
         ]
        }]
      }]
    }
    Backing_Vignette {
     id=#6
     possessed_class=Image
     tracks [{
       type=MovieSceneWidgetMaterialTrack
       sections [{type=MovieSceneParameterSection, channels [Anim_BackingAlpha {type=MovieSceneScriptingFloatChannel, keys [[0,0],[15,1]]}]}]
      }]
    }
    Icon {
     id=#7
     possessed_class=Overlay
     tracks [{
       type=MovieScene2DTransformTrack
       sections [{
         type=MovieScene2DTransformSection
         channels [
          Translation.X {type=MovieSceneScriptingFloatChannel, keys [[0,0],[15,-4]]}
          Translation.Y {type=MovieSceneScriptingFloatChannel, keys [[0,0],[15,-4]]}
          Scale.X {type=MovieSceneScriptingFloatChannel, keys [[0,1],[15,1.05]]}
          Scale.Y {type=MovieSceneScriptingFloatChannel, keys [[0,1],[15,1.05]]}
         ]
        }]
      }]
    }
    Text_AugmentDescription {
     id=#8
     possessed_class=CommonTextBlock
     tracks [{
       type=MovieSceneWidgetMaterialTrack
       display_name=Font.FontMaterial
       sections [{
         type=MovieSceneParameterSection
         channels [
          Anim_BackingAlpha {type=MovieSceneScriptingFloatChannel, keys [[0,0],[15,1]]}
          Anim_Color {type=MovieSceneScriptingFloatChannel, keys [[0,0],[15,1]]}
         ]
        }]
      }]
    }
    Image_SolidBG_AugmentDescription_Reflection {
     id=#9
     possessed_class=Image
     tracks [{
       type=MovieSceneWidgetMaterialTrack
       sections [{type=MovieSceneParameterSection, start_frame=0, channels [Anim_BackingAlpha {type=MovieSceneScriptingFloatChannel, keys [[0,0],[15,1]]}]}]
      }]
    }
    Button_AugmentActivate {
     id=#10
     possessed_class=Overlay
     tracks [
      {
       type=MovieScene2DTransformTrack
       sections [{
         type=MovieScene2DTransformSection
         channels [
          Scale.X {type=MovieSceneScriptingFloatChannel, const [1,1]}
          Scale.Y {type=MovieSceneScriptingFloatChannel, keys [[0,0.125],[17,0.125],[18,1]]}
         ]
        }]
      }
      {
       type=MovieSceneFloatTrack
       display_name=Render Opacity
       sections [{type=MovieSceneFloatSection, channels [None {type=MovieSceneScriptingFloatChannel, keys [[12,0],[14,1]]}]}]
      }
     ]
    }
    Image_SolidBG_AugmentDescription {
     id=#11
     possessed_class=Image
     tracks [{
       type=MovieSceneWidgetMaterialTrack
       sections [{
         type=MovieSceneParameterSection
         channels [
          Anim_BackingAlpha {type=MovieSceneScriptingFloatChannel, keys [[0,0],[15,1]]}
          Anim_Color {type=MovieSceneScriptingFloatChannel, keys [[0,0],[15,1]]}
         ]
        }]
      }]
    }
    Icon_Border {
     id=#12
     possessed_class=Image
     tracks [{
       type=MovieSceneWidgetMaterialTrack
       sections [{type=MovieSceneParameterSection, channels [Anim_BorderAlpha {type=MovieSceneScriptingFloatChannel, keys [[0,0],[15,1]]}]}]
      }]
    }
    Text_AugmentName {
     id=#13
     possessed_class=CommonTextBlock
     tracks [{
       type=MovieSceneWidgetMaterialTrack
       display_name=Font.FontMaterial
       sections [{
         type=MovieSceneParameterSection
         channels [
          Anim_BackingAlpha {type=MovieSceneScriptingFloatChannel, keys [[0,0],[15,1]]}
          Anim_Color {type=MovieSceneScriptingFloatChannel, keys [[0,0],[15,1]]}
         ]
        }]
      }]
    }
    Image_Activate_SolidBG {
     id=#14
     possessed_class=Image
     tracks [{
       type=MovieScene2DTransformTrack
       sections [{type=MovieScene2DTransformSection, channels [Scale.X {type=MovieSceneScriptingFloatChannel, keys [[15,3],[21,0.9],[27,1]]}]}]
      }]
    }
    Text {
     id=#15
     possessed_class=VerticalBox
     tracks [{
       type=MovieScene2DTransformTrack
       sections [{
         type=MovieScene2DTransformSection
         end_frame=15
         start_frame=0
         channels [
          Scale.X {type=MovieSceneScriptingFloatChannel, keys [[0,0.95],[15,1]]}
          Scale.Y {type=MovieSceneScriptingFloatChannel, keys [[0,0.95],[15,1]]}
         ]
        }]
      }]
    }
    Content {
     id=#16
     possessed_class=HorizontalBox
     tracks [{
       type=MovieScene2DTransformTrack
       sections [{
         type=MovieScene2DTransformSection
         end_frame=15
         start_frame=0
         channels [
          Scale.X {type=MovieSceneScriptingFloatChannel, keys [[0,0.9],[15,1]]}
          Scale.Y {type=MovieSceneScriptingFloatChannel, keys [[0,0.9],[15,1]]}
         ]
        }]
      }]
    }
    Image_Button_Glow {
     id=#17
     possessed_class=Image
     tracks [{
       type=MovieSceneWidgetMaterialTrack
       sections [{
         type=MovieSceneParameterSection
         end_frame=30
         start_frame=0
         channels [Backing_Alpha {type=MovieSceneScriptingFloatChannel, keys [[15,0],[30,1]]}]
        }]
      }]
    }
    Image_AugmentIcon {
     id=#18
     possessed_class=Image
     tracks [{
       type=MovieSceneWidgetMaterialTrack
       sections [{
         type=MovieSceneParameterSection
         channels [
          Anim_FinalAlpha {type=MovieSceneScriptingFloatChannel, keys [[0,0],[15,1]]}
          Anim_IconColor {type=MovieSceneScriptingFloatChannel, keys [[0,0],[15,1]]}
          Anim_BackingAlpha {type=MovieSceneScriptingFloatChannel, keys [[0,0],[15,1]]}
          Anim_GlowAmount {type=MovieSceneScriptingFloatChannel, keys [[0,0],[15,1]]}
         ]
        }]
      }]
    }
    Image_TextPanelBrackets {
     id=#19
     possessed_class=Image
     tracks [{
       type=MovieSceneWidgetMaterialTrack
       sections [{
         type=MovieSceneParameterSection
         channels [
          Border_Scale_X {type=MovieSceneScriptingFloatChannel, const [1.1,2]}
          Anim_BorderAlpha {type=MovieSceneScriptingFloatChannel, keys [[0,0],[15,1]]}
         ]
        }]
      }]
    }
   ]
   display_rate {numerator=60}
  }
 }
 Anim_ActivationConfirmation {
  binding_count=12
  length_seconds=1.5
  object=&self:Anim_ActivationConfirmation
  sequence {bindings [
    Image_GreenOverlay {
     id=#20
     possessed_class=Image
     tracks [{
       type=MovieSceneWidgetMaterialTrack
       sections [{type=MovieSceneParameterSection, channels [Anim_BackingAlpha {type=MovieSceneScriptingFloatChannel, keys [[0,0],[1.6667,0.5],[2.6667,1]]}]}]
      }]
    }
    Icon_BackingPanel {
     id=#21
     possessed_class=Image
     tracks [{
       type=MovieSceneWidgetMaterialTrack
       sections [{type=MovieSceneParameterSection, channels [Anim_BackingAlpha {type=MovieSceneScriptingFloatChannel, keys [[0,1],[2,4],[20,1]]}]}]
      }]
    }
    Backing_Glow {
     id=#22
     possessed_class=Image
     tracks [
      {
       type=MovieSceneWidgetMaterialTrack
       sections [{type=MovieSceneParameterSection, channels [Anim_BackingAlpha {type=MovieSceneScriptingFloatChannel, keys [[2,1],[4,16],[22,1]]}]}]
      }
      {
       type=MovieScene2DTransformTrack
       sections [{
         type=MovieScene2DTransformSection
         channels [
          Scale.X {type=MovieSceneScriptingFloatChannel, keys [[2,1],[6,1.333],[15,1]]}
          Scale.Y {type=MovieSceneScriptingFloatChannel, keys [[2,1],[6,1.125],[15,1]]}
         ]
        }]
      }
     ]
    }
    Backing_Panel {
     id=#23
     possessed_class=Image
     tracks [{
       type=MovieSceneWidgetMaterialTrack
       sections [{
         type=MovieSceneParameterSection
         end_frame=30
         start_frame=0
         channels [Anim_Color {type=MovieSceneScriptingFloatChannel, keys [[1,1],[3,2],[7,1]]}]
        }]
      }]
    }
    Icon_Border_ConfirmationHot {
     id=#24
     possessed_class=Image
     tracks [{
       type=MovieSceneWidgetMaterialTrack
       sections [{
         type=MovieSceneParameterSection
         channels [
          Anim_BorderAlpha {type=MovieSceneScriptingFloatChannel, keys [[5,0],[6,1],[30,0]]}
          Border_Scale_X {type=MovieSceneScriptingFloatChannel, keys [[5,0.9],[9,0.8],[20,1]]}
          Border_Scale_Y {type=MovieSceneScriptingFloatChannel, keys [[5,0.9],[9,0.8],[20,1]]}
         ]
        }]
      }]
    }
    Backing_EdgeGlow {
     id=#25
     possessed_class=Image
     tracks [{
       type=MovieSceneWidgetMaterialTrack
       sections [{type=MovieSceneParameterSection, channels [Anim_BorderAlpha {type=MovieSceneScriptingFloatChannel, keys [[0,1],[2,8],[30,1]]}]}]
      }]
    }
    Button_AugmentActivate {
     id=#26
     possessed_class=Overlay
     tracks [{
       type=MovieScene2DTransformTrack
       sections [{
         type=MovieScene2DTransformSection
         channels [
          Scale.X {type=MovieSceneScriptingFloatChannel, keys [[0,1],[1.6667,1.125],[4.6667,0.95],[8,1]]}
          Scale.Y {type=MovieSceneScriptingFloatChannel, keys [[0,1],[1.6667,1.125],[4.6667,0.95],[8,1]]}
         ]
        }]
      }]
    }
    Icon_Border_ConfirmationPulse {
     id=#27
     possessed_class=Image
     tracks [
      {
       type=MovieSceneWidgetMaterialTrack
       sections [{
         type=MovieSceneParameterSection
         channels [
          Anim_Color {type=MovieSceneScriptingFloatChannel, keys [[7,0],[9,1],[25,0]]}
          Anim_BorderAlpha {type=MovieSceneScriptingFloatChannel, keys [[5,0],[7,1],[27,0]]}
          Border_Scale_X {type=MovieSceneScriptingFloatChannel, keys [[7,0.75],[27,1.1]]}
          Border_Scale_Y {type=MovieSceneScriptingFloatChannel, keys [[7,0.75],[27,1.1]]}
         ]
        }]
      }
      {
       type=MovieScene2DTransformTrack
       sections [{
         type=MovieScene2DTransformSection
         channels [
          Scale.X {type=MovieSceneScriptingFloatChannel, keys [[7,1],[27,1.75]]}
          Scale.Y {type=MovieSceneScriptingFloatChannel, keys [[7,1],[27,1.75]]}
         ]
        }]
      }
     ]
    }
    Icon_ConfirmationFlare {
     id=#28
     possessed_class=Image
     tracks [
      {
       type=MovieScene2DTransformTrack
       sections [{type=MovieScene2DTransformSection, channels [Scale.Y {type=MovieSceneScriptingFloatChannel, keys [[5,1],[15,5]]}]}]
      }
      {
       type=MovieSceneWidgetMaterialTrack
       sections [{type=MovieSceneParameterSection, channels [Anim_BackingAlpha {type=MovieSceneScriptingFloatChannel, keys [[5,0],[8,1],[13,1],[30,0]]}]}]
      }
     ]
    }
    Image_CheckMark {
     id=#29
     possessed_class=Image
     tracks [
      {
       type=MovieSceneFloatTrack
       display_name=Render Opacity
       sections [{type=MovieSceneFloatSection, channels [None {type=MovieSceneScriptingFloatChannel, keys [[1.6667,0],[2.6667,1]]}]}]
      }
      {
       type=MovieScene2DTransformTrack
       sections [{
         type=MovieScene2DTransformSection
         channels [
          Translation.Y {type=MovieSceneScriptingFloatChannel, keys [[1.6667,2],[2.6667,-4],[4.6667,0]]}
          Scale.X {type=MovieSceneScriptingFloatChannel, keys [[1.6667,0.75],[2.6667,1.125],[4.6667,1]]}
          Scale.Y {type=MovieSceneScriptingFloatChannel, keys [[1.6667,0.75],[2.6667,1.125],[4.6667,1]]}
         ]
        }]
      }
     ]
    }
    Image_ConfirmationOutline {
     id=#30
     possessed_class=Image
     tracks [
      {
       type=MovieSceneWidgetMaterialTrack
       sections [{type=MovieSceneParameterSection, channels [Anim_BorderAlpha {type=MovieSceneScriptingFloatChannel, keys [[0,0],[2.3333,1]]}]}]
      }
      {
       type=MovieScene2DTransformTrack
       sections [{
         type=MovieScene2DTransformSection
         channels [
          Scale.X {type=MovieSceneScriptingFloatChannel, keys [[0,1.25],[2.3333,0.95],[4,1]]}
          Scale.Y {type=MovieSceneScriptingFloatChannel, keys [[0,1.25],[2.3333,0.95],[4,1]]}
         ]
        }]
      }
     ]
    }
    Image_AugmentIcon {
     id=#31
     possessed_class=Image
     tracks [
      {
       type=MovieScene2DTransformTrack
       sections [{
         type=MovieScene2DTransformSection
         channels [
          Translation.Y {type=MovieSceneScriptingFloatChannel, keys [[5,0],[6,0],[7,-6],[9,0]]}
          Scale.X {type=MovieSceneScriptingFloatChannel, keys [[5,1],[6,1],[7,1.05],[9,1]]}
          Scale.Y {type=MovieSceneScriptingFloatChannel, keys [[5,1],[6,1],[7,1.05],[9,1]]}
         ]
        }]
      }
      {
       type=MovieSceneWidgetMaterialTrack
       sections [{type=MovieSceneParameterSection, channels [Anim_BackingAlpha {type=MovieSceneScriptingFloatChannel, keys [[5,0],[7,1],[20,1],[30,0]]}]}]
      }
     ]
    }
   ]}
 }
 Anim_Reroll {
  binding_count=4
  length_seconds=1.5
  object=&self:Anim_Reroll
  sequence {bindings [
    Icon {
     id=#32
     possessed_class=Overlay
     tracks [
      {
       type=MovieScene2DTransformTrack
       sections [{type=MovieScene2DTransformSection, channels [Translation.Y {type=MovieSceneScriptingFloatChannel, keys [[10,64],[20,0]]}]}]
      }
      {
       type=MovieSceneFloatTrack
       display_name=Render Opacity
       sections [{type=MovieSceneFloatSection, end_frame=30, start_frame=0, channels [None {type=MovieSceneScriptingFloatChannel, keys [[10,0],[30,1]]}]}]
      }
     ]
    }
    Backing_Panel {
     id=#33
     possessed_class=Image
     tracks [{
       type=MovieSceneWidgetMaterialTrack
       sections [{
         type=MovieSceneParameterSection
         channels [
          Anim_ScanLines_Amount {type=MovieSceneScriptingFloatChannel, keys [[0,0],[3,0],[6,16],[23,0]]}
          ScanLines_Speed {type=MovieSceneScriptingFloatChannel, keys [[3,0.333],[6,2],[10,0.333]]}
         ]
        }]
      }]
    }
    Text {
     id=#34
     possessed_class=VerticalBox
     tracks [
      {
       type=MovieScene2DTransformTrack
       sections [{
         type=MovieScene2DTransformSection
         channels [
          Translation.X {type=MovieSceneScriptingFloatChannel, keys [[13,16],[23,0]]}
          Translation.Y {type=MovieSceneScriptingFloatChannel, keys [[13,16],[23,0]]}
          Scale.X {type=MovieSceneScriptingFloatChannel, keys [[13,0.9],[23,1]]}
          Scale.Y {type=MovieSceneScriptingFloatChannel, keys [[13,0.9],[23,1]]}
         ]
        }]
      }
      {
       type=MovieSceneFloatTrack
       display_name=Render Opacity
       sections [{type=MovieSceneFloatSection, channels [None {type=MovieSceneScriptingFloatChannel, keys [[13,0],[30,1]]}]}]
      }
     ]
    }
    border_pc-console-only {
     id=#35
     possessed_class=CommonHardwareVisibilityBorder
     tracks [{
       type=MovieSceneFloatTrack
       display_name=Render Opacity
       sections [{type=MovieSceneFloatSection, channels [None {type=MovieSceneScriptingFloatChannel, keys [[0,1],[2,0],[10,1]]}]}]
      }]
    }
   ]}
 }
 Anim_FocusPulse {
  binding_count=2
  length_seconds=1.5
  object=&self:Anim_FocusPulse
  sequence {bindings [
    OverlaySlot_2 {
     id=#36
     parent_id=#37
     possessed_class=OverlaySlot
     tracks [{
       type=MovieSceneMarginTrack
       display_name=Padding
       sections [{
         type=MovieSceneMarginSection
         channels [
          Left {type=MovieSceneScriptingFloatChannel, keys [[0,0],[20,-16]]}
          Top {type=MovieSceneScriptingFloatChannel, keys [[0,0],[20,-16]]}
          Right {type=MovieSceneScriptingFloatChannel, keys [[0,0],[20,-16]]}
          Bottom {type=MovieSceneScriptingFloatChannel, keys [[0,0],[20,-16]]}
         ]
        }]
      }]
    }
    Image_Outline_Scale {
     id=#37
     possessed_class=Image
     tracks [{
       type=MovieSceneWidgetMaterialTrack
       sections [{
         type=MovieSceneParameterSection
         channels [
          Anim_BorderAlpha {type=MovieSceneScriptingFloatChannel, keys [[0,0],[4,0.125],[20,0]]}
          Border_Scale_X {type=MovieSceneScriptingFloatChannel, keys [[10,0.85],[20,0.9]]}
          Border_Scale_Y {type=MovieSceneScriptingFloatChannel, thinned_from=3, keys [[10,0.85],[30,0.95]]}
         ]
        }]
      }]
    }
   ]}
 }
]
graphs [
 EventGraph {nodes [
   n1 = SetAugmentIcon!() -> (OutputDelegate, then, Asset:Soft Object Reference) {is=title K2Node_CustomEvent}
   > n2 = Async Load Asset(Asset=@n1.Asset)
   n3 = Cast To Texture2D(execute=@n2.Completed, Object=@n2)
   > n4 = GetDynamicMaterial(self=Get Image_AugmentIcon() -> (Image_AugmentIcon:Image Object Reference))
   > SetTextureParameterValue(self=@n4, ParameterName=Icon, Value=@n3)
   Event Destruct()
   n5 = Create Event()
   n6 = Get CommonInputSubsystem()
   > Unbind Event from On Input Method Changed(self=@n6, Delegate=@n5)
   Event Construct()
   > SetMobileVisuals!(execute, self, InputType:ECommonInputType Enum=GetCurrentInputType(self=@n6)) -> (then)
   > Bind Event to On Input Method Changed(self=@n6, Delegate=@n5)
   n7 = Event OnItemNewOrBorrowedSet!() -> (OutputDelegate, then, bIsNew:bool, bIsBorrowed:bool) {is=title K2Node_Event}
   > UpdateNewPrompt!(execute, self, IsNew:bool=@n7.bIsNew) -> (then)
   > UpdateBorrowedPrompt!(execute, self, IsBorrowed:bool=@n7.bIsBorrowed) -> (then)
   n8 = Event OnItemPinColorSet!() -> (OutputDelegate, then, PinColor:Linear Color Structure) {is=title K2Node_Event}
   > n9 = Sequence()
   > n10 = GetDynamicMaterial(self=Get Image_TeammateIcon_Color() -> (Image_TeammateIcon_Color:Image Object Reference))
   > SetVectorParameterValue(self=@n10, ParameterName=Backing_Color, Value=@n8.PinColor)
   n11 = GetDynamicMaterial(execute=@n9.then_1, self=Get Image_TeammateIcon_Color_Glow() -> (Image_TeammateIcon_Color_Glow:Image Object Reference))
   > SetVectorParameterValue(self=@n11, ParameterName=Backing_Color, Value=@n8.PinColor)
  ]}
 PlayHUDSwapAnim {nodes [
   PlayHUDSwapAnim() {is=title K2Node_FunctionEntry}
   > PlayAnimation(InAnimation=Get Anim_Reroll() -> (Anim_Reroll:Widget Animation Object Reference), PlaybackSpeed=1.75)
  ]}
 PlayPulseAnimation {nodes [
   PlayPulseAnimation() {is=title K2Node_FunctionEntry}
   > PlayAnimation(InAnimation=Get Anim_FocusPulse() -> (Anim_FocusPulse:Widget Animation Object Reference), NumLoopsToPlay=0)
   > SetVisibility(self=Get Image_Outline_Scale() -> (Image_Outline_Scale:Image Object Reference), InVisibility=SelfHitTestInvisible)
  ]}
 PlayRerollAnim {nodes [
   PlayRerollAnim() {is=title K2Node_FunctionEntry}
   > PlayAnimation(InAnimation=Get Anim_Reroll())
  ]}
 ResetState {nodes [
   ResetState() {is=title K2Node_FunctionEntry}
   n1 = Get Button_AugmentActivate() -> (Button_AugmentActivate:Overlay Object Reference)
   > SetRenderTransform!(
    execute
    self:Widget Object Reference=@n1
    InTransform_Translation:Vector 2D Structure=(X=0,Y=0)
    InTransform_Scale:Vector 2D Structure=(X=1,Y=1)
    InTransform_Shear:Vector 2D Structure=(X=0,Y=0)
    InTransform_Angle:float=0
   ) -> (then)
   > SetRenderTransform(self=@n1, InTransform_Translation=(X=0,Y=0), InTransform_Scale=(X=1,Y=1), InTransform_Shear=(X=0,Y=0), InTransform_Angle=0)
   > SetRenderTransform(
    self=Get Image_ConfirmationOutline() -> (Image_ConfirmationOutline:Image Object Reference)
    InTransform_Translation=(X=0,Y=0)
    InTransform_Scale=(X=1.25,Y=1.25)
    InTransform_Shear=(X=0,Y=0)
    InTransform_Angle=0
   )
   > StopAnimation(InAnimation=Get Anim_ActivationConfirmation() -> (Anim_ActivationConfirmation:Widget Animation Object Reference))
   n2 = Get Image_CheckMark() -> (Image_CheckMark:Image Object Reference)
   > SetRenderOpacity(self=@n2, InOpacity=0)
   > SetRenderTransform(self=@n2, InTransform_Translation=(X=0,Y=2), InTransform_Scale=(X=0.75,Y=0.75), InTransform_Shear=(X=0,Y=0), InTransform_Angle=0)
  ]}
 SetAugmentInfo {nodes [
   n1 = SetAugmentInfo!() -> (then, InAugmentItemDefinition:Fort Player Augment Item Definition Object Reference) {is=title K2Node_FunctionEntry}
   > n2 = Set Augment Item Definition(Augment Item Definition:Fort Player Augment Item Definition Object Reference=@n1) -> (Output_Get:Fort Player Augment Item Definition Object Reference)
   > UpdateAugment!(
    execute
    self:Fort Player Augment Info Entry Widget Object Reference
    AugmentItemDefinition:Fort Player Augment Item Definition Object Reference=@n2
   ) -> (then)
   > SetAugmentIcon!(
    execute
    self
    Asset:Soft Object Reference=GetHUDEffectImage!(self:Fort Player Augment Item Definition Object Reference=@n2) -> (ReturnValue:Texture 2D Soft Object Reference)
   ) -> (then) {is=title K2Node_CallFunction}
  ]}
 SetFocused {nodes [
   n1 = SetFocused!() -> (then, IsFocused:bool=~, PlayAnimation:bool=~) {is=title K2Node_FunctionEntry}
   n2 = Get Anim_Focused() -> (Anim_Focused:Widget Animation Object Reference)
   > PlayAnimation(
    InAnimation=@n2
    StartAtTime=Select(Option 0:float=GetEndTime(self=@n2), Option 1:float=0, Index=@n1.PlayAnimation) -> (ReturnValue:float)
    PlayMode=Select(Option 0:EUMGSequencePlayMode Enum=Reverse, Option 1:EUMGSequencePlayMode Enum, Index=Get IsFocused() -> (IsFocused:bool)) -> (ReturnValue:EUMGSequencePlayMode Enum)
   )
   > SetVisibility(
    self=Get Button_AugmentActivate()
    InVisibility=Select(Option 0:ESlateVisibility Enum=Collapsed, Option 1:ESlateVisibility Enum=SelfHitTestInvisible, Index=Get IsFocused()) -> (ReturnValue:ESlateVisibility Enum)
   )
   > n3 = Branch(Condition=Get IsFocused())
   SetVisibility(execute=@n3.else, self=Get Image_Outline_Scale(), InVisibility=Collapsed)
  ]}
 SetMobileVisuals {nodes [
   n1 = SetMobileVisuals!() -> (then, InputType:ECommonInputType Enum) {is=title K2Node_FunctionEntry}
   > n2 = Set IsMobileLayout(IsMobileLayout:bool=AND Boolean(A=IsMobileGame(), B="Equal (Enum)"(A=@n1, B=Touch))) -> (Output_Get:bool)
   > SetSize!(
    execute
    self:Spacer Object Reference=Get Spacer_WidthConrol() -> (Spacer_WidthConrol:Spacer Object Reference)
    InSize_X:double=864
    InSize_Y:double=Select(Option 0:double=256, Option 1:double=156, Index=@n2) -> (ReturnValue:double)
   ) -> (then)
   > SetVerticalAlignment(
    self=SlotAsOverlaySlot(Widget=Get Content() -> (Content:Horizontal Box Object Reference))
    InVerticalAlignment=Select(
     Option 0:EVerticalAlignment Enum=VAlign_Fill
     Option 1:EVerticalAlignment Enum=VAlign_Top
     Index=Get IsMobileLayout() -> (IsMobileLayout:bool)
    ) -> (ReturnValue:EVerticalAlignment Enum)
   )
   > SetPadding!(
    execute
    self:Vertical Box Slot Object Reference=SlotAsVerticalBoxSlot(Widget=Get DescriptionTextOverlay() -> (DescriptionTextOverlay:Overlay Object Reference))
    InPadding:Margin Structure=Select(
     Option 0:Margin Structure=Make Margin(Left=0, Top=0, Right=0, Bottom=32)
     Option 1:Margin Structure=Make Margin(Left=0, Top=0, Right=0, Bottom=15)
     Index=Get IsMobileLayout()
    ) -> (ReturnValue:Margin Structure)
   ) -> (then)
   > SetPadding(
    self=SlotAsVerticalBoxSlot(Widget=Get PlatformVisibilityOverlay() -> (PlatformVisibilityOverlay:Overlay Object Reference))
    InPadding=Select(
     Option 0=Make Margin(Left=0, Top=0, Right=48, Bottom=0)
     Option 1=Make Margin(Left=0, Top=0, Right=48, Bottom=15)
     Index=Get IsMobileLayout()
    )
   )
   n3 = Get IsMobileLayout()
   > SetVisibility(
    self=Get PlatformVisibilityOverlay()
    InVisibility=Select(Option 0:ESlateVisibility Enum=Collapsed, Option 1:ESlateVisibility Enum=SelfHitTestInvisible, Index=NOT Boolean(A=@n3)) -> (ReturnValue:ESlateVisibility Enum)
   )
  ]}
 TriggerActivationAnimation {nodes [
   TriggerActivationAnimation() {is=title K2Node_FunctionEntry}
   n1 = Get Anim_ActivationConfirmation()
   > PlayAnimation(InAnimation=@n1)
   > Return Node!(execute, AnimationEndTime:double=GetEndTime(self=@n1))
  ]}
 UpdateBorrowedPrompt {nodes [
   n1 = UpdateBorrowedPrompt!() -> (then, IsBorrowed:bool=~) {is=title K2Node_FunctionEntry}
   > SetVisibility(
    self=Get Overlay_TeammateAugment() -> (Overlay_TeammateAugment:Overlay Object Reference)
    InVisibility=Select(Option 0=Collapsed, Option 1=SelfHitTestInvisible, Index=@n1)
   )
  ]}
 UpdateNewPrompt {nodes [
   n1 = UpdateNewPrompt!() -> (then, IsNew:bool=~) {is=title K2Node_FunctionEntry}
   > SetVisibility(
    self=[
     Get Overlay_NewPrompt() -> (Overlay_NewPrompt:Overlay Object Reference)
     Get Overlay_UnlockedBang() -> (Overlay_UnlockedBang:Overlay Object Reference)
     Get Icon_Border_Unlocked() -> (Icon_Border_Unlocked:Image Object Reference)
     Get Icon_BackingPanel_UnlockedFX() -> (Icon_BackingPanel_UnlockedFX:Image Object Reference)
     Get Icon_Border_UnlockedGlow() -> (Icon_Border_UnlockedGlow:Image Object Reference)
    ]
    InVisibility=Select(Option 0=Collapsed, Option 1=HitTestInvisible, Index=@n1)
   )
  ]}
]
variables [Augment Item Definition {type=FortPlayerAugmentItemDefinition, default=null}]
widgets {
 info {
  inheritedWidgetCount=10
  parentClass=/Script/PlayerAugmentsUI.FortPlayerAugmentInfoEntryWidget
  rootWidgetClass=&Overlay
  widgetCount=76
 }
 widgets [
  All {widgetClassPath=&Overlay, widget=All}
  Spacer_WidthConrol {
   widgetClassPath=&Spacer
   bIsVariable=true
   parent=All
   widget=Spacer_WidthConrol
   properties {size=<864 256>}
   slot {class=OverlaySlot}
  }
  Backing_Vignette {
   widgetClassPath=&Image
   parent=All
   widget=Backing_Vignette
   properties {
    visibility=SelfHitTestInvisible
    brush {
     bIsDynamicallyLoaded=false
     drawAs=Box
     imageType=NoImage
     mirroring=NoMirror
     resourceName=None
     resourceObject=/PlayerAugmentsUI/Materials/MI_UI_PlayerAugment_Panel_Darken
     tiling=NoTile
     imageSize=<640 256>
     margin=<ltrb 0.5 0.5 0.5 0.5>
     outlineSettings {
      bUseBrushTransparency=false
      roundingType=HalfHeightRadius
      width=0
      color {colorUseRule=UseColor_Specified, specifiedColor=<rgba 0 0 0 0>}
      cornerRadii=<0 0 0 0>
     }
     tintColor {colorUseRule=UseColor_Specified, specifiedColor=<rgba 1 1 1 1>}
     uVRegion {bIsValid=false, max=<0 0>, min=<0 0>}
    }
   }
   slot {class=OverlaySlot, properties {horizontalAlignment=HAlign_Fill, verticalAlignment=VAlign_Fill, padding=<ltrb -64 -64 -64 -64>}}
  }
  Backing_Glow {
   widgetClassPath=&Image
   parent=All
   widget=Backing_Glow
   properties {
    visibility=SelfHitTestInvisible
    brush {
     bIsDynamicallyLoaded=false
     drawAs=Box
     imageType=NoImage
     mirroring=NoMirror
     resourceName=None
     resourceObject=/PlayerAugmentsUI/Materials/MI_UI_PlayerAugment_Panel_Glow
     tiling=NoTile
     imageSize=<640 256>
     margin=<ltrb 0.5 0.5 0.5 0.5>
     outlineSettings {
      bUseBrushTransparency=false
      roundingType=HalfHeightRadius
      width=0
      color {colorUseRule=UseColor_Specified, specifiedColor=<rgba 0 0 0 0>}
      cornerRadii=<0 0 0 0>
     }
     tintColor {colorUseRule=UseColor_Specified, specifiedColor=<rgba 1 1 1 1>}
     uVRegion {bIsValid=false, max=<0 0>, min=<0 0>}
    }
   }
   slot {class=OverlaySlot, properties {horizontalAlignment=HAlign_Fill, verticalAlignment=VAlign_Fill, padding=<ltrb -96 -96 -96 -96>}}
  }
  Backing_Panel {
   widgetClassPath=&Image
   parent=All
   widget=Backing_Panel
   properties {
    visibility=SelfHitTestInvisible
    brush {
     bIsDynamicallyLoaded=false
     drawAs=Box
     imageType=NoImage
     mirroring=NoMirror
     resourceName=None
     resourceObject=/PlayerAugmentsUI/Materials/MI_UI_PlayerAugment_BackingPanel
     tiling=NoTile
     imageSize=<64 64>
     margin=<ltrb 0.5 0.5 0.5 0.5>
     outlineSettings {
      bUseBrushTransparency=false
      roundingType=HalfHeightRadius
      width=0
      color {colorUseRule=UseColor_Specified, specifiedColor=<rgba 0 0 0 0>}
      cornerRadii=<0 0 0 0>
     }
     tintColor {colorUseRule=UseColor_Specified, specifiedColor=<rgba 1 1 1 1>}
     uVRegion {bIsValid=false, max=<0 0>, min=<0 0>}
    }
   }
   slot {class=OverlaySlot, properties {horizontalAlignment=HAlign_Fill, verticalAlignment=VAlign_Fill}}
  }
  Backing_EdgeGlow {
   widgetClassPath=&Image
   parent=All
   widget=Backing_EdgeGlow
   properties {
    visibility=SelfHitTestInvisible
    brush {
     bIsDynamicallyLoaded=false
     drawAs=Box
     imageType=NoImage
     mirroring=NoMirror
     resourceName=None
     resourceObject=/PlayerAugmentsUI/Materials/MI_UI_PlayerAugment_Border_Glow
     tiling=NoTile
     imageSize=<128 128>
     margin=<ltrb 0.5 0.5 0.5 0.5>
     outlineSettings {
      bUseBrushTransparency=false
      roundingType=HalfHeightRadius
      width=0
      color {colorUseRule=UseColor_Specified, specifiedColor=<rgba 0 0 0 0>}
      cornerRadii=<0 0 0 0>
     }
     tintColor {colorUseRule=UseColor_Specified, specifiedColor=<rgba 1 1 1 1>}
     uVRegion {bIsValid=false, max=<0 0>, min=<0 0>}
    }
   }
   slot {class=OverlaySlot, properties {horizontalAlignment=HAlign_Fill, verticalAlignment=VAlign_Fill, padding=<ltrb -16 -16 -16 -16>}}
  }
  Content {
   widgetClassPath=&HorizontalBox
   bIsVariable=true
   parent=All
   widget=Content
   slot {class=OverlaySlot, properties {verticalAlignment=VAlign_Fill, padding=<ltrb 104 0 0 0>}}
  }
  Icon {
   widgetClassPath=&Overlay
   parent=Content
   widget=Icon
   slot {class=HorizontalBoxSlot, properties {horizontalAlignment=HAlign_Center, verticalAlignment=VAlign_Top, padding=<ltrb 0 32 0 0>}}
  }
  Spacer_IconSize {widgetClassPath=&Spacer, parent=Icon, widget=Spacer_IconSize, properties {size=<128 128>}, slot {class=OverlaySlot}}
  Icon_BackingPanel_Reflection {
   widgetClassPath=&Image
   parent=Icon
   widget=Icon_BackingPanel_Reflection
   properties {
    visibility=SelfHitTestInvisible
    brush {
     bIsDynamicallyLoaded=false
     drawAs=Box
     imageType=NoImage
     mirroring=NoMirror
     resourceName=None
     resourceObject=/PlayerAugmentsUI/Materials/MI_UI_PlayerAugment_BackingPanel_NoNoise_Reflection
     tiling=NoTile
     imageSize=<64 64>
     margin=<ltrb 0.5 0.5 0.5 0.5>
     outlineSettings {
      bUseBrushTransparency=false
      roundingType=HalfHeightRadius
      width=0
      color {colorUseRule=UseColor_Specified, specifiedColor=<rgba 0 0 0 0>}
      cornerRadii=<0 0 0 0>
     }
     tintColor {colorUseRule=UseColor_Specified, specifiedColor=<rgba 1 1 1 1>}
     uVRegion {bIsValid=false, max=<0 0>, min=<0 0>}
    }
    renderTransform {angle=0, scale=<1 1>, shear=<0 0>, translation=<10 16>}
   }
   slot {class=OverlaySlot, properties {horizontalAlignment=HAlign_Fill, verticalAlignment=VAlign_Fill}}
  }
  Icon_ConfirmationFlare {
   widgetClassPath=&Image
   parent=Icon
   widget=Icon_ConfirmationFlare
   properties {
    visibility=SelfHitTestInvisible
    brush {
     bIsDynamicallyLoaded=false
     drawAs=Image
     imageType=NoImage
     mirroring=NoMirror
     resourceName=None
     resourceObject=/PlayerAugmentsUI/Materials/M_UI_PlayerAugment_Gradient_Flare
     tiling=NoTile
     imageSize=<32 32>
     margin=<ltrb 0 0 0 0>
     outlineSettings {
      bUseBrushTransparency=false
      roundingType=HalfHeightRadius
      width=0
      color {colorUseRule=UseColor_Specified, specifiedColor=<rgba 0 0 0 0>}
      cornerRadii=<0 0 0 0>
     }
     tintColor {colorUseRule=UseColor_Specified, specifiedColor=<rgba 1 1 1 1>}
     uVRegion {bIsValid=false, max=<0 0>, min=<0 0>}
    }
   }
   slot {class=OverlaySlot, properties {horizontalAlignment=HAlign_Fill, verticalAlignment=VAlign_Fill}}
  }
  Icon_BackingPanel {
   widgetClassPath=&Image
   parent=Icon
   widget=Icon_BackingPanel
   properties {
    visibility=SelfHitTestInvisible
    brush {
     bIsDynamicallyLoaded=false
     drawAs=Box
     imageType=NoImage
     mirroring=NoMirror
     resourceName=None
     resourceObject=/PlayerAugmentsUI/Materials/MI_UI_PlayerAugment_BackingPanel_NoNoise_Icon
     tiling=NoTile
     imageSize=<64 64>
     margin=<ltrb 0.5 0.5 0.5 0.5>
     outlineSettings {
      bUseBrushTransparency=false
      roundingType=HalfHeightRadius
      width=0
      color {colorUseRule=UseColor_Specified, specifiedColor=<rgba 0 0 0 0>}
      cornerRadii=<0 0 0 0>
     }
     tintColor {colorUseRule=UseColor_Specified, specifiedColor=<rgba 1 1 1 1>}
     uVRegion {bIsValid=false, max=<0 0>, min=<0 0>}
    }
   }
   slot {class=OverlaySlot, properties {horizontalAlignment=HAlign_Fill, verticalAlignment=VAlign_Fill}}
  }
  Icon_BackingPanel_UnlockedFX {
   widgetClassPath=&Image
   bIsVariable=true
   parent=Icon
   widget=Icon_BackingPanel_UnlockedFX
   properties {
    visibility=SelfHitTestInvisible
    brush {
     bIsDynamicallyLoaded=false
     drawAs=Box
     imageType=NoImage
     mirroring=NoMirror
     resourceName=None
     resourceObject=/PlayerAugmentsUI/Materials/M_UI_AugmentUnlocked_Pattern
     tiling=NoTile
     imageSize=<64 64>
     margin=<ltrb 0.5 0.5 0.5 0.5>
     outlineSettings {
      bUseBrushTransparency=false
      roundingType=HalfHeightRadius
      width=0
      color {colorUseRule=UseColor_Specified, specifiedColor=<rgba 0 0 0 0>}
      cornerRadii=<0 0 0 0>
     }
     tintColor {colorUseRule=UseColor_Specified, specifiedColor=<rgba 1 1 1 1>}
     uVRegion {bIsValid=false, max=<0 0>, min=<0 0>}
    }
   }
   slot {class=OverlaySlot, properties {horizontalAlignment=HAlign_Fill, verticalAlignment=VAlign_Fill}}
  }
  Icon_Border_UnlockedGlow {
   widgetClassPath=&Image
   bIsVariable=true
   parent=Icon
   widget=Icon_Border_UnlockedGlow
   properties {
    visibility=SelfHitTestInvisible
    brush {
     bIsDynamicallyLoaded=false
     drawAs=Box
     imageType=NoImage
     mirroring=NoMirror
     resourceName=None
     resourceObject=/PlayerAugmentsUI/Materials/MI_UI_PlayerAugment_Unlocked_IconBacking_Glow_Border
     tiling=NoTile
     imageSize=<64 64>
     margin=<ltrb 0.5 0.5 0.5 0.5>
     outlineSettings {
      bUseBrushTransparency=false
      roundingType=HalfHeightRadius
      width=0
      color {colorUseRule=UseColor_Specified, specifiedColor=<rgba 0 0 0 0>}
      cornerRadii=<0 0 0 0>
     }
     tintColor {colorUseRule=UseColor_Specified, specifiedColor=<rgba 1 1 1 1>}
     uVRegion {bIsValid=false, max=<0 0>, min=<0 0>}
    }
   }
   slot {class=OverlaySlot, properties {horizontalAlignment=HAlign_Fill, verticalAlignment=VAlign_Fill, padding=<ltrb -16 -16 -16 -16>}}
  }
  Icon_Border_ConfirmationPulse {
   widgetClassPath=&Image
   parent=Icon
   widget=Icon_Border_ConfirmationPulse
   properties {
    visibility=SelfHitTestInvisible
    brush {
     bIsDynamicallyLoaded=false
     drawAs=Box
     imageType=NoImage
     mirroring=NoMirror
     resourceName=None
     resourceObject=/PlayerAugmentsUI/Materials/MI_UI_PlayerAugmment_Border_ConfirmationPulse
     tiling=NoTile
     imageSize=<64 64>
     margin=<ltrb 0.5 0.5 0.5 0.5>
     outlineSettings {
      bUseBrushTransparency=false
      roundingType=HalfHeightRadius
      width=0
      color {colorUseRule=UseColor_Specified, specifiedColor=<rgba 0 0 0 0>}
      cornerRadii=<0 0 0 0>
     }
     tintColor {colorUseRule=UseColor_Specified, specifiedColor=<rgba 1 1 1 1>}
     uVRegion {bIsValid=false, max=<0 0>, min=<0 0>}
    }
   }
   slot {class=OverlaySlot, properties {horizontalAlignment=HAlign_Fill, verticalAlignment=VAlign_Fill}}
  }
  Icon_Border {
   widgetClassPath=&Image
   parent=Icon
   widget=Icon_Border
   properties {
    visibility=SelfHitTestInvisible
    brush {
     bIsDynamicallyLoaded=false
     drawAs=Box
     imageType=NoImage
     mirroring=NoMirror
     resourceName=None
     resourceObject=/PlayerAugmentsUI/Materials/MI_UI_PlayerAugment_Border
     tiling=NoTile
     imageSize=<64 64>
     margin=<ltrb 0.5 0.5 0.5 0.5>
     outlineSettings {
      bUseBrushTransparency=false
      roundingType=HalfHeightRadius
      width=0
      color {colorUseRule=UseColor_Specified, specifiedColor=<rgba 0 0 0 0>}
      cornerRadii=<0 0 0 0>
     }
     tintColor {colorUseRule=UseColor_Specified, specifiedColor=<rgba 1 1 1 1>}
     uVRegion {bIsValid=false, max=<0 0>, min=<0 0>}
    }
    renderTransformPivot=<1 1>
   }
   slot {class=OverlaySlot, properties {horizontalAlignment=HAlign_Fill, verticalAlignment=VAlign_Fill}}
  }
  Icon_Border_Unlocked {
   widgetClassPath=&Image
   bIsVariable=true
   parent=Icon
   widget=Icon_Border_Unlocked
   properties {
    visibility=SelfHitTestInvisible
    brush {
     bIsDynamicallyLoaded=false
     drawAs=Box
     imageType=NoImage
     mirroring=NoMirror
     resourceName=None
     resourceObject=/PlayerAugmentsUI/Materials/MI_UI_PlayerAugment_Border_Unlcoked
     tiling=NoTile
     imageSize=<64 64>
     margin=<ltrb 0.5 0.5 0.5 0.5>
     outlineSettings {
      bUseBrushTransparency=false
      roundingType=HalfHeightRadius
      width=0
      color {colorUseRule=UseColor_Specified, specifiedColor=<rgba 0 0 0 0>}
      cornerRadii=<0 0 0 0>
     }
     tintColor {colorUseRule=UseColor_Specified, specifiedColor=<rgba 1 1 1 1>}
     uVRegion {bIsValid=false, max=<0 0>, min=<0 0>}
    }
    renderTransformPivot=<1 1>
   }
   slot {class=OverlaySlot, properties {horizontalAlignment=HAlign_Fill, verticalAlignment=VAlign_Fill}}
  }
  Icon_Border_ConfirmationHot {
   widgetClassPath=&Image
   parent=Icon
   widget=Icon_Border_ConfirmationHot
   properties {
    visibility=SelfHitTestInvisible
    brush {
     bIsDynamicallyLoaded=false
     drawAs=Box
     imageType=NoImage
     mirroring=NoMirror
     resourceName=None
     resourceObject=/PlayerAugmentsUI/Materials/MI_UI_PlayerAugment_Border_White
     tiling=NoTile
     imageSize=<64 64>
     margin=<ltrb 0.5 0.5 0.5 0.5>
     outlineSettings {
      bUseBrushTransparency=false
      roundingType=HalfHeightRadius
      width=0
      color {colorUseRule=UseColor_Specified, specifiedColor=<rgba 0 0 0 0>}
      cornerRadii=<0 0 0 0>
     }
     tintColor {colorUseRule=UseColor_Specified, specifiedColor=<rgba 1 1 1 1>}
     uVRegion {bIsValid=false, max=<0 0>, min=<0 0>}
    }
    renderTransformPivot=<1 1>
   }
   slot {class=OverlaySlot, properties {horizontalAlignment=HAlign_Fill, verticalAlignment=VAlign_Fill}}
  }
  Image_AugmentIcon {
   widgetClassPath=&Image
   bIsVariable=true
   parent=Icon
   widget=Image_AugmentIcon
   properties {
    visibility=SelfHitTestInvisible
    brush {
     bIsDynamicallyLoaded=false
     drawAs=Image
     imageType=NoImage
     mirroring=NoMirror
     resourceName=None
     resourceObject=/PlayerAugmentsUI/Materials/MI_UI_PlayerAugment_HUD_Icon
     tiling=NoTile
     imageSize=<96 96>
     margin=<ltrb 0 0 0 0>
     outlineSettings {
      bUseBrushTransparency=false
      roundingType=HalfHeightRadius
      width=0
      color {colorUseRule=UseColor_Specified, specifiedColor=<rgba 0 0 0 0>}
      cornerRadii=<0 0 0 0>
     }
     tintColor {colorUseRule=UseColor_Specified, specifiedColor=<rgba 1 1 1 1>}
     uVRegion {bIsValid=false, max=<0 0>, min=<0 0>}
    }
   }
   slot {class=OverlaySlot, properties {horizontalAlignment=HAlign_Center, verticalAlignment=VAlign_Center}}
  }
  Overlay_NewPrompt {
   widgetClassPath=&Overlay
   bIsVariable=true
   parent=Icon
   widget=Overlay_NewPrompt
   properties {renderTransform {angle=0, scale=<1 1>, shear=<0 0>, translation=<0 12>}}
   slot {class=OverlaySlot, properties {horizontalAlignment=HAlign_Center, verticalAlignment=VAlign_Bottom}}
  }
  Image_1 {
   widgetClassPath=&Image
   parent=Overlay_NewPrompt
   widget=Image_1
   properties {
    visibility=SelfHitTestInvisible
    brush {
     bIsDynamicallyLoaded=false
     drawAs=Box
     imageType=NoImage
     mirroring=NoMirror
     resourceName=None
     resourceObject=/PlayerAugmentsUI/Materials/MI_UI_PlayerAugment_Unlocked_IconBacking_Glow
     tiling=NoTile
     imageSize=<48 48>
     margin=<ltrb 0.5 0.5 0.5 0.5>
     outlineSettings {
      bUseBrushTransparency=false
      roundingType=HalfHeightRadius
      width=0
      color {colorUseRule=UseColor_Specified, specifiedColor=<rgba 0 0 0 0>}
      cornerRadii=<0 0 0 0>
     }
     tintColor {colorUseRule=UseColor_Specified, specifiedColor=<rgba 1 1 1 1>}
     uVRegion {bIsValid=false, max=<0 0>, min=<0 0>}
    }
   }
   slot {class=OverlaySlot, properties {horizontalAlignment=HAlign_Fill, verticalAlignment=VAlign_Fill, padding=<ltrb -16 -16 -16 -16>}}
  }
  Image_410 {
   widgetClassPath=&Image
   parent=Overlay_NewPrompt
   widget=Image_410
   properties {
    visibility=SelfHitTestInvisible
    brush {
     bIsDynamicallyLoaded=false
     drawAs=Box
     imageType=NoImage
     mirroring=NoMirror
     resourceName=None
     resourceObject=/PlayerAugmentsUI/Materials/MI_UI_PlayerAugment_NotificationBacking
     tiling=NoTile
     imageSize=<16 16>
     margin=<ltrb 0.5 0.5 0.5 0.5>
     outlineSettings {
      bUseBrushTransparency=false
      roundingType=HalfHeightRadius
      width=0
      color {colorUseRule=UseColor_Specified, specifiedColor=<rgba 0 0 0 0>}
      cornerRadii=<0 0 0 0>
     }
     tintColor {colorUseRule=UseColor_Specified, specifiedColor=<rgba 1 1 1 1>}
     uVRegion {bIsValid=false, max=<0 0>, min=<0 0>}
    }
   }
   slot {class=OverlaySlot, properties {horizontalAlignment=HAlign_Fill, verticalAlignment=VAlign_Fill}}
  }
  Image {
   widgetClassPath=&Image
   parent=Overlay_NewPrompt
   widget=Image
   properties {
    visibility=SelfHitTestInvisible
    brush {
     bIsDynamicallyLoaded=false
     drawAs=Box
     imageType=NoImage
     mirroring=NoMirror
     resourceName=None
     resourceObject=/PlayerAugmentsUI/Materials/MI_UI_PlayerAugment_Notification_Border
     tiling=NoTile
     imageSize=<16 16>
     margin=<ltrb 0.5 0.5 0.5 0.5>
     outlineSettings {
      bUseBrushTransparency=false
      roundingType=HalfHeightRadius
      width=0
      color {colorUseRule=UseColor_Specified, specifiedColor=<rgba 0 0 0 0>}
      cornerRadii=<0 0 0 0>
     }
     tintColor {colorUseRule=UseColor_Specified, specifiedColor=<rgba 1 1 1 1>}
     uVRegion {bIsValid=false, max=<0 0>, min=<0 0>}
    }
   }
   slot {class=OverlaySlot, properties {horizontalAlignment=HAlign_Fill, verticalAlignment=VAlign_Fill}}
  }
  SizeBox_0 {
   widgetClassPath=&SizeBox
   parent=Overlay_NewPrompt
   widget=SizeBox_0
   properties {bOverride_MaxDesiredWidth=true, maxDesiredWidth=112, widthOverride=112}
   slot {class=OverlaySlot, properties {horizontalAlignment=HAlign_Center, padding=<ltrb 8 6 8 6>}}
  }
  ScaleBox_0 {
   widgetClassPath=&ScaleBox
   parent=SizeBox_0
   widget=ScaleBox_0
   properties {stretch=ScaleToFitX, stretchDirection=DownOnly}
   slot {class=SizeBoxSlot}
  }
  Text_NewAugment {
   widgetClassPath=&CommonTextBlock
   parent=ScaleBox_0
   widget=Text_NewAugment
   properties {
    mobileFontSizeMultiplier=1.4
    text="Unlocked!"
    textTransformPolicy=ToUpper
    font {
     bForceMonospaced=false
     bMaterialIsStencil=false
     fontMaterial=/PlayerAugmentsUI/Materials/MI_UI_PlayerAugment_Notification_Small_Font
     fontObject=&BurbankSmall
     letterSpacing=0
     monospacedWidth=1
     size=10
     skewAmount=0
     typefaceFontName=Black
     outlineSettings {
      bApplyOutlineToDropShadows=false
      bMiteredCorners=false
      bSeparateFillAlpha=true
      outlineMaterial=null
      outlineSize=0
      outlineColor=<rgba 0 0 0 0>
     }
    }
    shadowOffset=<0 0>
   }
   slot {class=ScaleBoxSlot}
  }
  Text {widgetClassPath=&VerticalBox, parent=Content, widget=Text, slot {class=HorizontalBoxSlot, properties {padding=<ltrb 24 32 0 0>}}}
  Overlay_0 {
   widgetClassPath=&Overlay
   parent=Text
   widget=Overlay_0
   slot {class=VerticalBoxSlot, properties {horizontalAlignment=HAlign_Left, verticalAlignment=VAlign_Center}}
  }
  Overlay_Header {widgetClassPath=&Overlay, parent=Overlay_0, widget=Overlay_Header, slot {class=OverlaySlot}}
  Text_Glow {
   widgetClassPath=&Image
   parent=Overlay_Header
   widget=Text_Glow
   properties {
    visibility=SelfHitTestInvisible
    brush {
     bIsDynamicallyLoaded=false
     drawAs=Box
     imageType=NoImage
     mirroring=NoMirror
     resourceName=None
     resourceObject=/PlayerAugmentsUI/Materials/MI_UI_PlayerAugment_Font_Glow
     tiling=NoTile
     imageSize=<64 64>
     margin=<ltrb 0.5 0.5 0.5 0.5>
     outlineSettings {
      bUseBrushTransparency=false
      roundingType=HalfHeightRadius
      width=0
      color {colorUseRule=UseColor_Specified, specifiedColor=<rgba 0 0 0 0>}
      cornerRadii=<0 0 0 0>
     }
     tintColor {colorUseRule=UseColor_Specified, specifiedColor=<rgba 1 1 1 1>}
     uVRegion {bIsValid=false, max=<0 0>, min=<0 0>}
    }
   }
   slot {class=OverlaySlot, properties {horizontalAlignment=HAlign_Fill, verticalAlignment=VAlign_Fill, padding=<ltrb -28 -16 -28 -16>}}
  }
  Text_AugmentName {
   widgetClassPath=&CommonTextBlock
   bInherited=true
   parent=Overlay_Header
   widget=Text_AugmentName
   properties {
    lineHeightPercentage=0.9
    mobileFontSizeMultiplier=1.2
    text="[PH] Augment Name"
    textTransformPolicy=ToUpper
    font {
     bForceMonospaced=false
     bMaterialIsStencil=false
     fontMaterial=null
     fontObject=&Burbank
     letterSpacing=50
     monospacedWidth=1
     size=20
     skewAmount=0
     typefaceFontName=BurbankRegular-Black
     outlineSettings {
      bApplyOutlineToDropShadows=false
      bMiteredCorners=false
      bSeparateFillAlpha=true
      outlineMaterial=&M-FontShadow
      outlineSize=3
      outlineColor=<rgba 0.6667 0 1 1>
     }
    }
    margin=<ltrb 0 -2 0 0>
    shadowColorAndOpacity=<rgba 0 0.8333 1 0.125>
    shadowOffset=<6 6>
   }
   slot {class=OverlaySlot}
  }
  Overlay_UnlockedBang {
   widgetClassPath=&Overlay
   bIsVariable=true
   parent=Overlay_0
   widget=Overlay_UnlockedBang
   properties {renderTransform {angle=0, scale=<1 1>, shear=<0 0>, translation=<22 -10>}}
   slot {class=OverlaySlot, properties {horizontalAlignment=HAlign_Right}}
  }
  Spacer_131 {
   widgetClassPath=&Spacer
   parent=Overlay_UnlockedBang
   widget=Spacer_131
   properties {size=<24 24>}
   slot {class=OverlaySlot, properties {horizontalAlignment=HAlign_Fill, verticalAlignment=VAlign_Fill}}
  }
  Image_UnlockedBangGlow {
   widgetClassPath=&Image
   parent=Overlay_UnlockedBang
   widget=Image_UnlockedBangGlow
   properties {
    visibility=SelfHitTestInvisible
    brush {
     bIsDynamicallyLoaded=false
     drawAs=Box
     imageType=NoImage
     mirroring=NoMirror
     resourceName=None
     resourceObject=&MI_UI_PlayerAugment_Unlocked_IconBacking_Glow
     tiling=NoTile
     imageSize=<48 48>
     margin=<ltrb 0.5 0.5 0.5 0.5>
     outlineSettings {
      bUseBrushTransparency=false
      roundingType=HalfHeightRadius
      width=0
      color {colorUseRule=UseColor_Specified, specifiedColor=<rgba 0 0 0 0>}
      cornerRadii=<0 0 0 0>
     }
     tintColor {colorUseRule=UseColor_Specified, specifiedColor=<rgba 1 1 1 1>}
     uVRegion {bIsValid=false, max=<0 0>, min=<0 0>}
    }
   }
   slot {class=OverlaySlot, properties {horizontalAlignment=HAlign_Fill, verticalAlignment=VAlign_Fill, padding=<ltrb -16 -16 -16 -16>}}
  }
  Image_UnlockedBangBacking {
   widgetClassPath=&Image
   parent=Overlay_UnlockedBang
   widget=Image_UnlockedBangBacking
   properties {
    visibility=SelfHitTestInvisible
    brush {
     bIsDynamicallyLoaded=false
     drawAs=Box
     imageType=NoImage
     mirroring=NoMirror
     resourceName=None
     resourceObject=&MI_UI_PlayerAugment_NotificationBacking
     tiling=NoTile
     imageSize=<16 16>
     margin=<ltrb 0.5 0.5 0.5 0.5>
     outlineSettings {
      bUseBrushTransparency=false
      roundingType=HalfHeightRadius
      width=0
      color {colorUseRule=UseColor_Specified, specifiedColor=<rgba 0 0 0 0>}
      cornerRadii=<0 0 0 0>
     }
     tintColor {colorUseRule=UseColor_Specified, specifiedColor=<rgba 1 1 1 1>}
     uVRegion {bIsValid=false, max=<0 0>, min=<0 0>}
    }
   }
   slot {class=OverlaySlot, properties {horizontalAlignment=HAlign_Fill, verticalAlignment=VAlign_Fill}}
  }
  Image_UnlockedBangOutline {
   widgetClassPath=&Image
   parent=Overlay_UnlockedBang
   widget=Image_UnlockedBangOutline
   properties {
    visibility=SelfHitTestInvisible
    brush {
     bIsDynamicallyLoaded=false
     drawAs=Box
     imageType=NoImage
     mirroring=NoMirror
     resourceName=None
     resourceObject=&MI_UI_PlayerAugment_Notification_Border
     tiling=NoTile
     imageSize=<16 16>
     margin=<ltrb 0.5 0.5 0.5 0.5>
     outlineSettings {
      bUseBrushTransparency=false
      roundingType=HalfHeightRadius
      width=0
      color {colorUseRule=UseColor_Specified, specifiedColor=<rgba 0 0 0 0>}
      cornerRadii=<0 0 0 0>
     }
     tintColor {colorUseRule=UseColor_Specified, specifiedColor=<rgba 1 1 1 1>}
     uVRegion {bIsValid=false, max=<0 0>, min=<0 0>}
    }
   }
   slot {class=OverlaySlot, properties {horizontalAlignment=HAlign_Fill, verticalAlignment=VAlign_Fill}}
  }
  Text_NewAugment_1 {
   widgetClassPath=&CommonTextBlock
   parent=Overlay_UnlockedBang
   widget=Text_NewAugment_1
   properties {
    text="!"
    textTransformPolicy=ToUpper
    font {
     bForceMonospaced=false
     bMaterialIsStencil=false
     fontMaterial=&MI_UI_PlayerAugment_Notification_Small_Font
     fontObject=&BurbankSmall
     letterSpacing=0
     monospacedWidth=1
     size=13
     skewAmount=0
     typefaceFontName=Black
     outlineSettings {
      bApplyOutlineToDropShadows=false
      bMiteredCorners=false
      bSeparateFillAlpha=true
      outlineMaterial=null
      outlineSize=0
      outlineColor=<rgba 0 0 0 0>
     }
    }
    shadowOffset=<0 0>
   }
   slot {class=OverlaySlot, properties {horizontalAlignment=HAlign_Center, verticalAlignment=VAlign_Center}}
  }
  Overlay_TeammateAugment {
   widgetClassPath=&Overlay
   bIsVariable=true
   parent=Text
   widget=Overlay_TeammateAugment
   properties {visibility=Collapsed}
   slot {class=VerticalBoxSlot, properties {padding=<ltrb 0 0 0 2>}}
  }
  HorizontalBox_0 {
   widgetClassPath=&HorizontalBox
   parent=Overlay_TeammateAugment
   widget=HorizontalBox_0
   properties {flowDirectionPreference=Culture}
   slot {class=OverlaySlot}
  }
  Overlay_5 {
   widgetClassPath=&Overlay
   parent=HorizontalBox_0
   widget=Overlay_5
   slot {class=HorizontalBoxSlot, properties {horizontalAlignment=HAlign_Left, verticalAlignment=VAlign_Center}}
  }
  Text_Glow_1 {
   widgetClassPath=&Image
   parent=Overlay_5
   widget=Text_Glow_1
   properties {
    visibility=SelfHitTestInvisible
    brush {
     bIsDynamicallyLoaded=false
     drawAs=Box
     imageType=NoImage
     mirroring=NoMirror
     resourceName=None
     resourceObject=&MI_UI_PlayerAugment_Font_Glow
     tiling=NoTile
     imageSize=<40 40>
     margin=<ltrb 0.5 0.5 0.5 0.5>
     outlineSettings {
      bUseBrushTransparency=false
      roundingType=HalfHeightRadius
      width=0
      color {colorUseRule=UseColor_Specified, specifiedColor=<rgba 0 0 0 0>}
      cornerRadii=<0 0 0 0>
     }
     tintColor {colorUseRule=UseColor_Specified, specifiedColor=<rgba 1 1 1 1>}
     uVRegion {bIsValid=false, max=<0 0>, min=<0 0>}
    }
   }
   slot {class=OverlaySlot, properties {horizontalAlignment=HAlign_Fill, verticalAlignment=VAlign_Fill, padding=<ltrb -8 -6 -8 -6>}}
  }
  Text_Borrow {
   widgetClassPath=&CommonTextBlock
   parent=Overlay_5
   widget=Text_Borrow
   properties {
    mobileFontSizeMultiplier=1.3
    text=Borrowed from
    textTransformPolicy=ToUpper
    font {
     bForceMonospaced=false
     bMaterialIsStencil=false
     fontMaterial=/PlayerAugmentsUI/Materials/MI_UI_PlayerAugment_SocialCallout_Font
     fontObject=&BurbankSmall
     letterSpacing=0
     monospacedWidth=1
     size=13
     skewAmount=0
     typefaceFontName=Bold
     outlineSettings {
      bApplyOutlineToDropShadows=false
      bMiteredCorners=false
      bSeparateFillAlpha=true
      outlineMaterial=&M-FontShadow
      outlineSize=2
      outlineColor=<rgba 0.6667 0 1 1>
     }
    }
    shadowColorAndOpacity=<rgba 0 0.8333 1 0.125>
    shadowOffset=<6 4>
   }
   slot {class=OverlaySlot}
  }
  Overlay_LenderAvatarIcon {
   widgetClassPath=&Overlay
   bInherited=true
   parent=HorizontalBox_0
   widget=Overlay_LenderAvatarIcon
   slot {class=HorizontalBoxSlot, properties {padding=<ltrb 0 0 5 0>}}
  }
  Image_TeammateIcon_Color_Glow {
   widgetClassPath=&Image
   bIsVariable=true
   parent=Overlay_LenderAvatarIcon
   widget=Image_TeammateIcon_Color_Glow
   properties {
    visibility=SelfHitTestInvisible
    brush {
     bIsDynamicallyLoaded=false
     drawAs=Image
     imageType=NoImage
     mirroring=NoMirror
     resourceName=None
     resourceObject=/PlayerAugmentsUI/Materials/MI_UI_Augment_SocialColorBacking_Glow
     tiling=NoTile
     imageSize=<24 24>
     margin=<ltrb 0 0 0 0>
     outlineSettings {
      bUseBrushTransparency=false
      roundingType=HalfHeightRadius
      width=0
      color {colorUseRule=UseColor_Specified, specifiedColor=<rgba 0 0 0 0>}
      cornerRadii=<0 0 0 0>
     }
     tintColor {colorUseRule=UseColor_Specified, specifiedColor=<rgba 1 1 1 1>}
     uVRegion {bIsValid=false, max=<0 0>, min=<0 0>}
    }
    renderTransform {angle=0, scale=<2 2>, shear=<0 0>, translation=<0 -2>}
   }
   slot {class=OverlaySlot, properties {horizontalAlignment=HAlign_Fill, verticalAlignment=VAlign_Fill}}
  }
  Image_TeammateIcon_Color {
   widgetClassPath=&Image
   bIsVariable=true
   parent=Overlay_LenderAvatarIcon
   widget=Image_TeammateIcon_Color
   properties {
    visibility=SelfHitTestInvisible
    brush {
     bIsDynamicallyLoaded=false
     drawAs=Image
     imageType=NoImage
     mirroring=NoMirror
     resourceName=None
     resourceObject=/PlayerAugmentsUI/Materials/MI_UI_Augment_SocialColorBacking
     tiling=NoTile
     imageSize=<24 24>
     margin=<ltrb 0 0 0 0>
     outlineSettings {
      bUseBrushTransparency=false
      roundingType=HalfHeightRadius
      width=0
      color {colorUseRule=UseColor_Specified, specifiedColor=<rgba 0 0 0 0>}
      cornerRadii=<0 0 0 0>
     }
     tintColor {colorUseRule=UseColor_Specified, specifiedColor=<rgba 1 1 1 1>}
     uVRegion {bIsValid=false, max=<0 0>, min=<0 0>}
    }
    renderTransform {angle=0, scale=<1.125 1.125>, shear=<0 0>, translation=<0 -2>}
   }
   slot {class=OverlaySlot, properties {horizontalAlignment=HAlign_Fill, verticalAlignment=VAlign_Fill}}
  }
  Icon_LenderAvatarIcon {
   widgetClassPath=/Script/FortniteUI.FortSocialAvatarIcon
   bInherited=true
   parent=Overlay_LenderAvatarIcon
   widget=Icon_LenderAvatarIcon
   properties {
    brush {
     bIsDynamicallyLoaded=false
     drawAs=Image
     imageType=NoImage
     mirroring=NoMirror
     resourceName=None
     resourceObject=/Game/UI/SocialPanel/Art/Materials/UI_UserPortrait_Face_NoBackground
     tiling=NoTile
     imageSize=<32 32>
     margin=<ltrb 0 0 0 0>
     outlineSettings {
      bUseBrushTransparency=false
      roundingType=HalfHeightRadius
      width=0
      color {colorUseRule=UseColor_Specified, specifiedColor=<rgba 0 0 0 0>}
      cornerRadii=<0 0 0 0>
     }
     tintColor {colorUseRule=UseColor_Specified, specifiedColor=<rgba 1 1 1 1>}
     uVRegion {bIsValid=false, max=<0 0>, min=<0 0>}
    }
    renderTransform {angle=0, scale=<1 1>, shear=<0 0>, translation=<0 -2>}
   }
   slot {class=OverlaySlot, properties {horizontalAlignment=HAlign_Center, verticalAlignment=VAlign_Center}}
  }
  Overlay_6 {
   widgetClassPath=&Overlay
   parent=HorizontalBox_0
   widget=Overlay_6
   slot {class=HorizontalBoxSlot, properties {horizontalAlignment=HAlign_Left, verticalAlignment=VAlign_Center}}
  }
  Text_Glow_2 {
   widgetClassPath=&Image
   parent=Overlay_6
   widget=Text_Glow_2
   properties {
    visibility=SelfHitTestInvisible
    brush {
     bIsDynamicallyLoaded=false
     drawAs=Box
     imageType=NoImage
     mirroring=NoMirror
     resourceName=None
     resourceObject=&MI_UI_PlayerAugment_Font_Glow
     tiling=NoTile
     imageSize=<40 40>
     margin=<ltrb 0.5 0.5 0.5 0.5>
     outlineSettings {
      bUseBrushTransparency=false
      roundingType=HalfHeightRadius
      width=0
      color {colorUseRule=UseColor_Specified, specifiedColor=<rgba 0 0 0 0>}
      cornerRadii=<0 0 0 0>
     }
     tintColor {colorUseRule=UseColor_Specified, specifiedColor=<rgba 1 1 1 1>}
     uVRegion {bIsValid=false, max=<0 0>, min=<0 0>}
    }
   }
   slot {class=OverlaySlot, properties {horizontalAlignment=HAlign_Fill, verticalAlignment=VAlign_Fill, padding=<ltrb -8 -6 -8 -6>}}
  }
  Text_BorrowingText {
   widgetClassPath=&CommonTextBlock
   bInherited=true
   parent=Overlay_6
   widget=Text_BorrowingText
   properties {
    mobileFontSizeMultiplier=1.3
    text="[PH] PlayerName"
    textTransformPolicy=ToUpper
    font {
     bForceMonospaced=false
     bMaterialIsStencil=false
     fontMaterial=&MI_UI_PlayerAugment_SocialCallout_Font
     fontObject=&BurbankSmall
     letterSpacing=0
     monospacedWidth=1
     size=13
     skewAmount=0
     typefaceFontName=Bold
     outlineSettings {
      bApplyOutlineToDropShadows=false
      bMiteredCorners=false
      bSeparateFillAlpha=true
      outlineMaterial=&M-FontShadow
      outlineSize=3
      outlineColor=<rgba 0.6667 0 1 1>
     }
    }
    shadowColorAndOpacity=<rgba 0 0.8333 1 0.125>
    shadowOffset=<6 4>
   }
   slot {class=OverlaySlot}
  }
  DescriptionTextOverlay {
   widgetClassPath=&Overlay
   bIsVariable=true
   parent=Text
   widget=DescriptionTextOverlay
   slot {class=VerticalBoxSlot, properties {horizontalAlignment=HAlign_Left, verticalAlignment=VAlign_Top, padding=<ltrb 0 0 0 32>}}
  }
  Spacer_DescriptionTextSize {
   widgetClassPath=&Spacer
   parent=DescriptionTextOverlay
   widget=Spacer_DescriptionTextSize
   properties {size=<448 1>}
   slot {class=OverlaySlot, properties {horizontalAlignment=HAlign_Fill, verticalAlignment=VAlign_Fill}}
  }
  Overlay_1 {widgetClassPath=&Overlay, parent=DescriptionTextOverlay, widget=Overlay_1, slot {class=OverlaySlot}}
  Image_SolidBG_AugmentDescription_Reflection {
   widgetClassPath=&Image
   parent=Overlay_1
   widget=Image_SolidBG_AugmentDescription_Reflection
   properties {
    visibility=SelfHitTestInvisible
    brush {
     bIsDynamicallyLoaded=false
     drawAs=Box
     imageType=NoImage
     mirroring=NoMirror
     resourceName=None
     resourceObject=/PlayerAugmentsUI/Materials/MI_UI_PlayerAugment_BackingPanel_NoNoise_Reflection_TextPanel
     tiling=NoTile
     imageSize=<32 32>
     margin=<ltrb 0.5 0.5 0.5 0.5>
     outlineSettings {
      bUseBrushTransparency=false
      roundingType=HalfHeightRadius
      width=0
      color {colorUseRule=UseColor_Specified, specifiedColor=<rgba 0 0 0 0>}
      cornerRadii=<0 0 0 0>
     }
     tintColor {colorUseRule=UseColor_Specified, specifiedColor=<rgba 1 1 1 1>}
     uVRegion {bIsValid=false, max=<0 0>, min=<0 0>}
    }
    renderTransform {angle=0, scale=<1 1>, shear=<0 0>, translation=<8 12>}
   }
   slot {class=OverlaySlot, properties {horizontalAlignment=HAlign_Fill, verticalAlignment=VAlign_Fill}}
  }
  Image_SolidBG_AugmentDescription {
   widgetClassPath=&Image
   parent=Overlay_1
   widget=Image_SolidBG_AugmentDescription
   properties {
    visibility=SelfHitTestInvisible
    brush {
     bIsDynamicallyLoaded=false
     drawAs=Box
     imageType=NoImage
     mirroring=NoMirror
     resourceName=None
     resourceObject=/PlayerAugmentsUI/Materials/MI_UI_PlayerAugment_BackingPanel_NoNoise
     tiling=NoTile
     imageSize=<32 32>
     margin=<ltrb 0.5 0.5 0.5 0.5>
     outlineSettings {
      bUseBrushTransparency=false
      roundingType=HalfHeightRadius
      width=0
      color {colorUseRule=UseColor_Specified, specifiedColor=<rgba 0 0 0 0>}
      cornerRadii=<0 0 0 0>
     }
     tintColor {colorUseRule=UseColor_Specified, specifiedColor=<rgba 1 1 1 1>}
     uVRegion {bIsValid=false, max=<0 0>, min=<0 0>}
    }
   }
   slot {class=OverlaySlot, properties {horizontalAlignment=HAlign_Fill, verticalAlignment=VAlign_Fill}}
  }
  Image_TextPanelBrackets {
   widgetClassPath=&Image
   parent=Overlay_1
   widget=Image_TextPanelBrackets
   properties {
    visibility=SelfHitTestInvisible
    brush {
     bIsDynamicallyLoaded=false
     drawAs=Box
     imageType=NoImage
     mirroring=NoMirror
     resourceName=None
     resourceObject=/PlayerAugmentsUI/Materials/MI_UI_PlayerAugment_Brackets
     tiling=NoTile
     imageSize=<32 32>
     margin=<ltrb 0.5 0.5 0.5 0.5>
     outlineSettings {
      bUseBrushTransparency=false
      roundingType=HalfHeightRadius
      width=0
      color {colorUseRule=UseColor_Specified, specifiedColor=<rgba 0 0 0 0>}
      cornerRadii=<0 0 0 0>
     }
     tintColor {colorUseRule=UseColor_Specified, specifiedColor=<rgba 1 1 1 1>}
     uVRegion {bIsValid=false, max=<0 0>, min=<0 0>}
    }
   }
   slot {class=OverlaySlot, properties {horizontalAlignment=HAlign_Fill, verticalAlignment=VAlign_Fill}}
  }
  Text_AugmentDescription {
   widgetClassPath=&CommonTextBlock
   bInherited=true
   parent=Overlay_1
   widget=Text_AugmentDescription
   properties {
    lineHeightPercentage=0.9
    mobileFontSizeMultiplier=1.45
    text="[PH] Augment Description."
    wrapTextAt=432
    font {
     bForceMonospaced=false
     bMaterialIsStencil=false
     fontMaterial=/PlayerAugmentsUI/Materials/MI_UI_PlayerAugment_Body_Small_Font
     fontObject=&BurbankSmall
     letterSpacing=0
     monospacedWidth=1
     size=13
     skewAmount=0
     typefaceFontName=Bold
     outlineSettings {
      bApplyOutlineToDropShadows=false
      bMiteredCorners=false
      bSeparateFillAlpha=true
      outlineMaterial=&M-FontShadow
      outlineSize=2
      outlineColor=<rgba 0.127 0 0.648 1>
     }
    }
    shadowColorAndOpacity=<rgba 0 0.8333 1 0.125>
    shadowOffset=<6 4>
   }
   slot {class=OverlaySlot, properties {horizontalAlignment=HAlign_Fill, padding=<ltrb 12 8 12 8>}}
  }
  PlatformVisibilityOverlay {
   widgetClassPath=&Overlay
   bIsVariable=true
   parent=Text
   widget=PlatformVisibilityOverlay
   slot {class=VerticalBoxSlot, properties {horizontalAlignment=HAlign_Center, verticalAlignment=VAlign_Bottom, padding=<ltrb 0 0 48 0>}}
  }
  Button_AugmentActivate {
   widgetClassPath=&Overlay
   bIsVariable=true
   parent=PlatformVisibilityOverlay
   widget=Button_AugmentActivate
   slot {class=OverlaySlot}
  }
  Image_Button_Glow {
   widgetClassPath=&Image
   parent=Button_AugmentActivate
   widget=Image_Button_Glow
   properties {
    visibility=SelfHitTestInvisible
    brush {
     bIsDynamicallyLoaded=false
     drawAs=Box
     imageType=NoImage
     mirroring=NoMirror
     resourceName=None
     resourceObject=/PlayerAugmentsUI/Materials/MI_UI_PlayerAugment_Button_Glow
     tiling=NoTile
     imageSize=<104 104>
     margin=<ltrb 0.5 0.5 0.5 0.5>
     outlineSettings {
      bUseBrushTransparency=false
      roundingType=HalfHeightRadius
      width=0
      color {colorUseRule=UseColor_Specified, specifiedColor=<rgba 0 0 0 0>}
      cornerRadii=<0 0 0 0>
     }
     tintColor {colorUseRule=UseColor_Specified, specifiedColor=<rgba 1 1 1 1>}
     uVRegion {bIsValid=false, max=<0 0>, min=<0 0>}
    }
   }
   slot {class=OverlaySlot, properties {horizontalAlignment=HAlign_Fill, verticalAlignment=VAlign_Fill, padding=<ltrb -32 -32 -32 -32>}}
  }
  Image_Activate_SolidBG {
   widgetClassPath=&Image
   parent=Button_AugmentActivate
   widget=Image_Activate_SolidBG
   properties {
    visibility=SelfHitTestInvisible
    brush {
     bIsDynamicallyLoaded=false
     drawAs=Box
     imageType=NoImage
     mirroring=NoMirror
     resourceName=None
     resourceObject=/PlayerAugmentsUI/Materials/MI_UI_PlayerAugment_ActivationButton_Backing
     tiling=NoTile
     imageSize=<48 48>
     margin=<ltrb 0.5 0.5 0.5 0.5>
     outlineSettings {
      bUseBrushTransparency=false
      roundingType=HalfHeightRadius
      width=0
      color {colorUseRule=UseColor_Specified, specifiedColor=<rgba 0 0 0 0>}
      cornerRadii=<0 0 0 0>
     }
     tintColor {colorUseRule=UseColor_Specified, specifiedColor=<rgba 1 1 1 1>}
     uVRegion {bIsValid=false, max=<0 0>, min=<0 0>}
    }
   }
   slot {class=OverlaySlot, properties {horizontalAlignment=HAlign_Fill, verticalAlignment=VAlign_Fill}}
  }
  Image_Outline_Scale {
   widgetClassPath=&Image
   bIsVariable=true
   parent=Button_AugmentActivate
   widget=Image_Outline_Scale
   properties {
    visibility=SelfHitTestInvisible
    brush {
     bIsDynamicallyLoaded=false
     drawAs=Box
     imageType=NoImage
     mirroring=NoMirror
     resourceName=None
     resourceObject=/PlayerAugmentsUI/Materials/MI_UI_PlayerAugment_Button_Pulse
     tiling=NoTile
     imageSize=<48 48>
     margin=<ltrb 0.5 0.5 0.5 0.5>
     outlineSettings {
      bUseBrushTransparency=false
      roundingType=HalfHeightRadius
      width=0
      color {colorUseRule=UseColor_Specified, specifiedColor=<rgba 0 0 0 0>}
      cornerRadii=<0 0 0 0>
     }
     tintColor {colorUseRule=UseColor_Specified, specifiedColor=<rgba 1 1 1 1>}
     uVRegion {bIsValid=false, max=<0 0>, min=<0 0>}
    }
   }
   slot {class=OverlaySlot, properties {horizontalAlignment=HAlign_Fill, verticalAlignment=VAlign_Fill}}
  }
  HorizontalBox_119 {
   widgetClassPath=&HorizontalBox
   parent=Button_AugmentActivate
   widget=HorizontalBox_119
   slot {class=OverlaySlot, properties {horizontalAlignment=HAlign_Center, verticalAlignment=VAlign_Center, padding=<ltrb 20 8 20 8>}}
  }
  border_pc-console-only {
   widgetClassPath=&CommonHardwareVisibilityBorder
   parent=HorizontalBox_119
   widget=border_pc-console-only
   properties {
    horizontalAlignment=HAlign_Center
    verticalAlignment=VAlign_Center
    visibilityQuery {
     autoDescription=" ANY( Input.Gamepad, Input.MouseAndKeyboard )"
     tokenStreamVersion=0
     userDescription=pc-console-only
     queryTokenStream [0,1,1,2,0,1]
     tagDictionary [{tagName=Input.Gamepad}, {tagName=Input.MouseAndKeyboard}]
    }
   }
   slot {class=HorizontalBoxSlot, properties {horizontalAlignment=HAlign_Center, verticalAlignment=VAlign_Bottom}}
  }
  SizeBox {
   widgetClassPath=&SizeBox
   parent=border_pc-console-only
   widget=SizeBox
   properties {bOverride_HeightOverride=true, bOverride_WidthOverride=true, heightOverride=32, widthOverride=32}
   slot {class=BorderSlot, properties {horizontalAlignment=HAlign_Center, verticalAlignment=VAlign_Center, padding=<ltrb 0 0 0 0>}}
  }
  Keybind_ActivateAction {
   widgetClassPath=/FortUILibrary/Input/KeybindWidget
   bIsVariable=true
   parent=SizeBox
   widget=Keybind_ActivateAction
   properties {boundAction=Fire}
   slot {class=SizeBoxSlot}
  }
  Text_Activate {
   widgetClassPath=&CommonTextBlock
   parent=HorizontalBox_119
   widget=Text_Activate
   properties {
    text=Activate
    textTransformPolicy=ToUpper
    font {
     bForceMonospaced=false
     bMaterialIsStencil=false
     fontMaterial=/PlayerAugmentsUI/Materials/MI_UI_PlayerAugment_Button_Font
     fontObject=&Burbank
     letterSpacing=400
     monospacedWidth=1
     size=15
     skewAmount=0
     typefaceFontName=BurbankRegular-Black
     outlineSettings {
      bApplyOutlineToDropShadows=false
      bMiteredCorners=false
      bSeparateFillAlpha=false
      outlineMaterial=null
      outlineSize=2
      outlineColor=<rgba 0 0 0 0.25>
     }
    }
   }
   slot {class=HorizontalBoxSlot, properties {horizontalAlignment=HAlign_Left, verticalAlignment=VAlign_Center, padding=<ltrb 8 0 0 0>}}
  }
  Image_GreenOverlay {
   widgetClassPath=&Image
   bIsVariable=true
   parent=Button_AugmentActivate
   widget=Image_GreenOverlay
   properties {
    visibility=SelfHitTestInvisible
    brush {
     bIsDynamicallyLoaded=false
     drawAs=Box
     imageType=NoImage
     mirroring=NoMirror
     resourceName=None
     resourceObject=/PlayerAugmentsUI/Materials/MI_UI_PlayerAugment_ActivationButton_Confirm
     tiling=NoTile
     imageSize=<48 48>
     margin=<ltrb 0.5 0.5 0.5 0.5>
     outlineSettings {
      bUseBrushTransparency=false
      roundingType=HalfHeightRadius
      width=0
      color {colorUseRule=UseColor_Specified, specifiedColor=<rgba 0 0 0 0>}
      cornerRadii=<0 0 0 0>
     }
     tintColor {colorUseRule=UseColor_Specified, specifiedColor=<rgba 1 1 1 1>}
     uVRegion {bIsValid=false, max=<0 0>, min=<0 0>}
    }
   }
   slot {class=OverlaySlot, properties {horizontalAlignment=HAlign_Fill, verticalAlignment=VAlign_Fill}}
  }
  Image_CheckMark {
   widgetClassPath=&Image
   bIsVariable=true
   parent=Button_AugmentActivate
   widget=Image_CheckMark
   properties {
    renderOpacity=0
    visibility=SelfHitTestInvisible
    brush {
     bIsDynamicallyLoaded=false
     drawAs=Image
     imageType=NoImage
     mirroring=NoMirror
     resourceName=None
     resourceObject=/Game/UI/Foundation/Textures/Icons/Checks/T-Icon-Check-32
     tiling=NoTile
     imageSize=<32 32>
     margin=<ltrb 0 0 0 0>
     outlineSettings {
      bUseBrushTransparency=false
      roundingType=HalfHeightRadius
      width=0
      color {colorUseRule=UseColor_Specified, specifiedColor=<rgba 0 0 0 0>}
      cornerRadii=<0 0 0 0>
     }
     tintColor {colorUseRule=UseColor_Specified, specifiedColor=<rgba 1 1 1 1>}
     uVRegion {bIsValid=false, max=<0 0>, min=<0 0>}
    }
    colorAndOpacity=<rgba 0 0.0331 0.0497 0.5>
   }
   slot {class=OverlaySlot, properties {horizontalAlignment=HAlign_Center, verticalAlignment=VAlign_Center}}
  }
  Image_ConfirmationOutline {
   widgetClassPath=&Image
   bIsVariable=true
   parent=Button_AugmentActivate
   widget=Image_ConfirmationOutline
   properties {
    visibility=SelfHitTestInvisible
    brush {
     bIsDynamicallyLoaded=false
     drawAs=Box
     imageType=NoImage
     mirroring=NoMirror
     resourceName=None
     resourceObject=/PlayerAugmentsUI/Materials/MI_UI_PlayerAugment_Border_White_Confirm
     tiling=NoTile
     imageSize=<48 48>
     margin=<ltrb 0.5 0.5 0.5 0.5>
     outlineSettings {
      bUseBrushTransparency=false
      roundingType=HalfHeightRadius
      width=0
      color {colorUseRule=UseColor_Specified, specifiedColor=<rgba 0 0 0 0>}
      cornerRadii=<0 0 0 0>
     }
     tintColor {colorUseRule=UseColor_Specified, specifiedColor=<rgba 1 1 1 1>}
     uVRegion {bIsValid=false, max=<0 0>, min=<0 0>}
    }
   }
   slot {class=OverlaySlot, properties {horizontalAlignment=HAlign_Fill, verticalAlignment=VAlign_Fill, padding=<ltrb -11 -11 -11 -11>}}
  }
  HardwareVisibilityBorder_TouchOnlyActivate {
   widgetClassPath=&CommonHardwareVisibilityBorder
   parent=Button_AugmentActivate
   widget=HardwareVisibilityBorder_TouchOnlyActivate
   properties {
    visibleType=Visible
    visibilityQuery {
     autoDescription=" ANY( Input.Touch )"
     tokenStreamVersion=0
     userDescription=touch-only
     queryTokenStream [0,1,1,1,0]
     tagDictionary [{tagName=Input.Touch}]
    }
   }
   slot {class=OverlaySlot, properties {horizontalAlignment=HAlign_Fill, verticalAlignment=VAlign_Fill}}
  }
  FortTouchUserWidget {
   widgetClassPath=/Script/FortniteUI.FortTouchUserWidget
   bIsVariable=true
   parent=Button_AugmentActivate
   widget=FortTouchUserWidget
   properties {visibility=Visible, actionNames [TouchActivateAugment]}
   slot {class=OverlaySlot, properties {horizontalAlignment=HAlign_Fill, verticalAlignment=VAlign_Fill}}
  }
  Text_AugmentRarity {widgetClassPath=&CommonTextBlock, bInherited=true, widget=null}
  Text_AugmentCategory {widgetClassPath=&CommonTextBlock, bInherited=true, widget=null}
  ItemCategoryIndicator {widgetClassPath=/Script/FortniteUI.FortItemCategoryIndicator, bInherited=true, widget=null}
  Image_CategoryAsImage {widgetClassPath=&FortLazyImage, bInherited=true, widget=null}
  Image_LenderSquadColor {widgetClassPath=&FortLazyImage, bInherited=true, widget=null}
 ]
}
