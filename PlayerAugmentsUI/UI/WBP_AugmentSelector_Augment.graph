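# Widget for one augment entry in the augment selector, wrapping an AugmentInfo child, optional top/bottom shadows and a touch-only info widget.
# Focus changes play sounds on highlight and event-repeater animations, with functions that curve-lerp the SizeBox max height to expand or shrink the entry.
# It forwards info, pulse, reroll, activation and HUD-swap calls to the child, adding its own reroll and activation animations.
# ResetState unfocuses and clears the child's outline/overlay materials and scale.
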WidgetBlueprint /PlayerAugmentsUI/UI/WBP_AugmentSelector_Augment
animations [
 Anim_Focus {
  binding_count=0
  length_seconds=0.1333
  object=&self:Anim_Focus
  sequence {
   bindings []
   display_rate {numerator=60}
   tracks [{type=MovieSceneEventTrack, sections [{type=MovieSceneEventRepeaterSection, end_frame=8, start_frame=0}]}]
  }
 }
 Anim_Activate {
  binding_count=1
  length_seconds=1
  object=&self:Anim_Activate
  sequence {
   bindings [Overlay_Main {
     id=#1
     possessed_class=Overlay
     tracks [{
       type=MovieScene2DTransformTrack
       sections [{
         type=MovieScene2DTransformSection
         channels [
          Scale.X {type=MovieSceneScriptingFloatChannel, keys [[0,1],[5,0.975],[12,1.025],[30,1]]}
          Scale.Y {type=MovieSceneScriptingFloatChannel, keys [[0,1],[5,0.975],[12,1.025],[30,1],[60,1]]}
         ]
        }]
      }]
    }]
   display_rate {numerator=60}
  }
 }
 Anim_Reroll {
  binding_count=3
  length_seconds=0.5
  object=&self:Anim_Reroll
  sequence {
   bindings [
    Image_BottomShadow {
     id=#2
     possessed_class=Image
     tracks [{
       type=MovieSceneWidgetMaterialTrack
       sections [{type=MovieSceneParameterSection, channels [Anim_BackingAlpha {type=MovieSceneScriptingFloatChannel, keys [[0,0],[6,1],[30,0]]}]}]
      }]
    }
    Image_TopShadow {
     id=#3
     possessed_class=Image
     tracks [{
       type=MovieSceneWidgetMaterialTrack
       sections [{type=MovieSceneParameterSection, channels [Anim_BackingAlpha {type=MovieSceneScriptingFloatChannel, keys [[0,0],[6,1],[30,0]]}]}]
      }]
    }
    WBP_AugmentSelector_AugmentInfo {
     id=#4
     possessed_class=WBP_AugmentSelector_AugmentInfo_C
     tracks [{
       type=MovieSceneFloatTrack
       display_name=Render Opacity
       sections [{
         type=MovieSceneFloatSection
         end_frame=30
         start_frame=0
         channels [None {type=MovieSceneScriptingFloatChannel, thinned_from=9, keys [[0,1],[3,0.25],[4,0.5],[6,0.5],[7,0.25],[12,0.25],[13,0],[30,1]]}]
        }]
      }]
    }
   ]
   display_rate {numerator=60}
  }
 }
 Anim_UnfocusScale {
  binding_count=0
  length_seconds=0.1333
  object=&self:Anim_UnfocusScale
  sequence {
   bindings []
   display_rate {numerator=60}
   tracks [{type=MovieSceneEventTrack, sections [{type=MovieSceneEventRepeaterSection, end_frame=8, start_frame=0}]}]
  }
 }
]
graphs [
 ActivateAugment {nodes [
   ActivateAugment() {is=title K2Node_FunctionEntry}
   > PlayAnimation(InAnimation=Get Anim_Activate() -> (Anim_Activate:Widget Animation Object Reference))
   > n1 = TriggerActivationAnimation!(
    execute
    self:WBP Augment Selector Augment Info Object Reference=Get WBP_AugmentSelector_AugmentInfo() -> (WBP_AugmentSelector_AugmentInfo:WBP Augment Selector Augment Info Object Reference)
   ) -> (then, AnimationEndTime:double)
   n2 = Get Anim_Activate()
   > Return Node!(execute, AnimationEndTime:double="Max (Float)"(A=@n1, B=GetEndTime(self=@n2)))
  ]}
 EventGraph {nodes [
   Event OnInitialized()
   > BindToAnimationFinished(Animation=Get Anim_Focus() -> (Anim_Focus:Widget Animation Object Reference), Delegate=Create Event())
   Event Construct()
   > SetVisibility(
    self=Get Image_TopShadow() -> (Image_TopShadow:Image Object Reference)
    InVisibility=Select(Option 0=Collapsed, Option 1=SelfHitTestInvisible, Index=Get bEnableTopShadowVisibility() -> (bEnableTopShadowVisibility:bool))
   )
   > SetVisibility(
    self=Get Image_BottomShadow() -> (Image_BottomShadow:Image Object Reference)
    InVisibility=Select(Option 0=Collapsed, Option 1=SelfHitTestInvisible, Index=Get bEnableBottomShadowVisibility() -> (bEnableBottomShadowVisibility:bool))
   )
   > n1 = GetDynamicMaterial(self=Get Image_ConfirmationOutline!(self:WBP Augment Selector Augment Info Object Reference=Get WBP_AugmentSelector_AugmentInfo()) -> (Image_ConfirmationOutline:Image Object Reference))
   > Set MID_ConfirmationOutline(MID_ConfirmationOutline:Material Instance Dynamic Object Reference=@n1) -> (Output_Get:Material Instance Dynamic Object Reference)
   > n2 = GetDynamicMaterial(self=Get Image_GreenOverlay!(self:WBP Augment Selector Augment Info Object Reference=Get WBP_AugmentSelector_AugmentInfo()) -> (Image_GreenOverlay:Image Object Reference))
   > Set MID_GreenOverlay(MID_GreenOverlay:Material Instance Dynamic Object Reference=@n2) -> (Output_Get:Material Instance Dynamic Object Reference)
  ]}
 FocusAugment {nodes [
   n1 = FocusAugment!() -> (then, InFocusAugment:bool=~) {is=title K2Node_FunctionEntry}
   > n2 = Set IsFocused(IsFocused:bool=@n1) -> (Output_Get:bool)
   > n3 = Branch(Condition=@n2)
   > StopAnimation(InAnimation=Get Anim_UnfocusScale() -> (Anim_UnfocusScale:Widget Animation Object Reference))
   > PlayAnimation(InAnimation=Get Anim_Focus())
   > PlaySound2D(Sound=/PlayerAugmentsUI/Sound/PlayerAugments_UI_Selector_Highlight_Cue)
   > PlaySound2D(Sound=/PlayerAugmentsUI/Sound/PlayerAugments_UI_SetAugmentIcon_Cue)
   StopAnimation(execute=@n3.else, InAnimation=Get Anim_Focus())
   > PlayAnimation(InAnimation=Get Anim_UnfocusScale())
  ]}
 Highlight {nodes [
   n1 = Highlight!() -> (then, PlayAnimation:bool=~) {is=title K2Node_FunctionEntry}
   > FocusAugment!(execute, self, InFocusAugment:bool=true) -> (then) {is=title K2Node_CallFunction}
   > SetFocused!(
    execute
    self:WBP Augment Selector Augment Info Object Reference=Get WBP_AugmentSelector_AugmentInfo()
    IsFocused:bool=true
    PlayAnimation:bool=@n1
   ) -> (then)
  ]}
 LerpHeightAnim_Focus {nodes [
   LerpHeightAnim_Focus() {is=title K2Node_FunctionEntry}
   n1 = Get Anim_Focus()
   n2 = SafeDivide(
    A=GetAnimationCurrentTime!(self:User Widget Object Reference, InAnimation:Widget Animation Object Reference=@n1) -> (ReturnValue:float)
    B=GetEndTime(self=@n1)
   )
   n3 = Get FocusAnimCurve() -> (FocusAnimCurve:Curve Float Object Reference)
   n4 = Get!() -> (
    HeightSizeRange_LowerBound_Type:ERangeBoundTypes Enum
    HeightSizeRange_LowerBound_Value:float=~
    HeightSizeRange_UpperBound_Type:ERangeBoundTypes Enum
    HeightSizeRange_UpperBound_Value:float=~
   )
   > SetMaxDesiredHeight!(
    execute
    self:Size Box Object Reference=Get SizeBox_AugmentScale() -> (SizeBox_AugmentScale:Size Box Object Reference)
    InMaxDesiredHeight:float=Lerp(
     A=@n4.HeightSizeRange_LowerBound_Value
     B=@n4.HeightSizeRange_UpperBound_Value
     Alpha=GetFloatValue!(self:Curve Float Object Reference=@n3, InTime:float=@n2) -> (ReturnValue:float)
    )
   ) -> (then)
  ]}
 LerpHeightAnim_Unfocus {nodes [
   LerpHeightAnim_Unfocus() {is=title K2Node_FunctionEntry}
   n1 = Get Anim_UnfocusScale()
   n2 = SafeDivide(A=GetAnimationCurrentTime(InAnimation=@n1), B=GetEndTime(self=@n1))
   n3 = Get FocusAnimCurve()
   n4 = Get() -> (HeightSizeRange_LowerBound_Value=~, HeightSizeRange_UpperBound_Value=~)
   > SetMaxDesiredHeight(
    self=Get SizeBox_AugmentScale()
    InMaxDesiredHeight=Lerp(A=@n4.HeightSizeRange_UpperBound_Value, B=@n4.HeightSizeRange_LowerBound_Value, Alpha=GetFloatValue(self=@n3, InTime=@n2))
   )
  ]}
 PlayHUDSwapAnim {nodes [
   PlayHUDSwapAnim() {is=title K2Node_FunctionEntry}
   > PlayHUDSwapAnim(self:WBP Augment Selector Augment Info Object Reference=Get WBP_AugmentSelector_AugmentInfo()) {is=title K2Node_CallFunction}
  ]}
 RerollAugment {nodes [
   RerollAugment() {is=title K2Node_FunctionEntry}
   > PlayAnimation(InAnimation=Get Anim_Reroll() -> (Anim_Reroll:Widget Animation Object Reference))
   > PlayRerollAnim(self:WBP Augment Selector Augment Info Object Reference=Get WBP_AugmentSelector_AugmentInfo())
  ]}
 ResetState {nodes [
   ResetState() {is=title K2Node_FunctionEntry}
   > FocusAugment(InFocusAugment=false)
   > SetScalarParameterValue(
    self=Get MID_ConfirmationOutline() -> (MID_ConfirmationOutline:Material Instance Dynamic Object Reference)
    ParameterName=Anim_BorderAlpha
    Value=0
   )
   > SetScalarParameterValue(self=Get MID_GreenOverlay() -> (MID_GreenOverlay:Material Instance Dynamic Object Reference), ParameterName=Anim_BackingAlpha, Value=0)
   > SetRenderTransform!(
    execute
    self:Widget Object Reference=Get Overlay_Main() -> (Overlay_Main:Overlay Object Reference)
    InTransform_Translation:Vector 2D Structure=(X=0,Y=0)
    InTransform_Scale:Vector 2D Structure=(X=1,Y=1)
    InTransform_Shear:Vector 2D Structure=(X=0,Y=0)
    InTransform_Angle:float=0
   ) -> (then)
   > ResetState(self:WBP Augment Selector Augment Info Object Reference=Get WBP_AugmentSelector_AugmentInfo()) {is=title K2Node_CallFunction}
  ]}
 Set Augment Index {nodes [
   n1 = Set Augment Index!() -> (then, New Index:int) {is=title K2Node_FunctionEntry}
   > SetAugmentIndexInList!(
    execute
    self:Fort Player Augment Info Touch Widget Object Reference=Get FortPlayerAugmentInfoTouchWidget() -> (FortPlayerAugmentInfoTouchWidget:Fort Player Augment Info Touch Widget Object Reference)
    InAugmentIndexInList:int=@n1
   ) -> (then)
  ]}
 SetAugmentInfo {nodes [
   n1 = SetAugmentInfo!() -> (then, InAugmentItemDefinition:Fort Player Augment Item Definition Object Reference) {is=title K2Node_FunctionEntry}
   > SetAugmentInfo!(
    execute
    self:WBP Augment Selector Augment Info Object Reference=Get WBP_AugmentSelector_AugmentInfo()
    InAugmentItemDefinition:Fort Player Augment Item Definition Object Reference=@n1
   ) -> (then) {is=title K2Node_CallFunction}
  ]}
 TriggerFocusPulseAnim {nodes [
   TriggerFocusPulseAnim() {is=title K2Node_FunctionEntry}
   > Branch(Condition=Get IsFocused() -> (IsFocused:bool))
   > PlayPulseAnimation(self:WBP Augment Selector Augment Info Object Reference=Get WBP_AugmentSelector_AugmentInfo())
  ]}
 Unhighlight {nodes [
   n1 = Unhighlight!() -> (then, PlayAnimation:bool=~) {is=title K2Node_FunctionEntry}
   > FocusAugment(InFocusAugment=false)
   > SetFocused(self=Get WBP_AugmentSelector_AugmentInfo(), IsFocused=false, PlayAnimation=@n1)
  ]}
]
variables [
 bEnableBottomShadowVisibility {type=boolean}
 bEnableTopShadowVisibility {type=boolean}
 FocusAnimCurve {type=CurveFloat, default=&/FocusSimpleCubicCurve}
 HeightSizeRange {type=FloatRange, default {lowerBound {type=Exclusive, value=215}, upperBound {type=Exclusive, value=380}}}
 In Augment Item Definition {type=FortPlayerAugmentItemDefinition, default=null}
 IsFocused {type=boolean}
 MID_ConfirmationOutline {type=MaterialInstanceDynamic}
 MID_GreenOverlay {type=MaterialInstanceDynamic}
]
widgets {
 info {inheritedWidgetCount=0, parentClass=&UserWidget, rootWidgetClass=&SizeBox, widgetCount=7}
 widgets [
  SizeBox_AugmentScale {
   widgetClassPath=&SizeBox
   bIsVariable=true
   widget=SizeBox_AugmentScale
   properties {bOverride_MaxDesiredHeight=true, maxDesiredHeight=375}
  }
  Overlay_Main {widgetClassPath=&Overlay, bIsVariable=true, parent=SizeBox_AugmentScale, widget=Overlay_Main, slot {class=SizeBoxSlot}}
  HardwareVisibilityBorder_TouchOnlyAugment {
   widgetClassPath=&CommonHardwareVisibilityBorder
   parent=Overlay_Main
   widget=HardwareVisibilityBorder_TouchOnlyAugment
   properties {
    visibleType=Visible
    visibilityQuery {
     autoDescription=" ANY( Input.Touch )"
     tokenStreamVersion=0
     userDescription=touch-only
     queryTokenStream [0,1,1,1,0]
     tagDictionary [{tagName=Input.Touch}]
    }
   }
   slot {class=OverlaySlot, properties {horizontalAlignment=HAlign_Fill, verticalAlignment=VAlign_Fill}}
  }
  FortPlayerAugmentInfoTouchWidget {
   widgetClassPath=/Script/PlayerAugmentsUI.FortPlayerAugmentInfoTouchWidget
   bIsVariable=true
   parent=HardwareVisibilityBorder_TouchOnlyAugment
   widget=FortPlayerAugmentInfoTouchWidget
   properties {visibility=Visible}
   slot {class=BorderSlot, properties {padding=<ltrb 0 0 0 0>}}
  }
  Image_TopShadow {
   widgetClassPath=&Image
   bIsVariable=true
   parent=Overlay_Main
   widget=Image_TopShadow
   properties {
    visibility=Collapsed
    brush {
     bIsDynamicallyLoaded=false
     drawAs=Image
     imageType=NoImage
     mirroring=NoMirror
     resourceName=None
     resourceObject=/PlayerAugmentsUI/Materials/MI_UI_PlayerAugment_BackingPanel_DivderBrighten_FadeTop
     tiling=NoTile
     imageSize=<32 64>
     margin=<ltrb 0 0 0 0>
     outlineSettings {
      bUseBrushTransparency=false
      roundingType=HalfHeightRadius
      width=0
      color {colorUseRule=UseColor_Specified, specifiedColor=<rgba 0 0 0 0>}
      cornerRadii=<0 0 0 0>
     }
     tintColor {colorUseRule=UseColor_Specified, specifiedColor=<rgba 1 1 1 1>}
     uVRegion {bIsValid=false, max=<0 0>, min=<0 0>}
    }
    renderTransform {angle=0, scale=<1 1>, shear=<0 0>, translation=<0 -66>}
   }
   slot {class=OverlaySlot, properties {horizontalAlignment=HAlign_Fill}}
  }
  Image_BottomShadow {
   widgetClassPath=&Image
   bIsVariable=true
   parent=Overlay_Main
   widget=Image_BottomShadow
   properties {
    visibility=Collapsed
    brush {
     bIsDynamicallyLoaded=false
     drawAs=Image
     imageType=NoImage
     mirroring=NoMirror
     resourceName=None
     resourceObject=/PlayerAugmentsUI/Materials/MI_UI_PlayerAugment_BackingPanel_DivderBrighten_FadeBottom
     tiling=NoTile
     imageSize=<32 64>
     margin=<ltrb 0 0 0 0>
     outlineSettings {
      bUseBrushTransparency=false
      roundingType=HalfHeightRadius
      width=0
      color {colorUseRule=UseColor_Specified, specifiedColor=<rgba 0 0 0 0>}
      cornerRadii=<0 0 0 0>
     }
     tintColor {colorUseRule=UseColor_Specified, specifiedColor=<rgba 1 1 1 1>}
     uVRegion {bIsValid=false, max=<0 0>, min=<0 0>}
    }
    renderTransform {angle=0, scale=<1 1>, shear=<0 0>, translation=<0 66>}
   }
   slot {class=OverlaySlot, properties {horizontalAlignment=HAlign_Fill, verticalAlignment=VAlign_Bottom}}
  }
  WBP_AugmentSelector_AugmentInfo {
   widgetClassPath=&/WBP_AugmentSelector_AugmentInfo
   bIsVariable=true
   parent=Overlay_Main
   widget=WBP_AugmentSelector_AugmentInfo
   properties {renderTransformPivot=<0.5 0>}
   slot {class=OverlaySlot, properties {horizontalAlignment=HAlign_Fill, verticalAlignment=VAlign_Fill}}
  }
 ]
}
