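# WBP_AugmentSelector_AugmentDivider is a visual-only divider widget, named for the augment selector in the player augments UI.
# Nested Overlays hold a Spacer that gives the divider a minimum height, with a Shadow image filling the same area, slightly inset vertically.
# The image draws a divider-brighten backing panel material instance, and the widget has an empty event graph with no variables.
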WidgetBlueprint /PlayerAugmentsUI/UI/WBP_AugmentSelector_AugmentDivider
graphs [EventGraph {links [], nodes []}]
variables []
widgets {
 info {inheritedWidgetCount=0, parentClass=&UserWidget, rootWidgetClass=&Overlay, widgetCount=4}
 widgets [
  Overlay_Root {widgetClassPath=&Overlay, widget=Overlay_Root}
  Overlay_AugmentDivider {
   widgetClassPath=&Overlay
   parent=Overlay_Root
   widget=Overlay_AugmentDivider
   slot {class=OverlaySlot, properties {horizontalAlignment=HAlign_Fill, verticalAlignment=VAlign_Fill}}
  }
  Spacer_83 {
   widgetClassPath=&Spacer
   bIsVariable=true
   parent=Overlay_AugmentDivider
   widget=Spacer_83
   properties {size=<1 40>}
   slot {class=OverlaySlot, properties {horizontalAlignment=HAlign_Fill, verticalAlignment=VAlign_Fill}}
  }
  Shadow {
   widgetClassPath=&Image
   bIsVariable=true
   parent=Overlay_AugmentDivider
   widget=Shadow
   properties {brush {
     bIsDynamicallyLoaded=false
     drawAs=Image
     imageType=NoImage
     mirroring=NoMirror
     resourceName=None
     resourceObject=/PlayerAugmentsUI/Materials/MI_UI_PlayerAugment_BackingPanel_DivderBrighten
     tiling=NoTile
     imageSize=<24 32>
     margin=<ltrb 0 0 0 0>
     outlineSettings {
      bUseBrushTransparency=false
      roundingType=HalfHeightRadius
      width=0
      color {colorUseRule=UseColor_Specified, specifiedColor=<rgba 0 0 0 0>}
      cornerRadii=<0 0 0 0>
     }
     tintColor {colorUseRule=UseColor_Specified, specifiedColor=<rgba 1 1 1 1>}
     uVRegion {bIsValid=false, max=<0 0>, min=<0 0>}
    }}
   slot {class=OverlaySlot, properties {horizontalAlignment=HAlign_Fill, verticalAlignment=VAlign_Fill, padding=<ltrb 0 2 0 2>}}
  }
 ]
}
