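# HUD widget for the hold-to-reroll augment prompt, showing keybind or touch input, owned currency and cost.
# OnInitialized binds the controller's AugmentUIEvents hold changes to start/end progress, and switches alert wording to coins when a global currency override exists.
# Currency updates show "Free!" at zero cost and drive DisplayAlert, swapping in a red warning when funds are short.
# Rerolling plays Anim_Reroll, which has no bindings.
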WidgetBlueprint /PlayerAugmentsUI/UI/WBP_AugmentSelector_Reroll
animations [Anim_Reroll {
  binding_count=0
  length_seconds=0.3334
  object=&self:Anim_Reroll
  sequence {bindings [], display_rate {numerator=30}}
 }]
graphs [
 DisplayAlert {nodes [
   DisplayAlert!() -> (then, InDisplayAlert:bool=~) {is=title K2Node_FunctionEntry}
   > SetActiveWidget(
    self=Get Switcher_RerollAction() -> (Switcher_RerollAction:Common Visibility Switcher Object Reference)
    Widget=Select(
     Option 0:Widget Object Reference=Get Keybind_RerollAction() -> (Keybind_RerollAction:Keybind Widget Object Reference)
     Option 1:Widget Object Reference=Get Overlay_AlertIcon() -> (Overlay_AlertIcon:Overlay Object Reference)
     Index=Get InDisplayAlert() -> (InDisplayAlert:bool)
    ) -> (ReturnValue:Widget Object Reference)
   )
   > SetVisibility(
    self=Get Overlay_NotEnoughCurrency() -> (Overlay_NotEnoughCurrency:Overlay Object Reference)
    InVisibility=Select(Option 0:ESlateVisibility Enum=Collapsed, Option 1:ESlateVisibility Enum=HitTestInvisible, Index=Get InDisplayAlert()) -> (ReturnValue:ESlateVisibility Enum)
   )
   n1 = Get InDisplayAlert()
   > SetColorAndOpacity(
    self=Get Text_OwnedCurrency() -> (Text_OwnedCurrency:Common Text Object Reference)
    InColorAndOpacity=Make Slate Color(SpecifiedColor=SelectColor(A=(R=0.855,G=0.006,B=0.0176,A=1), B=(R=1,G=1,B=1,A=1), bPickA=@n1))
   )
   > SetVisibility(
    self=Get RerollText() -> (RerollText:Overlay Object Reference)
    InVisibility=Select(Option 0=SelfHitTestInvisible, Option 1=Collapsed, Index=Get InDisplayAlert())
   )
  ]}
 EventGraph {nodes [
   n1 = OnAugmentHoldRerollChanged!() -> (OutputDelegate, then, HoldState:EAugmentHoldInputRerollState Enum=Pressed, HoldDurationSeconds:float=~) {is=title K2Node_CustomEvent}
   > n2 = Switch on EAugmentHoldInputRerollState!(execute, Selection:EAugmentHoldInputRerollState Enum=@n1.HoldState) -> (Pressed:Exec, Released:Exec) {is=title K2Node_SwitchEnum}
   > StartRerollHoldProgress!(execute, self:Fort Player Augment HUDReroll Widget Object Reference, Duration:float=@n1.HoldDurationSeconds) -> (then)
   EndRerollHoldProgress(execute=@n2.Released, self:Fort Player Augment HUDReroll Widget Object Reference)
   n3 = Event PreConstruct()
   > Branch(Condition=@n3.IsDesignTime)
   > DisplayAlert!(execute, self, InDisplayAlert:bool=false) -> (then) {is=title K2Node_CallFunction}
   Event OnRerollAugments() {is=title K2Node_Event}
   > PlayAnimation(InAnimation=Get Anim_Reroll() -> (Anim_Reroll:Widget Animation Object Reference), bRestoreState=true)
   n4 = Event OnUpdateCurrency!() -> (OutputDelegate, then, OwnedCurrency:int=0, RerollCost:int=0) {is=title K2Node_Event}
   > UpdateRerollInfo!(execute, self, OwnedCurrency:int=@n4.OwnedCurrency, AugmentCost:int=@n4.RerollCost) -> (then)
   Event OnInitialized()
   n5 = GetOwningPlayer()
   > Is Valid(InputObject=@n5)
   n6 = GetComponentByClass!(
    self:Actor Object Reference=@n5
    ComponentClass:Actor Component Class Reference=/Script/PlayerAugmentsCodeRuntime.FortControllerComponent_AugmentUIEvents
   ) -> (ReturnValue:Fort Controller Component Augment UIEvents Object Reference)
   > n7 = Is Valid(InputObject=@n6)
   n8 = GetPlayspaceGlobalCurrencyOverrideForActor!(execute=[@n9.then, @n7.Is Not Valid], Actor:Actor Object Reference=GetOwningPlayer(self=Self-Reference())) -> (then, ReturnValue:Playspace Component Global Currency Override Object Reference)
   > Is Valid(InputObject=@n8)
   > Is Valid(InputObject=Get OverrideGlobalCurrency!(self:Playspace Component Global Currency Override Object Reference=@n8) -> (OverrideGlobalCurrency:Fort World Item Definition Object Reference))
   > Set Text(Text="LOCTEXT(\"Not enough coins to reroll!\")", self=Get Text_Alert() -> (Text_Alert:Common Text Object Reference))
   n9 = Bind Event to Augment Hold Input Reroll Changed(execute=@n7.Is Valid, self:Fort Controller Component Augment UIEvents Object Reference=@n6, Delegate=@n1) {is=title K2Node_AddDelegate}
  ]}
 UpdateRerollInfo {nodes [
   n1 = UpdateRerollInfo!() -> (then, OwnedCurrency:int, AugmentCost:int) {is=title K2Node_FunctionEntry}
   > SetActiveWidgetIndex(
    self:Common Visibility Switcher Object Reference=Get Switcher_Cost() -> (Switcher_Cost:Common Visibility Switcher Object Reference)
    Index="To Integer (Boolean)"!(InBool:bool="Equal (Integer)"(A=@n1.AugmentCost)) -> (ReturnValue:int=0)
   )
   > DisplayAlert(InDisplayAlert="integer < integer"(A=Get OwnedCurrency() -> (OwnedCurrency:int=0), B=@n1.AugmentCost))
  ]}
]
variables []
widgets {
 info {
  inheritedWidgetCount=4
  parentClass=/Script/PlayerAugmentsUI.FortPlayerAugmentHUDRerollWidget
  rootWidgetClass=&Overlay
  widgetCount=39
 }
 widgets [
  Overlay_Main_RerollPrompt {widgetClassPath=&Overlay, widget=Overlay_Main_RerollPrompt}
  Spacer_HeightControl {
   widgetClassPath=&Spacer
   parent=Overlay_Main_RerollPrompt
   widget=Spacer_HeightControl
   properties {size=<1 56>}
   slot {class=OverlaySlot, properties {verticalAlignment=VAlign_Fill}}
  }
  Image_Backing_Glow {
   widgetClassPath=&Image
   parent=Overlay_Main_RerollPrompt
   widget=Image_Backing_Glow
   properties {
    visibility=SelfHitTestInvisible
    brush {
     bIsDynamicallyLoaded=false
     drawAs=Box
     imageType=NoImage
     mirroring=NoMirror
     resourceName=None
     resourceObject=/PlayerAugmentsUI/Materials/MI_UI_PlayerAugment_Font_Glow
     tiling=NoTile
     imageSize=<88 88>
     margin=<ltrb 0.5 0.5 0.5 0.5>
     outlineSettings {
      bUseBrushTransparency=false
      roundingType=HalfHeightRadius
      width=0
      color {colorUseRule=UseColor_Specified, specifiedColor=<rgba 0 0 0 0>}
      cornerRadii=<0 0 0 0>
     }
     tintColor {colorUseRule=UseColor_Specified, specifiedColor=<rgba 1 1 1 1>}
     uVRegion {bIsValid=false, max=<0 0>, min=<0 0>}
    }
   }
   slot {class=OverlaySlot, properties {horizontalAlignment=HAlign_Fill, verticalAlignment=VAlign_Fill}}
  }
  HardwareVisibilityBorder_TouchOnly {
   widgetClassPath=&CommonHardwareVisibilityBorder
   parent=Overlay_Main_RerollPrompt
   widget=HardwareVisibilityBorder_TouchOnly
   properties {
    visibleType=Visible
    visibilityQuery {
     autoDescription=" ANY( Input.Touch )"
     tokenStreamVersion=0
     userDescription=touch-only
     queryTokenStream [0,1,1,1,0]
     tagDictionary [{tagName=Input.Touch}]
    }
   }
   slot {class=OverlaySlot, properties {horizontalAlignment=HAlign_Fill, verticalAlignment=VAlign_Fill}}
  }
  FortTouchUserWidget {
   widgetClassPath=/Script/FortniteUI.FortTouchUserWidget
   bIsVariable=true
   parent=HardwareVisibilityBorder_TouchOnly
   widget=FortTouchUserWidget
   properties {visibility=Visible, actionNames [AugmentReroll]}
   slot {class=BorderSlot, properties {padding=<ltrb 0 0 0 0>}}
  }
  HorizontalBox_43 {
   widgetClassPath=&HorizontalBox
   parent=Overlay_Main_RerollPrompt
   widget=HorizontalBox_43
   slot {class=OverlaySlot, properties {horizontalAlignment=HAlign_Fill, verticalAlignment=VAlign_Fill, padding=<ltrb 0 2 0 0>}}
  }
  border_touch-only {
   widgetClassPath=&CommonHardwareVisibilityBorder
   parent=HorizontalBox_43
   widget=border_touch-only
   properties {
    horizontalAlignment=HAlign_Center
    verticalAlignment=VAlign_Center
    visibilityQuery {
     autoDescription=" ANY( Input.Touch )"
     tokenStreamVersion=0
     userDescription=touch-only
     queryTokenStream [0,1,1,1,0]
     tagDictionary [{tagName=Input.Touch}]
    }
   }
   slot {class=HorizontalBoxSlot, properties {padding=<ltrb 0 0 5 0>}}
  }
  Overlay_ReRollProgress {
   widgetClassPath=&Overlay
   parent=border_touch-only
   widget=Overlay_ReRollProgress
   slot {class=BorderSlot, properties {horizontalAlignment=HAlign_Center, verticalAlignment=VAlign_Center, padding=<ltrb 0 0 0 0>}}
  }
  Image_ProgressBar {
   widgetClassPath=&Image
   bInherited=true
   parent=Overlay_ReRollProgress
   widget=Image_ProgressBar
   properties {
    visibility=Collapsed
    brush {
     bIsDynamicallyLoaded=false
     drawAs=Image
     imageType=NoImage
     mirroring=NoMirror
     resourceName=None
     resourceObject=/Game/Athena/HUD/Materials/M_UI_RadialProgress_Consumable
     tiling=NoTile
     imageSize=<45 45>
     margin=<ltrb 0 0 0 0>
     outlineSettings {
      bUseBrushTransparency=false
      roundingType=HalfHeightRadius
      width=0
      color {colorUseRule=UseColor_Specified, specifiedColor=<rgba 0 0 0 0>}
      cornerRadii=<0 0 0 0>
     }
     tintColor {colorUseRule=UseColor_Specified, specifiedColor=<rgba 1 1 1 1>}
     uVRegion {bIsValid=false, max=<0 0>, min=<0 0>}
    }
    renderTransform {angle=0, scale=<1.5 1.5>, shear=<0 0>, translation=<0 0>}
   }
   slot {class=OverlaySlot, properties {horizontalAlignment=HAlign_Center, verticalAlignment=VAlign_Center}}
  }
  Image_RerollIcon {
   widgetClassPath=&FortLazyImage
   parent=Overlay_ReRollProgress
   widget=Image_RerollIcon
   properties {brush {
     bIsDynamicallyLoaded=false
     drawAs=Image
     imageType=NoImage
     mirroring=NoMirror
     resourceName=None
     resourceObject=/Game/UI/Icons/Icon-RandomDice/Icon-RandomDice-128
     tiling=NoTile
     imageSize=<45 45>
     margin=<ltrb 0 0 0 0>
     outlineSettings {
      bUseBrushTransparency=false
      roundingType=HalfHeightRadius
      width=0
      color {colorUseRule=UseColor_Specified, specifiedColor=<rgba 0 0 0 0>}
      cornerRadii=<0 0 0 0>
     }
     tintColor {colorUseRule=UseColor_Specified, specifiedColor=<rgba 1 1 1 1>}
     uVRegion {bIsValid=false, max=<0 0>, min=<0 0>}
    }}
   slot {class=OverlaySlot, properties {horizontalAlignment=HAlign_Center, verticalAlignment=VAlign_Center}}
  }
  Overlay_RerollPrompt {
   widgetClassPath=&Overlay
   parent=HorizontalBox_43
   widget=Overlay_RerollPrompt
   slot {class=HorizontalBoxSlot, properties {horizontalAlignment=HAlign_Left, verticalAlignment=VAlign_Center, padding=<ltrb 0 0 10 0>}}
  }
  HorizontalBox_119 {widgetClassPath=&HorizontalBox, parent=Overlay_RerollPrompt, widget=HorizontalBox_119, slot {class=OverlaySlot}}
  SizeBox_0 {
   widgetClassPath=&SizeBox
   parent=HorizontalBox_119
   widget=SizeBox_0
   properties {bOverride_HeightOverride=true, heightOverride=52, visibility=Visible, widthOverride=52}
   slot {class=HorizontalBoxSlot, properties {horizontalAlignment=HAlign_Center, verticalAlignment=VAlign_Center}}
  }
  Switcher_RerollAction {
   widgetClassPath=&CommonVisibilitySwitcher
   bIsVariable=true
   parent=SizeBox_0
   widget=Switcher_RerollAction
   slot {class=SizeBoxSlot, properties {horizontalAlignment=HAlign_Center, verticalAlignment=VAlign_Center}}
  }
  Keybind_RerollAction {
   widgetClassPath=/FortUILibrary/Input/KeybindWidget
   bIsVariable=true
   parent=Switcher_RerollAction
   widget=Keybind_RerollAction
   properties {boundAction=AugmentReroll, forcedHoldKeybindStatus=ForcedHold, boundActionFallbacks [UseOrReload]}
   slot {class=CommonVisibilitySwitcherSlot, properties {horizontalAlignment=HAlign_Center, verticalAlignment=VAlign_Center}}
  }
  Overlay_AlertIcon {
   widgetClassPath=&Overlay
   bIsVariable=true
   parent=Switcher_RerollAction
   widget=Overlay_AlertIcon
   slot {class=CommonVisibilitySwitcherSlot}
  }
  Image_AlertIcon {
   widgetClassPath=&Image
   parent=Overlay_AlertIcon
   widget=Image_AlertIcon
   properties {brush {
     bIsDynamicallyLoaded=false
     drawAs=Image
     imageType=NoImage
     mirroring=NoMirror
     resourceName=None
     resourceObject=/PlayerAugmentsUI/Materials/MI_UI_PlayerAugment_Alert_Glow
     tiling=NoTile
     imageSize=<24 24>
     margin=<ltrb 0 0 0 0>
     outlineSettings {
      bUseBrushTransparency=false
      roundingType=HalfHeightRadius
      width=0
      color {colorUseRule=UseColor_Specified, specifiedColor=<rgba 0 0 0 0>}
      cornerRadii=<0 0 0 0>
     }
     tintColor {colorUseRule=UseColor_Specified, specifiedColor=<rgba 0.402 0.0006 0.0003 1>}
     uVRegion {bIsValid=false, max=<0 0>, min=<0 0>}
    }}
   slot {class=OverlaySlot, properties {horizontalAlignment=HAlign_Fill, verticalAlignment=VAlign_Fill, padding=<ltrb -2 0 -2 0>}}
  }
  Image_AlertIcon_1 {
   widgetClassPath=&Image
   parent=Overlay_AlertIcon
   widget=Image_AlertIcon_1
   properties {brush {
     bIsDynamicallyLoaded=false
     drawAs=Image
     imageType=NoImage
     mirroring=NoMirror
     resourceName=None
     resourceObject=/PlayerAugmentsUI/Materials/MI_UI_PlayerAugment_Alert_Backing
     tiling=NoTile
     imageSize=<24 24>
     margin=<ltrb 0 0 0 0>
     outlineSettings {
      bUseBrushTransparency=false
      roundingType=HalfHeightRadius
      width=0
      color {colorUseRule=UseColor_Specified, specifiedColor=<rgba 0 0 0 0>}
      cornerRadii=<0 0 0 0>
     }
     tintColor {colorUseRule=UseColor_Specified, specifiedColor=<rgba 0.402 0.0006 0.0003 1>}
     uVRegion {bIsValid=false, max=<0 0>, min=<0 0>}
    }}
   slot {class=OverlaySlot, properties {horizontalAlignment=HAlign_Center, verticalAlignment=VAlign_Center}}
  }
  Text_Alert_1 {
   widgetClassPath=&CommonTextBlock
   parent=Overlay_AlertIcon
   widget=Text_Alert_1
   properties {
    justification=Center
    text="!"
    textTransformPolicy=ToUpper
    font {
     bForceMonospaced=false
     bMaterialIsStencil=false
     fontMaterial=/PlayerAugmentsUI/Materials/MI_UI_PlayerAugment_Alert_Font
     fontObject=&BurbankSmall
     letterSpacing=0
     monospacedWidth=1
     size=15
     skewAmount=0
     typefaceFontName=Black
     outlineSettings {
      bApplyOutlineToDropShadows=false
      bMiteredCorners=false
      bSeparateFillAlpha=true
      outlineMaterial=null
      outlineSize=2
      outlineColor=<rgba 0 0 0 0>
     }
    }
    shadowOffset=<0 0>
   }
   slot {class=OverlaySlot, properties {horizontalAlignment=HAlign_Center, verticalAlignment=VAlign_Center}}
  }
  RerollText {
   widgetClassPath=&Overlay
   bIsVariable=true
   parent=HorizontalBox_119
   widget=RerollText
   slot {class=HorizontalBoxSlot, properties {horizontalAlignment=HAlign_Left, verticalAlignment=VAlign_Center, padding=<ltrb 8 2 0 0>}}
  }
  Text_Glow_1 {
   widgetClassPath=&Image
   parent=RerollText
   widget=Text_Glow_1
   properties {
    visibility=SelfHitTestInvisible
    brush {
     bIsDynamicallyLoaded=false
     drawAs=Box
     imageType=NoImage
     mirroring=NoMirror
     resourceName=None
     resourceObject=&MI_UI_PlayerAugment_Font_Glow
     tiling=NoTile
     imageSize=<40 40>
     margin=<ltrb 0.5 0.5 0.5 0.5>
     outlineSettings {
      bUseBrushTransparency=false
      roundingType=HalfHeightRadius
      width=0
      color {colorUseRule=UseColor_Specified, specifiedColor=<rgba 0 0 0 0>}
      cornerRadii=<0 0 0 0>
     }
     tintColor {colorUseRule=UseColor_Specified, specifiedColor=<rgba 1 1 1 1>}
     uVRegion {bIsValid=false, max=<0 0>, min=<0 0>}
    }
   }
   slot {class=OverlaySlot, properties {horizontalAlignment=HAlign_Fill, verticalAlignment=VAlign_Fill, padding=<ltrb -8 -6 -8 -6>}}
  }
  Text_Reroll {
   widgetClassPath=&CommonTextBlock
   parent=RerollText
   widget=Text_Reroll
   properties {
    mobileFontSizeMultiplier=1.4
    text="Reroll (Hold)"
    textTransformPolicy=ToUpper
    font {
     bForceMonospaced=false
     bMaterialIsStencil=false
     fontMaterial=/PlayerAugmentsUI/Materials/MI_UI_PlayerAugment_RerollText
     fontObject=&BurbankSmall
     letterSpacing=0
     monospacedWidth=1
     size=16
     skewAmount=0
     typefaceFontName=Black
     outlineSettings {
      bApplyOutlineToDropShadows=false
      bMiteredCorners=false
      bSeparateFillAlpha=true
      outlineMaterial=&M-FontShadow
      outlineSize=2
      outlineColor=<rgba 0.6667 0 1 0.75>
     }
    }
    shadowColorAndOpacity=<rgba 0 0.8333 1 0.125>
    shadowOffset=<4 2>
   }
   slot {class=OverlaySlot}
  }
  Overlay_GoldCount {
   widgetClassPath=&Overlay
   parent=HorizontalBox_43
   widget=Overlay_GoldCount
   slot {class=HorizontalBoxSlot, properties {verticalAlignment=VAlign_Center, padding=<ltrb 8 0 0 0>}}
  }
  Switcher_Cost {
   widgetClassPath=&CommonVisibilitySwitcher
   bIsVariable=true
   parent=Overlay_GoldCount
   widget=Switcher_Cost
   slot {class=OverlaySlot, properties {horizontalAlignment=HAlign_Fill, verticalAlignment=VAlign_Fill}}
  }
  Overlay_Cost {
   widgetClassPath=&Overlay
   parent=Switcher_Cost
   widget=Overlay_Cost
   properties {renderTransform {angle=0, scale=<1 1>, shear=<0 0>, translation=<0 4>}}
   slot {class=CommonVisibilitySwitcherSlot}
  }
  Image_Cost_Glow {
   widgetClassPath=&Image
   parent=Overlay_Cost
   widget=Image_Cost_Glow
   properties {
    visibility=SelfHitTestInvisible
    brush {
     bIsDynamicallyLoaded=false
     drawAs=Box
     imageType=NoImage
     mirroring=NoMirror
     resourceName=None
     resourceObject=&MI_UI_PlayerAugment_Font_Glow
     tiling=NoTile
     imageSize=<64 64>
     margin=<ltrb 0.5 0.5 0.5 0.5>
     outlineSettings {
      bUseBrushTransparency=false
      roundingType=HalfHeightRadius
      width=0
      color {colorUseRule=UseColor_Specified, specifiedColor=<rgba 0 0 0 0>}
      cornerRadii=<0 0 0 0>
     }
     tintColor {colorUseRule=UseColor_Specified, specifiedColor=<rgba 1 1 1 1>}
     uVRegion {bIsValid=false, max=<0 0>, min=<0 0>}
    }
   }
   slot {class=OverlaySlot, properties {horizontalAlignment=HAlign_Fill, verticalAlignment=VAlign_Fill, padding=<ltrb -16 -6 -16 -6>}}
  }
  HorizontalBox_121 {
   widgetClassPath=&HorizontalBox
   parent=Overlay_Cost
   widget=HorizontalBox_121
   properties {flowDirectionPreference=Culture}
   slot {class=OverlaySlot}
  }
  Overlay_CurrencyIcon {
   widgetClassPath=&Overlay
   parent=HorizontalBox_121
   widget=Overlay_CurrencyIcon
   slot {class=HorizontalBoxSlot, properties {horizontalAlignment=HAlign_Center, verticalAlignment=VAlign_Center}}
  }
  Image_CurrencyIcon {
   widgetClassPath=&FortLazyImage
   bInherited=true
   bIsVariable=true
   parent=Overlay_CurrencyIcon
   widget=Image_CurrencyIcon
   properties {brush {
     bIsDynamicallyLoaded=false
     drawAs=Image
     imageType=NoImage
     mirroring=NoMirror
     resourceName=None
     resourceObject=/PlayerAugmentsUI/Materials/M_UI_PlayerAugment_Currency_Icon
     tiling=NoTile
     imageSize=<48 48>
     margin=<ltrb 0 0 0 0>
     outlineSettings {
      bUseBrushTransparency=false
      roundingType=HalfHeightRadius
      width=0
      color {colorUseRule=UseColor_Specified, specifiedColor=<rgba 0 0 0 0>}
      cornerRadii=<0 0 0 0>
     }
     tintColor {colorUseRule=UseColor_Specified, specifiedColor=<rgba 1 1 1 1>}
     uVRegion {bIsValid=false, max=<0 0>, min=<0 0>}
    }}
   slot {class=OverlaySlot, properties {horizontalAlignment=HAlign_Center, verticalAlignment=VAlign_Center}}
  }
  Text_OwnedCurrency {
   widgetClassPath=&CommonTextBlock
   bInherited=true
   bIsVariable=true
   parent=HorizontalBox_121
   widget=Text_OwnedCurrency
   properties {
    mobileFontSizeMultiplier=1.4
    text=0000
    textTransformPolicy=ToUpper
    font {
     bForceMonospaced=false
     bMaterialIsStencil=false
     fontMaterial=/PlayerAugmentsUI/Materials/MI_UI_PlayerAugment_CurrencyOwned_Font
     fontObject=&BurbankSmall
     letterSpacing=100
     monospacedWidth=1
     size=16
     skewAmount=0
     typefaceFontName=Black
     outlineSettings {
      bApplyOutlineToDropShadows=false
      bMiteredCorners=false
      bSeparateFillAlpha=true
      outlineMaterial=&M-FontShadow
      outlineSize=2
      outlineColor=<rgba 0.6667 0 1 0.75>
     }
    }
    shadowColorAndOpacity=<rgba 0 0.8333 1 0.125>
    shadowOffset=<4 2>
   }
   slot {class=HorizontalBoxSlot, properties {horizontalAlignment=HAlign_Left, verticalAlignment=VAlign_Center}}
  }
  Text_RerollCost {
   widgetClassPath=&CommonTextBlock
   bInherited=true
   bIsVariable=true
   parent=HorizontalBox_121
   widget=Text_RerollCost
   properties {
    mobileFontSizeMultiplier=1.4
    text=" / 0000"
    textTransformPolicy=ToUpper
    font {
     bForceMonospaced=false
     bMaterialIsStencil=false
     fontMaterial=/PlayerAugmentsUI/Materials/MI_UI_PlayerAugment_CurrencyNeeded_Font
     fontObject=&BurbankSmall
     letterSpacing=100
     monospacedWidth=1
     size=16
     skewAmount=0
     typefaceFontName=Black
     outlineSettings {
      bApplyOutlineToDropShadows=false
      bMiteredCorners=false
      bSeparateFillAlpha=true
      outlineMaterial=&M-FontShadow
      outlineSize=2
      outlineColor=<rgba 0.6667 0 1 0.75>
     }
    }
    shadowColorAndOpacity=<rgba 0 0.8333 1 0.125>
    shadowOffset=<4 2>
   }
   slot {class=HorizontalBoxSlot, properties {horizontalAlignment=HAlign_Left, verticalAlignment=VAlign_Center, padding=<ltrb -8 0 0 0>}}
  }
  Overlay_FreeIndicator {
   widgetClassPath=&Overlay
   parent=Switcher_Cost
   widget=Overlay_FreeIndicator
   slot {class=CommonVisibilitySwitcherSlot, properties {horizontalAlignment=HAlign_Right, verticalAlignment=VAlign_Center}}
  }
  Image_FreeBacking_Glow {
   widgetClassPath=&Image
   parent=Overlay_FreeIndicator
   widget=Image_FreeBacking_Glow
   properties {
    visibility=SelfHitTestInvisible
    brush {
     bIsDynamicallyLoaded=false
     drawAs=Box
     imageType=NoImage
     mirroring=NoMirror
     resourceName=None
     resourceObject=/PlayerAugmentsUI/Materials/MI_UI_PlayerAugment_FreeBacking_Glow
     tiling=NoTile
     imageSize=<32 32>
     margin=<ltrb 0.5 0.5 0.5 0.5>
     outlineSettings {
      bUseBrushTransparency=false
      roundingType=HalfHeightRadius
      width=0
      color {colorUseRule=UseColor_Specified, specifiedColor=<rgba 0 0 0 0>}
      cornerRadii=<0 0 0 0>
     }
     tintColor {colorUseRule=UseColor_Specified, specifiedColor=<rgba 1 1 1 1>}
     uVRegion {bIsValid=false, max=<0 0>, min=<0 0>}
    }
   }
   slot {class=OverlaySlot, properties {horizontalAlignment=HAlign_Fill, verticalAlignment=VAlign_Fill, padding=<ltrb -8 -8 -8 -8>}}
  }
  Image_FreeBacking {
   widgetClassPath=&Image
   parent=Overlay_FreeIndicator
   widget=Image_FreeBacking
   properties {
    visibility=SelfHitTestInvisible
    brush {
     bIsDynamicallyLoaded=false
     drawAs=Box
     imageType=NoImage
     mirroring=NoMirror
     resourceName=None
     resourceObject=/PlayerAugmentsUI/Materials/MI_UI_PlayerAugment_FreeBacking
     tiling=NoTile
     imageSize=<16 16>
     margin=<ltrb 0.5 0.5 0.5 0.5>
     outlineSettings {
      bUseBrushTransparency=false
      roundingType=HalfHeightRadius
      width=0
      color {colorUseRule=UseColor_Specified, specifiedColor=<rgba 0 0 0 0>}
      cornerRadii=<0 0 0 0>
     }
     tintColor {colorUseRule=UseColor_Specified, specifiedColor=<rgba 1 1 1 1>}
     uVRegion {bIsValid=false, max=<0 0>, min=<0 0>}
    }
   }
   slot {class=OverlaySlot, properties {horizontalAlignment=HAlign_Fill, verticalAlignment=VAlign_Fill}}
  }
  Text_Free {
   widgetClassPath=&CommonTextBlock
   parent=Overlay_FreeIndicator
   widget=Text_Free
   properties {
    justification=Center
    text="Free!"
    textTransformPolicy=ToUpper
    font {
     bForceMonospaced=false
     bMaterialIsStencil=false
     fontMaterial=/PlayerAugmentsUI/Materials/MI_UI_PlayerAugment_FreeIndicator_Font
     fontObject=&BurbankSmall
     letterSpacing=0
     monospacedWidth=1
     size=13
     skewAmount=0
     typefaceFontName=Black
     outlineSettings {
      bApplyOutlineToDropShadows=false
      bMiteredCorners=false
      bSeparateFillAlpha=false
      outlineMaterial=null
      outlineSize=1
      outlineColor=<rgba 1 1 1 0.25>
     }
    }
    shadowOffset=<0 0>
   }
   slot {class=OverlaySlot, properties {horizontalAlignment=HAlign_Center, verticalAlignment=VAlign_Center, padding=<ltrb 8 2 8 2>}}
  }
  Overlay_NotEnoughCurrency {
   widgetClassPath=&Overlay
   bIsVariable=true
   parent=HorizontalBox_43
   widget=Overlay_NotEnoughCurrency
   properties {visibility=Collapsed}
   slot {class=HorizontalBoxSlot, properties {horizontalAlignment=HAlign_Left, verticalAlignment=VAlign_Center, padding=<ltrb 8 0 0 0>}}
  }
  Image_AlertGlow {
   widgetClassPath=&Image
   parent=Overlay_NotEnoughCurrency
   widget=Image_AlertGlow
   properties {
    visibility=SelfHitTestInvisible
    brush {
     bIsDynamicallyLoaded=false
     drawAs=Box
     imageType=NoImage
     mirroring=NoMirror
     resourceName=None
     resourceObject=&MI_UI_PlayerAugment_Alert_Glow
     tiling=NoTile
     imageSize=<32 32>
     margin=<ltrb 0.5 0.5 0.5 0.5>
     outlineSettings {
      bUseBrushTransparency=false
      roundingType=HalfHeightRadius
      width=0
      color {colorUseRule=UseColor_Specified, specifiedColor=<rgba 0 0 0 0>}
      cornerRadii=<0 0 0 0>
     }
     tintColor {colorUseRule=UseColor_Specified, specifiedColor=<rgba 1 1 1 1>}
     uVRegion {bIsValid=false, max=<0 0>, min=<0 0>}
    }
   }
   slot {class=OverlaySlot, properties {horizontalAlignment=HAlign_Fill, verticalAlignment=VAlign_Fill, padding=<ltrb -2 -2 -2 -2>}}
  }
  Image_AlertBacking {
   widgetClassPath=&Image
   parent=Overlay_NotEnoughCurrency
   widget=Image_AlertBacking
   properties {
    visibility=SelfHitTestInvisible
    brush {
     bIsDynamicallyLoaded=false
     drawAs=Box
     imageType=NoImage
     mirroring=NoMirror
     resourceName=None
     resourceObject=&MI_UI_PlayerAugment_Alert_Backing
     tiling=NoTile
     imageSize=<16 16>
     margin=<ltrb 0.5 0.5 0.5 0.5>
     outlineSettings {
      bUseBrushTransparency=false
      roundingType=HalfHeightRadius
      width=0
      color {colorUseRule=UseColor_Specified, specifiedColor=<rgba 0 0 0 0>}
      cornerRadii=<0 0 0 0>
     }
     tintColor {colorUseRule=UseColor_Specified, specifiedColor=<rgba 1 1 1 1>}
     uVRegion {bIsValid=false, max=<0 0>, min=<0 0>}
    }
   }
   slot {class=OverlaySlot, properties {horizontalAlignment=HAlign_Fill, verticalAlignment=VAlign_Fill}}
  }
  Text_Alert {
   widgetClassPath=&CommonTextBlock
   bIsVariable=true
   parent=Overlay_NotEnoughCurrency
   widget=Text_Alert
   properties {
    text="Not enough to reroll!"
    textTransformPolicy=ToUpper
    font {
     bForceMonospaced=false
     bMaterialIsStencil=false
     fontMaterial=&MI_UI_PlayerAugment_Alert_Font
     fontObject=&BurbankSmall
     letterSpacing=0
     monospacedWidth=1
     size=13
     skewAmount=0
     typefaceFontName=Black
     outlineSettings {
      bApplyOutlineToDropShadows=false
      bMiteredCorners=false
      bSeparateFillAlpha=true
      outlineMaterial=null
      outlineSize=2
      outlineColor=<rgba 0 0 0 0>
     }
    }
    shadowOffset=<0 0>
   }
   slot {class=OverlaySlot, properties {verticalAlignment=VAlign_Center, padding=<ltrb 9 0 9 0>}}
  }
 ]
}
