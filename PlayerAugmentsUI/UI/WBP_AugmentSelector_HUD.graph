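# WBP_AugmentSelector_HUD is the player augment selection HUD, built on FortPlayerAugmentGranterWidget.
# A visibility switcher animates between the option list with reroll and the active-augments list, hiding other HUD elements while open.
# Granter events build, highlight, reroll and activate option widgets, and adapt keybinds to input type and layout to touch.
# Timers delay activation and closing; console variables gate whether stacked pending grants keep it open, with refreshes deferred during those timers.
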WidgetBlueprint /PlayerAugmentsUI/UI/WBP_AugmentSelector_HUD
animations [
 Anim_Close {
  binding_count=1
  length_seconds=0.4
  object=&self:Anim_Close
  sequence {
   bindings [Switcher_AugmentHUD {
     id=#1
     possessed_class=CommonVisibilitySwitcher
     tracks [
      {
       type=MovieScene2DTransformTrack
       sections [{
         type=MovieScene2DTransformSection
         channels [
          Translation.X {type=MovieSceneScriptingFloatChannel, keys [[0,0],[10,0],[24,-64]]}
          Translation.Y {type=MovieSceneScriptingFloatChannel, keys [[0,0],[10,-50],[24.001,384]]}
          Scale.X {type=MovieSceneScriptingFloatChannel, keys [[0,1],[10,1.125],[24.001,0.333]]}
          Scale.Y {type=MovieSceneScriptingFloatChannel, keys [[0,1],[10,1.125],[24.001,0.333]]}
         ]
        }]
      }
      {
       type=MovieSceneFloatTrack
       display_name=Render Opacity
       sections [{type=MovieSceneFloatSection, channels [None {type=MovieSceneScriptingFloatChannel, keys [[0,1],[3,1],[6,0.5],[10,0.5],[13,1],[24,0]]}]}]
      }
     ]
    }]
   display_rate {numerator=60}
  }
 }
 Anim_TranslateSwapIndicator {
  binding_count=1
  length_seconds=0.25
  object=&self:Anim_TranslateSwapIndicator
  sequence {bindings [VerticalBox_CycleInput {
     id=#2
     display_name=VerticalBox_CycleInputt
     possessed_class=VerticalBox
     tracks [{
       type=MovieScene2DTransformTrack
       sections [{type=MovieScene2DTransformSection, channels [Translation.Y {type=MovieSceneScriptingFloatChannel, keys [[0,72],[5,-156]]}]}]
      }]
    }]}
 }
 Anim_ViewMorePressed {
  binding_count=0
  length_seconds=0.1667
  object=&self:Anim_ViewMorePressed
  sequence {bindings [], display_rate {numerator=60}}
 }
 Anim_ReturnToSelectorPressed {
  binding_count=1
  length_seconds=0.1667
  object=&self:Anim_ReturnToSelectorPressed
  sequence {
   bindings [Overlay_ReadyInput {
     id=#3
     possessed_class=Overlay
     tracks [{
       type=MovieScene2DTransformTrack
       sections [{
         type=MovieScene2DTransformSection
         end_frame=10
         start_frame=0
         channels [
          Translation.X {type=MovieSceneScriptingFloatChannel, const [0,1]}
          Translation.Y {type=MovieSceneScriptingFloatChannel, const [0,1]}
          Angle {type=MovieSceneScriptingFloatChannel, const [0,1]}
          Scale.X {type=MovieSceneScriptingFloatChannel, keys [[0,1],[5,0.8],[10,1]]}
          Scale.Y {type=MovieSceneScriptingFloatChannel, keys [[0,1],[5,0.8],[10,1]]}
          Shear.X {type=MovieSceneScriptingFloatChannel, const [0,1]}
          Shear.Y {type=MovieSceneScriptingFloatChannel, const [0,1]}
         ]
        }]
      }]
    }]
   display_rate {numerator=60}
  }
 }
 Anim_SwapToSelector {
  binding_count=2
  length_seconds=0.4333
  object=&self:Anim_SwapToSelector
  sequence {
   bindings [
    Overlay_AugmentSelection {
     id=#4
     possessed_class=Overlay
     tracks [
      {
       type=MovieSceneFloatTrack
       display_name=Render Opacity
       sections [{type=MovieSceneFloatSection, channels [None {type=MovieSceneScriptingFloatChannel, keys [[0,0],[20,1]]}]}]
      }
      {
       type=MovieScene2DTransformTrack
       sections [{
         type=MovieScene2DTransformSection
         channels [
          Translation.X {type=MovieSceneScriptingFloatChannel, keys [[0,16],[20,0]]}
          Translation.Y {type=MovieSceneScriptingFloatChannel, keys [[0,16],[20,0]]}
          Scale.X {type=MovieSceneScriptingFloatChannel, keys [[0,0.9],[20,1]]}
          Scale.Y {type=MovieSceneScriptingFloatChannel, keys [[0,0.9],[20,1]]}
         ]
        }]
      }
     ]
    }
    Overlay_Reroll {
     id=#5
     possessed_class=Overlay
     tracks [
      {
       type=MovieSceneFloatTrack
       display_name=Render Opacity
       sections [{type=MovieSceneFloatSection, channels [None {type=MovieSceneScriptingFloatChannel, keys [[6,0],[26,1]]}]}]
      }
      {
       type=MovieScene2DTransformTrack
       sections [{
         type=MovieScene2DTransformSection
         channels [
          Translation.X {type=MovieSceneScriptingFloatChannel, keys [[6,16],[26,0]]}
          Translation.Y {type=MovieSceneScriptingFloatChannel, keys [[6,16],[26,0]]}
          Scale.X {type=MovieSceneScriptingFloatChannel, keys [[6,0.9],[26,1]]}
          Scale.Y {type=MovieSceneScriptingFloatChannel, keys [[6,0.9],[26,1]]}
         ]
        }]
      }
     ]
    }
   ]
   display_rate {numerator=60}
  }
 }
 Anim_SwapToList {
  binding_count=2
  length_seconds=0.3667
  object=&self:Anim_SwapToList
  sequence {
   bindings [
    WBP_ActiveAugments_List {
     id=#6
     possessed_class=WBP_ActiveAugments_List_C
     tracks [
      {
       type=MovieSceneFloatTrack
       display_name=Render Opacity
       sections [{type=MovieSceneFloatSection, channels [None {type=MovieSceneScriptingFloatChannel, keys [[0,0],[20,1]]}]}]
      }
      {
       type=MovieScene2DTransformTrack
       sections [{
         type=MovieScene2DTransformSection
         channels [
          Translation.X {type=MovieSceneScriptingFloatChannel, keys [[0,16],[20,0]]}
          Translation.Y {type=MovieSceneScriptingFloatChannel, keys [[0,16],[20,0]]}
          Scale.X {type=MovieSceneScriptingFloatChannel, keys [[0,0.9],[20,1]]}
          Scale.Y {type=MovieSceneScriptingFloatChannel, keys [[0,0.9],[20,1]]}
         ]
        }]
      }
     ]
    }
    Overlay_BottomBar {
     id=#7
     possessed_class=Overlay
     tracks [
      {
       type=MovieSceneFloatTrack
       display_name=Render Opacity
       sections [{type=MovieSceneFloatSection, channels [None {type=MovieSceneScriptingFloatChannel, keys [[2,0],[22,1]]}]}]
      }
      {
       type=MovieScene2DTransformTrack
       sections [{
         type=MovieScene2DTransformSection
         channels [
          Translation.X {type=MovieSceneScriptingFloatChannel, keys [[2,16],[22,0]]}
          Translation.Y {type=MovieSceneScriptingFloatChannel, keys [[2,16],[22,0]]}
          Scale.X {type=MovieSceneScriptingFloatChannel, keys [[2,0.9],[22,1]]}
          Scale.Y {type=MovieSceneScriptingFloatChannel, keys [[2,0.9],[22,1]]}
         ]
        }]
      }
     ]
    }
   ]
   display_rate {numerator=60}
  }
 }
]
graphs [
 ActivateSelectedAugment {nodes [
   n1 = ActivateSelectedAugment!() -> (then, AugmentIndex:int) {is=title K2Node_FunctionEntry}
   > n2 = ActivateAugment!(
    execute
    self:WBP Augment Selector Augment Object Reference="Get (a copy)"(
     Array:Array of WBP Augment Selector Augment Object References=Get AugmentOptions_Widgets() -> (AugmentOptions_Widgets:Array of WBP Augment Selector Augment Object References)
     Dimension 1=@n1
    ) -> (Output:WBP Augment Selector Augment Object Reference)
   ) -> (then, AnimationEndTime:double)
   > n3 = Set Timer by Event(Delegate=Create Event(), Time=float + float(A=@n2, B=Get ExtraDelayAfterConfirm() -> (ExtraDelayAfterConfirm:double)), bLooping=false)
   > Set ActivateAugmentTimer(ActivateAugmentTimer:Timer Handle Structure=@n3) -> (Output_Get:Timer Handle Structure)
   > Set IsNumPendingToGrantUpdated(IsNumPendingToGrantUpdated:bool=false) -> (Output_Get:bool)
  ]}
 CloseHUD {nodes [
   CloseHUD() {is=title K2Node_FunctionEntry}
   > SetHideTags!(execute, self, bHideElements:bool=false) -> (then)
   > SetVisibility(InVisibility=Collapsed)
   n1 = Get Anim_SwapToList() -> (Anim_SwapToList:Widget Animation Object Reference)
   > PlayAnimation(InAnimation=@n1, StartAtTime=GetEndTime(self=@n1), PlayMode=Reverse)
  ]}
 CreateAugmentInfos {nodes [
   n1 = CreateAugmentInfos!() -> (then, Count:int) {is=title K2Node_FunctionEntry}
   > n2 = For Loop(
    FirstIndex=Length(TargetArray:Array of WBP Augment Selector Augment Object References=Get AugmentOptions_Widgets())
    LastIndex=int - int(A=@n1)
   )
   > n3 = Set CurrentIndex(CurrentIndex:int=@n2) -> (Output_Get:int=0)
   > n4 = Create Widget!(
    execute:Exec
    Class:User Widget Class Reference=&/WBP_AugmentSelector_Augment.WBP_AugmentSelector_Augment_C
    OwningPlayer:Player Controller Object Reference
   ) -> (then:Exec, ReturnValue:WBP Augment Selector Augment Object Reference)
   > Set Augment Index!(execute, self:WBP Augment Selector Augment Object Reference=@n4, New Index:int=@n3) -> (then) {is=title K2Node_CallFunction}
   > AddChildToVerticalBox(self=Get VBox_AugmentSelectionList() -> (VBox_AugmentSelectionList:Vertical Box Object Reference), Content=@n4)
   > Add(
    TargetArray:Array of WBP Augment Selector Augment Object References=Get AugmentOptions_Widgets()
    NewItem:"WBP Augment Selector Augment Object Reference (by ref)"=@n4
   ) {is=title K2Node_CallArrayFunction}
   > n5 = Sequence()
   > Branch(Condition=Is Valid Index!(
     TargetArray:Array of Fort Player Augment Item Definition Object References=Get AvailableAugmentDefinitions() -> (AvailableAugmentDefinitions:Array of Fort Player Augment Item Definition Object References)
     IndexToTest:int=Get CurrentIndex() -> (CurrentIndex:int=0)
    ) -> (ReturnValue:bool))
   > SetAugmentInfo!(
    execute
    self:WBP Augment Selector Augment Object Reference=@n4
    InAugmentItemDefinition:Fort Player Augment Item Definition Object Reference="Get (a copy)"(
     Array:Array of Fort Player Augment Item Definition Object References=Get AvailableAugmentDefinitions()
     Dimension 1=Get CurrentIndex() -> (CurrentIndex=0)
    ) -> (Output:Fort Player Augment Item Definition Object Reference)
   ) -> (then)
   n6 = Get Count() -> (Count:int=0)
   n7 = Branch(execute=@n5.then_1, Condition="Equal (Integer)"(A=Get CurrentIndex() -> (CurrentIndex=0), B=int - int(A=@n6)))
   n8 = Create Widget(execute=@n7.else, Class=&/WBP_AugmentSelector_AugmentDivider.WBP_AugmentSelector_AugmentDivider_C) -> (ReturnValue:WBP Augment Selector Augment Divider Object Reference)
   > AddChildToVerticalBox(self=Get VBox_AugmentSelectionList(), Content=@n8)
  ]}
 EventGraph {nodes [
   Event CheckForQueuedSelections() {is=title K2Node_CustomEvent}
   > n1 = GetConsoleVariableBoolValue!(execute, VariableName:String=Fort.Augments.EnableKeepSeletorEquippedWhenStackedGrants) -> (then, ReturnValue:bool)
   > n2 = Branch(Condition=@n1)
   > n3 = GetConsoleVariableBoolValue(VariableName=Fort.Augments.EnableCheckIsPendingAugmentsUpdated)
   > n4 = Branch(Condition=@n3)
   > n5 = Branch(Condition=Get IsNumPendingToGrantUpdated() -> (IsNumPendingToGrantUpdated:bool))
   n6 = Branch(execute=[@n8.then, @n5.then, @n4.else], Condition="integer > integer"(A=Get Num Pending to Grant() -> (Num Pending to Grant:int=0)))
   > n7 = Branch(Condition=Get HasNewSelectionsPendingRefresh() -> (HasNewSelectionsPendingRefresh:bool))
   > ManualRefreshAugmentSelections()
   Set PopulateNeedsRefresh(execute=@n7.else, PopulateNeedsRefresh:bool=true) -> (Output_Get:bool)
   PlayCloseAnimation(execute=[@n6.else, @n8.else, @n2.else])
   > Set PopulateNeedsRefresh(PopulateNeedsRefresh=false)
   n8 = Branch(execute=@n5.else, Condition="integer > integer"(A=Get Num Pending to Grant() -> (Num Pending to Grant=0), B=1))
   n9 = ManualRefreshAugmentSelections() {is=title K2Node_CustomEvent}
   n10 = For Each Loop(Exec=[@n9.then, @n35.then], Array:Array of WBP Augment Selector Augment Object References=Get AugmentOptions_Widgets()) -> (Array Element:WBP Augment Selector Augment Object Reference)
   n11 = Branch(execute=[@n10.Completed, @n35.else], Condition=ShouldRefreshAugmentInfo!() -> (ShouldRefresh:bool))
   > RefreshAugmentInfo()
   > Swap Augment Focus!(execute, self, UpdatedAugmentIndex:int=Get Highlighted Augment Index() -> (Highlighted Augment Index:int=0), PlayAnimation:bool=false) -> (then)
   > Set PopulateNeedsRefresh(PopulateNeedsRefresh=false)
   > Set HasNewSelectionsPendingRefresh(HasNewSelectionsPendingRefresh:bool=false) -> (Output_Get:bool)
   Set HasNewSelectionsPendingRefresh(execute=@n11.else, HasNewSelectionsPendingRefresh=true)
   ResetState(execute=@n10.LoopBody, self:WBP Augment Selector Augment Object Reference=@n10)
   > PlayHUDSwapAnim(self:WBP Augment Selector Augment Object Reference=@n10)
   Event OnRerollAugments() {is=title K2Node_Event}
   > n12 = For Each Loop(Array=Get AugmentOptions_Widgets())
   > RerollAugment(self:WBP Augment Selector Augment Object Reference=@n12)
   n13 = Event OnAugmentSelectorOpenedStatusChanged!() -> (OutputDelegate, then, bIsOpened:bool, bViewingAugmentInventory:bool) {is=title K2Node_Event}
   > n14 = Sequence()
   > Set AugmentHUDOpened(AugmentHUDOpened:bool=@n13.bIsOpened) -> (Output_Get:bool)
   > Set Viewing Augment Active List(Viewing Augment Active List:bool=@n13.bViewingAugmentInventory) -> (Output_Get:bool)
   > n15 = Branch(Condition=Get AugmentHUDOpened() -> (AugmentHUDOpened:bool))
   > n16 = GetFortHUDContextByPlayer!(
    execute
    LocalPlayer:Local Player Object Reference=GetOwningLocalPlayer!(self:Widget Object Reference) -> (ReturnValue:Local Player Object Reference)
   ) -> (then, ReturnValue:Fort HUDContext Object Reference)
   > SetHUDElementVisibility!(
    execute
    self:Fort HUDContext Object Reference=@n16
    HUDElementTags:"Gameplay Tag Container Structure (by ref)"=Get HUDElement Tags() -> (HUDElement Tags:Gameplay Tag Container Structure)
    bHideElements:bool=true
    ReasonName:"Name (by ref)"=Hides kill streak widget while this widget is open to avoid clutter
   ) -> (then)
   > StopAnimation(InAnimation=Get Anim_Close() -> (Anim_Close:Widget Animation Object Reference))
   > ResetState(self:Self Object Reference)
   > n17 = Branch(Condition=Get Viewing Augment Active List() -> (Viewing Augment Active List:bool))
   > SetVisibility(InVisibility=SelfHitTestInvisible)
   > n18 = OnSwapToActiveList()
   SetHideTags(execute=[@n18.then, @n19.then], bHideElements=true)
   > SpawnSound2D!(
    execute
    Sound:Sound Base Object Reference=/PlayerAugmentsUI/Sound/PlayerAugments_UI_Window_Display_Cue
    VolumeMultiplier:float=1
    PitchMultiplier:float=1
    StartTime:float=0
    ConcurrencySettings:Sound Concurrency Object Reference
    bPersistAcrossLevelTransition:bool=false
    bAutoDestroy:bool=true
   ) -> (then, ReturnValue:Audio Component Object Reference)
   RefreshAugmentInfo(execute=@n17.else)
   > SetVisibility(InVisibility=SelfHitTestInvisible)
   > n19 = OnSwapToSelector()
   n20 = GetFortHUDContextByPlayer(execute=@n15.else, LocalPlayer=GetOwningLocalPlayer())
   > SetHUDElementVisibility(
    self=@n20
    HUDElementTags=Get HUDElement Tags()
    bHideElements=false
    ReasonName=Hides kill streak widget while this widget is open to avoid clutter
   )
   > PlayCloseAnimation()
   Branch(execute=@n14.then_1, Condition=IsUsingTouch(PlayerController=GetOwningPlayer()))
   > n21 = Cast To FortPlayerControllerAthena(Object=GetOwningPlayer())
   > SetHUDElementVisibilityWithContext!(
    execute
    self:Fort Player Controller Object Reference=@n21
    HUDElementTags:"Gameplay Tag Container Structure (by ref)"=Get MobileHUDTagsToHide() -> (MobileHUDTagsToHide:Gameplay Tag Container Structure)
    bHideElements:bool=@n13.bIsOpened
    ReasonName:Name=Augments Showing
    LogContext:String
   ) -> (then)
   Event OnAugmentScrollUpEvent() {is=title K2Node_Event}
   n22 = Get WBP_ActiveAugments_List() -> (WBP_ActiveAugments_List:WBP Active Augments List Object Reference)
   > ScrollUp(self:WBP Active Augments List Object Reference=@n22)
   n23 = Event OnCanSelectAugmentChanged!() -> (OutputDelegate, then, bCanSelectAugment:bool) {is=title K2Node_Event}
   > Set Can Select Augment(Can Select Augment:bool=@n23.bCanSelectAugment) -> (Output_Get:bool)
   > SetVisibility(
    self=Get Overlay_ReadyInput() -> (Overlay_ReadyInput:Overlay Object Reference)
    InVisibility=Select(Option 0=Hidden, Option 1=SelfHitTestInvisible, Index=Get Can Select Augment() -> (Can Select Augment:bool))
   )
   n24 = Event PreConstruct()
   > Branch(Condition=@n24.IsDesignTime)
   > ClearChildren(self=Get VBox_AugmentSelectionList())
   n25 = Get TestAugmentOptonsIndexNum() -> (TestAugmentOptonsIndexNum:int=0)
   > n26 = For Loop(LastIndex=@n25)
   > n27 = Create Widget(Class=&WBP_AugmentSelector_Augment_C) -> (ReturnValue:WBP Augment Selector Augment Object Reference)
   > Set Augment Index(self=@n27, New Index=@n26)
   > AddChildToVerticalBox(self=Get VBox_AugmentSelectionList(), Content=@n27)
   > n28 = Branch(Condition="Equal (Integer)"(A=@n26, B=@n25))
   n29 = Create Widget(execute=@n28.else, Class=&WBP_AugmentSelector_AugmentDivider_C) -> (ReturnValue:WBP Augment Selector Augment Divider Object Reference)
   > AddChildToVerticalBox(self=Get VBox_AugmentSelectionList(), Content=@n29)
   Event Construct()
   n30 = Get WBP_ActiveAugments_Timer() -> (WBP_ActiveAugments_Timer:WBP Active Augments Timer Object Reference)
   > Bind Event to On No Aug Left for Selection(self:WBP Active Augments Timer Object Reference=@n30, Delegate=Create Event()) {is=title K2Node_AddDelegate}
   > Bind Event to On Num Aug Pending Updated(self:WBP Active Augments Timer Object Reference=@n30, Delegate=Create Event()) {is=title K2Node_AddDelegate}
   > n31 = GetDynamicMaterial(self=Get Image_ReadyButton_Glow() -> (Image_ReadyButton_Glow:Image Object Reference))
   > SetScalarParameterValue(self=@n31, ParameterName=Anim_BackingAlpha, Value=1)
   > SetMobileVisuals()
   > CreateAugmentInfos!(execute, self, Count:int=2) -> (then) {is=title K2Node_CallFunction}
   Event Destruct()
   > SetHideTags(bHideElements=false)
   n32 = Event OnPopulateAvailableAugmentsHUD!() -> (OutputDelegate, then, AugmentItemDefinition:Array of Fort Player Augment Item Definition Object References) {is=title K2Node_Event}
   > n33 = Set AvailableAugmentDefinitions(AvailableAugmentDefinitions:Array of Fort Player Augment Item Definition Object References=@n32.AugmentItemDefinition) -> (Output_Get:Array of Fort Player Augment Item Definition Object References)
   > n34 = Branch(Condition="Equal (Integer)"(A=Length(TargetArray:Array of Fort Player Augment Item Definition Object References=@n33)))
   > Set PopulateNeedsRefresh(PopulateNeedsRefresh=false)
   n35 = Branch(execute=@n34.else, Condition=Get PopulateNeedsRefresh() -> (PopulateNeedsRefresh:bool))
   n36 = Event OnInputTypeChanged!() -> (OutputDelegate, then, NewInputType:ECommonInputType Enum) {is=title K2Node_Event}
   > n37 = Set Input Type(Input Type:ECommonInputType Enum=@n36.NewInputType) -> (Output_Get:ECommonInputType Enum)
   > UpdateSwapBoundAction()
   > UpdateScrollBoundAction!(execute, self:WBP Active Augments List Object Reference=Get WBP_ActiveAugments_List(), InputType:ECommonInputType Enum=@n37) -> (then)
   n38 = Event OnAugmentHighlighted!() -> (OutputDelegate, then, HighlightedAugmentIndex:int=0) {is=title K2Node_Event}
   > Set Highlighted Augment Index(Highlighted Augment Index:int=@n38.HighlightedAugmentIndex) -> (Output_Get:int=0)
   > n39 = Branch(Condition="integer < integer"(A=Get Highlighted Augment Index() -> (Highlighted Augment Index=0)))
   > n40 = Swap Augment Focus(UpdatedAugmentIndex=@n38.HighlightedAugmentIndex, PlayAnimation=true)
   UpdateSwapBoundAction(execute=[@n40.then, @n41.then])
   Swap Augment Focus(execute=@n39.else, UpdatedAugmentIndex=@n38.HighlightedAugmentIndex, PlayAnimation=true)
   > n41 = PlaySound2D(Sound=/PlayerAugmentsUI/Sound/PlayerAugments_UI_Selector_Tick_Cue)
   n42 = Event OnAugmentSelectionStarted!() -> (OutputDelegate, then, SelectedAugmentIndex:int=0) {is=title K2Node_Event}
   > ActivateSelectedAugment!(execute, self, AugmentIndex:int=@n42.SelectedAugmentIndex) -> (then) {is=title K2Node_CallFunction}
   n43 = Event OnAugmentSelectorToolViewSwapped!() -> (OutputDelegate, then, bViewingAugmentInventory:bool) {is=title K2Node_Event}
   > n44 = Set Viewing Augment Active List(Viewing Augment Active List=@n43.bViewingAugmentInventory)
   > n45 = Branch(Condition=@n44)
   > n46 = "Play Animation with Finished event (legacy)"!(
    execute:Exec
    Widget:User Widget Object Reference=Self-Reference()
    InAnimation:Widget Animation Object Reference=Get Anim_ViewMorePressed() -> (Anim_ViewMorePressed:Widget Animation Object Reference)
    StartAtTime:float=0
    NumLoopsToPlay:int=1
    PlayMode:EUMGSequencePlayMode Enum
    PlaybackSpeed:float=1
   ) -> (then:Exec, Result:UMGSequence Player Object Reference, Finished:Exec) {is=title K2Node_PlayAnimation}
   > PlaySound2D(Sound=/PlayerAugmentsUI/Sound/PlayerAugments_UI_Selector_Show_List_Active_Cue)
   OnSwapToActiveList(execute=@n46.Finished)
   n47 = "Play Animation with Finished event (legacy)"(
    execute=@n45.else
    Widget=Self-Reference()
    InAnimation=Get Anim_ReturnToSelectorPressed() -> (Anim_ReturnToSelectorPressed:Widget Animation Object Reference)
    StartAtTime=0
    NumLoopsToPlay=1
    PlaybackSpeed=1
   )
   > PlaySound2D(Sound=/PlayerAugmentsUI/Sound/PlayerAugments_UI_Selector_Show_List_Avail_Cue)
   OnSwapToSelector(execute=@n47.Finished)
   Event OnAugmentScrollDownEvent() {is=title K2Node_Event}
   > ScrollDown(self:WBP Active Augments List Object Reference=@n22)
  ]}
 HideReturnToSelectorInput {nodes [
   HideReturnToSelectorInput() {is=title K2Node_FunctionEntry}
   > SetVisibility(self=Get Overlay_ReadyInput(), InVisibility=Collapsed)
  ]}
 OnActivateAugmentAnimationFinished {nodes [
   OnActivateAugmentAnimationFinished() {is=title K2Node_FunctionEntry}
   > Clear and Invalidate Timer by Handle(Handle=Get ActivateAugmentTimer() -> (ActivateAugmentTimer:Timer Handle Structure))
   > Event CheckForQueuedSelections() {is=title K2Node_CallFunction}
  ]}
 OnSwapToActiveList {nodes [
   OnSwapToActiveList() {is=title K2Node_FunctionEntry}
   > SetActiveWidget(
    self=Get Switcher_AugmentHUD() -> (Switcher_AugmentHUD:Common Visibility Switcher Object Reference)
    Widget=Get VerticalBox_ActiveList() -> (VerticalBox_ActiveList:Vertical Box Object Reference)
   )
   > PlayAnimation(InAnimation=Get Anim_SwapToList())
   > OnActiveListOpened(self:WBP Active Augments List Object Reference=Get WBP_ActiveAugments_List())
  ]}
 OnSwapToSelector {nodes [
   OnSwapToSelector() {is=title K2Node_FunctionEntry}
   > SetActiveWidget(self=Get Switcher_AugmentHUD(), Widget=Get VerticalBox_SelectorHUD() -> (VerticalBox_SelectorHUD:Vertical Box Object Reference))
   > PlayAnimation(InAnimation=Get Anim_SwapToSelector() -> (Anim_SwapToSelector:Widget Animation Object Reference))
   > n1 = For Each Loop(Array=Get AugmentOptions_Widgets())
   > PlayHUDSwapAnim(self=@n1)
  ]}
 PlayCloseAnimation {nodes [
   PlayCloseAnimation() {is=title K2Node_FunctionEntry}
   > n1 = Branch(Condition=Is Timer Active by Handle!(Handle:Timer Handle Structure=Get CloseHUDTimer() -> (CloseHUDTimer:Timer Handle Structure)) -> (ReturnValue:bool))
   n2 = Get Anim_Close()
   PlayAnimation(execute=@n1.else, InAnimation=@n2)
   > n3 = Set Timer by Event(Delegate=Create Event(), Time=GetEndTime(self=@n2), bLooping=false)
   > Set CloseHUDTimer(CloseHUDTimer:Timer Handle Structure=@n3) -> (Output_Get:Timer Handle Structure)
   > SpawnSound2D(
    Sound=Get CloseSound() -> (CloseSound:Sound Base Object Reference)
    VolumeMultiplier=1
    PitchMultiplier=1
    StartTime=0
    bPersistAcrossLevelTransition=false
    bAutoDestroy=true
   )
  ]}
 RefreshArrows {nodes [
   RefreshArrows() {is=title K2Node_FunctionEntry}
   n1 = Get Highlighted Augment Index() -> (Highlighted Augment Index=0)
   > SetRenderOpacity(self=Get Image_ArrowUp() -> (Image_ArrowUp:Image Object Reference), InOpacity=SelectFloat(A=0.25, B=1, bPickA="Equal (Integer)"(A=@n1)))
   n2 = Last Index!(TargetArray:Array of WBP Augment Selector Augment Object References=Get AugmentOptions_Widgets()) -> (ReturnValue=0)
   n3 = Get Highlighted Augment Index() -> (Highlighted Augment Index=0)
   > SetRenderOpacity(
    self=Get Image_ArrowDown() -> (Image_ArrowDown:Image Object Reference)
    InOpacity=SelectFloat(A=0.25, B=1, bPickA="Equal (Integer)"(A=@n3, B=@n2))
   )
  ]}
 RefreshAugmentInfo {nodes [
   RefreshAugmentInfo() {is=title K2Node_FunctionEntry}
   n1 = Length(TargetArray:Array of WBP Augment Selector Augment Object References=Get AugmentOptions_Widgets())
   n2 = Length(TargetArray:Array of Fort Player Augment Item Definition Object References=Get AvailableAugmentDefinitions())
   > n3 = Branch(Condition=AND Boolean(A="integer >= integer"(A=@n1, B=2), B="Equal (Integer)"(A=@n1, B=@n2)))
   > n4 = For Each Loop(Array=Get AugmentOptions_Widgets())
   > SetAugmentInfo(self=@n4, InAugmentItemDefinition="Get (a copy)"(Array=Get AvailableAugmentDefinitions(), Dimension 1=@n4.Array Index))
   ClearChildren(execute=@n3.else, self=Get VBox_AugmentSelectionList())
   > Clear(TargetArray:Array of WBP Augment Selector Augment Object References=Get AugmentOptions_Widgets())
   n5 = Get AvailableAugmentDefinitions()
   > Branch(Condition="integer >= integer"(A=Length(TargetArray=@n5), B=2))
   > CreateAugmentInfos(Count=Length(TargetArray=Get AvailableAugmentDefinitions()))
  ]}
 ResetState {nodes [
   ResetState() {is=title K2Node_FunctionEntry}
   > Clear and Invalidate Timer by Handle(Handle=Get ActivateAugmentTimer())
   > Clear and Invalidate Timer by Handle(Handle=Get CloseHUDTimer())
   n1 = Get Switcher_AugmentHUD()
   > SetRenderOpacity(self=@n1, InOpacity=1)
   > SetRenderTranslation(self=@n1)
   > SetRenderScale!(execute, self:Widget Object Reference=@n1, Scale:Vector 2D Structure=(X=1,Y=1)) -> (then)
   > n2 = For Each Loop(Array=Get AugmentOptions_Widgets())
   > ResetState(self:WBP Augment Selector Augment Object Reference=@n2) {is=title K2Node_CallFunction}
   Set Highlighted Augment Index(execute=@n2.Completed, Highlighted Augment Index=-1) -> (Output_Get=0)
  ]}
 SetHideTags {nodes [
   n1 = SetHideTags!() -> (then, bHideElements:bool=~) {is=title K2Node_FunctionEntry}
   > n2 = Cast To FortPlayerController(Object=GetOwningPlayer())
   > SetHUDElementVisibilityWithContext(
    self=@n2
    HUDElementTags=Get HideSecondaryForAbilityTags() -> (HideSecondaryForAbilityTags:Gameplay Tag Container Structure)
    bHideElements=@n1
    ReasonName=Get HideSecondaryReasonName() -> (HideSecondaryReasonName:Name=None)
   )
  ]}
 SetMobileVisuals {nodes [
   SetMobileVisuals() {is=title K2Node_FunctionEntry}
   n1 = GetOwningPlayer()
   > n2 = Set IsMobileLayout(IsMobileLayout:bool=AND Boolean(A=IsMobileGame(), B=IsUsingTouch(PlayerController=@n1))) -> (Output_Get:bool)
   > SetRenderScale(Scale=Select(Option 0:Vector 2D Structure=(X=1,Y=1), Option 1:Vector 2D Structure=(X=1.2,Y=1.2), Index=@n2) -> (ReturnValue:Vector 2D Structure))
   > SetPadding!(
    execute
    self:Vertical Box Slot Object Reference=SlotAsVerticalBoxSlot(Widget=Get Overlay_AugmentSelection() -> (Overlay_AugmentSelection:Overlay Object Reference))
    InPadding:Margin Structure=Select(
     Option 0:Margin Structure=Make Margin(Left=10, Top=0, Right=10, Bottom=40)
     Option 1:Margin Structure=Make Margin(Left=10, Top=0, Right=10, Bottom=15)
     Index=Get IsMobileLayout() -> (IsMobileLayout:bool)
    ) -> (ReturnValue:Margin Structure)
   ) -> (then)
   > SetPadding(
    self:Overlay Slot Object Reference=SlotAsOverlaySlot(Widget=Get VerticalBox_CycleInput() -> (VerticalBox_CycleInput:Vertical Box Object Reference))
    InPadding=Select(Option 0=Make Margin(Left=0, Top=0, Right=0, Bottom=0), Option 1=Make Margin(Left=0, Top=45, Right=0, Bottom=0), Index=Get IsMobileLayout())
   )
  ]}
 ShouldRefreshAugmentInfo {nodes [
   ShouldRefreshAugmentInfo() {is=title K2Node_FunctionEntry}
   n1 = Is Timer Active by Handle(Handle=Get CloseHUDTimer())
   n2 = Is Timer Active by Handle(Handle=Get ActivateAugmentTimer())
   > Return Node!(execute, ShouldRefresh:bool=NOT Boolean(A=OR Boolean(A=@n1, B=@n2)))
  ]}
 Swap Augment Focus {nodes [
   Swap Augment Focus!() -> (then, UpdatedAugmentIndex:int, PlayAnimation:bool=~) {is=title K2Node_FunctionEntry}
   > n1 = For Each Loop(Array=Get AugmentOptions_Widgets())
   > n2 = Branch(Condition="Equal (Integer)"(A=@n1.Array Index, B=Get Highlighted Augment Index() -> (Highlighted Augment Index=0)))
   > Highlight!(execute, self:WBP Augment Selector Augment Object Reference=@n1, PlayAnimation:bool=Get PlayAnimation() -> (PlayAnimation:bool)) -> (then)
   Unhighlight!(execute=@n2.else, self:WBP Augment Selector Augment Object Reference=@n1, PlayAnimation:bool=Get PlayAnimation()) -> (then)
   RefreshArrows(execute=@n1.Completed) {is=title K2Node_CallFunction}
   n3 = Get AugmentOptions_Widgets()
   > n4 = Branch(Condition="integer > integer"(A=Length(TargetArray:Array of WBP Augment Selector Augment Object References=@n3), B=2))
   n5 = Get Highlighted Augment Index() -> (Highlighted Augment Index=0)
   Branch(execute=@n4.else, Condition="integer >= integer"(A=@n5))
   n6 = Get Anim_TranslateSwapIndicator() -> (Anim_TranslateSwapIndicator:Widget Animation Object Reference)
   > PlayAnimation(
    InAnimation=@n6
    StartAtTime=SelectFloat(A=0, B=GetEndTime(self=@n6), bPickA=Get PlayAnimation())
    PlayMode=Select(Option 0:EUMGSequencePlayMode Enum, Option 1:EUMGSequencePlayMode Enum=Reverse, Index:int=@n5) -> (ReturnValue:EUMGSequencePlayMode Enum)
   )
  ]}
 UpdateSelectionReadyCount {nodes [
   n1 = UpdateSelectionReadyCount!() -> (then, NumPendingToGrant:int) {is=title K2Node_FunctionEntry}
   > n2 = Set Num Pending to Grant(Num Pending to Grant:int=@n1) -> (Output_Get:int=0)
   > Set IsNumPendingToGrantUpdated(IsNumPendingToGrantUpdated=true)
   n3 = Get Text_ReadyCount() -> (Text_ReadyCount:Common Text Object Reference)
   > "SetText (Text)"(self=@n3, InText=Format Text!(Format="LOCTEXT(\"x{Num}\")", Num:Text="To Text (Integer)"(Value=@n2)) -> (Result))
   > SetVisibility(
    self=@n3
    InVisibility=Select(Option 0:ESlateVisibility Enum=Collapsed, Option 1:ESlateVisibility Enum=Visible, Index:bool="integer >= integer"(A=@n2, B=2)) -> (ReturnValue:ESlateVisibility Enum)
   )
  ]}
 UpdateSwapBoundAction {nodes [
   UpdateSwapBoundAction() {is=title K2Node_FunctionEntry}
   n1 = Get Highlighted Augment Index() -> (Highlighted Augment Index=0)
   > Branch(Condition="integer >= integer"(A=@n1))
   > SetBoundAction!(
    execute
    self:Fort Keybind Widget Object Reference=Get Keybind_CycleAction() -> (Keybind_CycleAction:Keybind Widget Object Reference)
    NewBoundAction:Name=Select!(
     MouseAndKeyboard:Name=Select!(Option 0:Name=SlotUp, Option 1:Name=SlotDown, Index:int=@n1) -> (ReturnValue)
     Gamepad:Name=Target
     Touch:Name=None
     Count:Name=None
     Index:ECommonInputType Enum=Get Input Type() -> (Input Type:ECommonInputType Enum)
    ) -> (ReturnValue:Name)
   ) -> (then)
  ]}
]
variables [
 ActivateAugmentTimer {type=TimerHandle}
 AugmentHUDOpened {type=boolean}
 AugmentOptions_Widgets {type="array<WBP_AugmentSelector_Augment_C>", default []}
 AvailableAugmentDefinitions {type="array<FortPlayerAugmentItemDefinition>", default []}
 Can Select Augment {type=boolean}
 CloseHUDTimer {type=TimerHandle}
 CloseSound {type=SoundBase, default=/PlayerAugmentsUI/Sound/PlayerAugments_UI_Window_Swipe_Away_Cue}
 ExtraDelayAfterConfirm {type=number}
 HasNewSelectionsPendingRefresh {type=boolean}
 Highlighted Augment Index {type=integer, default=-1}
 HUDElement Tags {type=GameplayTagContainer, default {gameplayTags [{tagName=Athena.LTM.Mash.StreakHide}]}}
 Input Type {type=ECommonInputType, default=MouseAndKeyboard}
 IsNumPendingToGrantUpdated {type=boolean}
 IsTopHighlighted {type=boolean}
 MobileHUDTagsToHide {
  type=GameplayTagContainer
  default {gameplayTags [{tagName=HUD.Athena.SpecialEventChallengePreview}, {tagName=HUD.TopRight.QuestUpdatesLog}]}
 }
 Num Pending to Grant {type=integer}
 PopulateNeedsRefresh {type=boolean}
 TestAugmentOptonsIndexNum {type=integer, default=2}
 Viewing Augment Active List {type=boolean}
]
widgets {
 info {
  inheritedWidgetCount=2
  parentClass=/Script/PlayerAugmentsUI.FortPlayerAugmentGranterWidget
  rootWidgetClass=&CommonVisibilitySwitcher
  widgetCount=34
 }
 widgets [
  Switcher_AugmentHUD {widgetClassPath=&CommonVisibilitySwitcher, bIsVariable=true, widget=Switcher_AugmentHUD}
  VerticalBox_SelectorHUD {
   widgetClassPath=&VerticalBox
   bIsVariable=true
   parent=Switcher_AugmentHUD
   widget=VerticalBox_SelectorHUD
   slot {class=CommonVisibilitySwitcherSlot, properties {padding=<ltrb 0 11 0 0>}}
  }
  CommonHardwareVisibilityBorder_147 {
   widgetClassPath=&CommonHardwareVisibilityBorder
   parent=VerticalBox_SelectorHUD
   widget=CommonHardwareVisibilityBorder_147
   properties {visibilityQuery {
     autoDescription=" ANY( Input.Touch )"
     tokenStreamVersion=0
     userDescription=Is Touch
     queryTokenStream [0,1,1,1,0]
     tagDictionary [{tagName=Input.Touch}]
    }}
   slot {class=VerticalBoxSlot, properties {padding=<ltrb 80 0 80 20>}}
  }
  Overlay_AugmentSelection {
   widgetClassPath=&Overlay
   bIsVariable=true
   parent=VerticalBox_SelectorHUD
   widget=Overlay_AugmentSelection
   slot {class=VerticalBoxSlot, properties {padding=<ltrb 10 0 10 40>}}
  }
  VBox_AugmentSelectionList {
   widgetClassPath=&VerticalBox
   bIsVariable=true
   parent=Overlay_AugmentSelection
   widget=VBox_AugmentSelectionList
   slot {class=OverlaySlot, properties {horizontalAlignment=HAlign_Fill, verticalAlignment=VAlign_Fill}}
  }
  VerticalBox_CycleInput {
   widgetClassPath=&VerticalBox
   bIsVariable=true
   parent=Overlay_AugmentSelection
   widget=VerticalBox_CycleInput
   slot {class=OverlaySlot, properties {verticalAlignment=VAlign_Center}}
  }
  Image_ArrowUp {
   widgetClassPath=&Image
   bIsVariable=true
   parent=VerticalBox_CycleInput
   widget=Image_ArrowUp
   properties {
    visibility=SelfHitTestInvisible
    brush {
     bIsDynamicallyLoaded=false
     drawAs=Image
     imageType=NoImage
     mirroring=NoMirror
     resourceName=None
     resourceObject=/PlayerAugmentsUI/Materials/MI_UI_AugmentArrow
     tiling=NoTile
     imageSize=<32 32>
     margin=<ltrb 0 0 0 0>
     outlineSettings {
      bUseBrushTransparency=false
      roundingType=HalfHeightRadius
      width=0
      color {colorUseRule=UseColor_Specified, specifiedColor=<rgba 0 0 0 0>}
      cornerRadii=<0 0 0 0>
     }
     tintColor {colorUseRule=UseColor_Specified, specifiedColor=<rgba 1 1 1 1>}
     uVRegion {bIsValid=false, max=<0 0>, min=<0 0>}
    }
    renderTransform {angle=180, scale=<1 1>, shear=<0 0>, translation=<0 0>}
   }
   slot {class=VerticalBoxSlot, properties {horizontalAlignment=HAlign_Center, verticalAlignment=VAlign_Center, padding=<ltrb 0 0 0 1>}}
  }
  SizeBox_0 {
   widgetClassPath=&SizeBox
   parent=VerticalBox_CycleInput
   widget=SizeBox_0
   properties {bOverride_HeightOverride=true, bOverride_WidthOverride=true, heightOverride=24, widthOverride=24}
   slot {class=VerticalBoxSlot, properties {horizontalAlignment=HAlign_Center, verticalAlignment=VAlign_Center}}
  }
  Keybind_CycleAction {
   widgetClassPath=/FortUILibrary/Input/KeybindWidget
   bIsVariable=true
   parent=SizeBox_0
   widget=Keybind_CycleAction
   properties {boundAction=Target}
   slot {class=SizeBoxSlot}
  }
  border_touch-only {
   widgetClassPath=&CommonHardwareVisibilityBorder
   parent=VerticalBox_CycleInput
   widget=border_touch-only
   properties {
    horizontalAlignment=HAlign_Center
    verticalAlignment=VAlign_Center
    visibilityQuery {
     autoDescription=" ANY( Input.Touch )"
     tokenStreamVersion=0
     userDescription=touch-only
     queryTokenStream [0,1,1,1,0]
     tagDictionary [{tagName=Input.Touch}]
    }
   }
   slot {class=VerticalBoxSlot}
  }
  Image_SwitchIcon {
   widgetClassPath=&FortLazyImage
   parent=border_touch-only
   widget=Image_SwitchIcon
   properties {brush {
     bIsDynamicallyLoaded=false
     drawAs=Image
     imageType=NoImage
     mirroring=NoMirror
     resourceName=None
     resourceObject=/PlayerAugmentsUI/MobileUI/Textures/AugmentSwap_Tile_Sprite
     tiling=NoTile
     imageSize=<55 55>
     margin=<ltrb 0 0 0 0>
     outlineSettings {
      bUseBrushTransparency=false
      roundingType=HalfHeightRadius
      width=0
      color {colorUseRule=UseColor_Specified, specifiedColor=<rgba 0 0 0 0>}
      cornerRadii=<0 0 0 0>
     }
     tintColor {colorUseRule=UseColor_Specified, specifiedColor=<rgba 1 1 1 1>}
     uVRegion {bIsValid=false, max=<0 0>, min=<0 0>}
    }}
   slot {class=BorderSlot, properties {horizontalAlignment=HAlign_Center, verticalAlignment=VAlign_Center, padding=<ltrb 0 0 0 0>}}
  }
  Image_ArrowDown {
   widgetClassPath=&Image
   bIsVariable=true
   parent=VerticalBox_CycleInput
   widget=Image_ArrowDown
   properties {
    visibility=SelfHitTestInvisible
    brush {
     bIsDynamicallyLoaded=false
     drawAs=Image
     imageType=NoImage
     mirroring=NoMirror
     resourceName=None
     resourceObject=&MI_UI_AugmentArrow
     tiling=NoTile
     imageSize=<32 32>
     margin=<ltrb 0 0 0 0>
     outlineSettings {
      bUseBrushTransparency=false
      roundingType=HalfHeightRadius
      width=0
      color {colorUseRule=UseColor_Specified, specifiedColor=<rgba 0 0 0 0>}
      cornerRadii=<0 0 0 0>
     }
     tintColor {colorUseRule=UseColor_Specified, specifiedColor=<rgba 1 1 1 1>}
     uVRegion {bIsValid=false, max=<0 0>, min=<0 0>}
    }
   }
   slot {class=VerticalBoxSlot, properties {horizontalAlignment=HAlign_Center, verticalAlignment=VAlign_Center}}
  }
  Overlay_Reroll {
   widgetClassPath=&Overlay
   bInherited=true
   parent=VerticalBox_SelectorHUD
   widget=Overlay_Reroll
   slot {class=VerticalBoxSlot, properties {padding=<ltrb 80 0 80 0>}}
  }
  Backing_Glow {
   widgetClassPath=&Image
   parent=Overlay_Reroll
   widget=Backing_Glow
   properties {
    visibility=SelfHitTestInvisible
    brush {
     bIsDynamicallyLoaded=false
     drawAs=Box
     imageType=NoImage
     mirroring=NoMirror
     resourceName=None
     resourceObject=/PlayerAugmentsUI/Materials/MI_UI_PlayerAugment_Panel_Glow
     tiling=NoTile
     imageSize=<512 64>
     margin=<ltrb 0.5 0.5 0.5 0.5>
     outlineSettings {
      bUseBrushTransparency=false
      roundingType=HalfHeightRadius
      width=0
      color {colorUseRule=UseColor_Specified, specifiedColor=<rgba 0 0 0 0>}
      cornerRadii=<0 0 0 0>
     }
     tintColor {colorUseRule=UseColor_Specified, specifiedColor=<rgba 1 1 1 1>}
     uVRegion {bIsValid=false, max=<0 0>, min=<0 0>}
    }
   }
   slot {class=OverlaySlot, properties {horizontalAlignment=HAlign_Fill, verticalAlignment=VAlign_Fill, padding=<ltrb 0 -24 0 -24>}}
  }
  Image_RerollBG {
   widgetClassPath=&Image
   parent=Overlay_Reroll
   widget=Image_RerollBG
   properties {
    visibility=SelfHitTestInvisible
    brush {
     bIsDynamicallyLoaded=false
     drawAs=Image
     imageType=NoImage
     mirroring=NoMirror
     resourceName=None
     resourceObject=/PlayerAugmentsUI/Materials/MI_UI_PlayerAugment_BackingPanel
     tiling=NoTile
     imageSize=<32 32>
     margin=<ltrb 0 0 0 0>
     outlineSettings {
      bUseBrushTransparency=false
      roundingType=HalfHeightRadius
      width=0
      color {colorUseRule=UseColor_Specified, specifiedColor=<rgba 0 0 0 0>}
      cornerRadii=<0 0 0 0>
     }
     tintColor {colorUseRule=UseColor_Specified, specifiedColor=<rgba 1 1 1 1>}
     uVRegion {bIsValid=false, max=<0 0>, min=<0 0>}
    }
   }
   slot {class=OverlaySlot, properties {horizontalAlignment=HAlign_Fill, verticalAlignment=VAlign_Fill}}
  }
  AugmentSelector_Reroll {
   widgetClassPath=&/WBP_AugmentSelector_Reroll
   bIsVariable=true
   parent=Overlay_Reroll
   widget=AugmentSelector_Reroll
   slot {class=OverlaySlot, properties {horizontalAlignment=HAlign_Center, verticalAlignment=VAlign_Center, padding=<ltrb 0 0 0 2>}}
  }
  VerticalBox_ActiveList {
   widgetClassPath=&VerticalBox
   bIsVariable=true
   parent=Switcher_AugmentHUD
   widget=VerticalBox_ActiveList
   slot {class=CommonVisibilitySwitcherSlot}
  }
  WBP_ActiveAugments_List {
   widgetClassPath=/PlayerAugmentsUI/UI/ActiveAugmentsPanel/WBP_ActiveAugments_List
   bIsVariable=true
   parent=VerticalBox_ActiveList
   widget=WBP_ActiveAugments_List
   slot {class=VerticalBoxSlot, properties {horizontalAlignment=HAlign_Center, verticalAlignment=VAlign_Center, padding=<ltrb 0 40 0 20>}}
  }
  Overlay_BottomBar {
   widgetClassPath=&Overlay
   parent=VerticalBox_ActiveList
   widget=Overlay_BottomBar
   slot {class=VerticalBoxSlot, properties {verticalAlignment=VAlign_Center}}
  }
  Image_ReturnBG {
   widgetClassPath=&Image
   parent=Overlay_BottomBar
   widget=Image_ReturnBG
   properties {
    visibility=SelfHitTestInvisible
    brush {
     bIsDynamicallyLoaded=false
     drawAs=Image
     imageType=NoImage
     mirroring=NoMirror
     resourceName=None
     resourceObject=&MI_UI_PlayerAugment_BackingPanel
     tiling=NoTile
     imageSize=<32 32>
     margin=<ltrb 0 0 0 0>
     outlineSettings {
      bUseBrushTransparency=false
      roundingType=HalfHeightRadius
      width=0
      color {colorUseRule=UseColor_Specified, specifiedColor=<rgba 0 0 0 0>}
      cornerRadii=<0 0 0 0>
     }
     tintColor {colorUseRule=UseColor_Specified, specifiedColor=<rgba 1 1 1 1>}
     uVRegion {bIsValid=false, max=<0 0>, min=<0 0>}
    }
   }
   slot {class=OverlaySlot, properties {horizontalAlignment=HAlign_Fill, verticalAlignment=VAlign_Fill}}
  }
  Image_Backing_Glow_1 {
   widgetClassPath=&Image
   parent=Overlay_BottomBar
   widget=Image_Backing_Glow_1
   properties {
    visibility=SelfHitTestInvisible
    brush {
     bIsDynamicallyLoaded=false
     drawAs=Box
     imageType=NoImage
     mirroring=NoMirror
     resourceName=None
     resourceObject=&MI_UI_PlayerAugment_Panel_Glow
     tiling=NoTile
     imageSize=<412 64>
     margin=<ltrb 0.5 0.5 0.5 0.5>
     outlineSettings {
      bUseBrushTransparency=false
      roundingType=HalfHeightRadius
      width=0
      color {colorUseRule=UseColor_Specified, specifiedColor=<rgba 0 0 0 0>}
      cornerRadii=<0 0 0 0>
     }
     tintColor {colorUseRule=UseColor_Specified, specifiedColor=<rgba 1 1 1 1>}
     uVRegion {bIsValid=false, max=<0 0>, min=<0 0>}
    }
   }
   slot {class=OverlaySlot, properties {horizontalAlignment=HAlign_Fill, verticalAlignment=VAlign_Fill, padding=<ltrb 0 -24 0 -24>}}
  }
  Image_Backing_Glow_2 {
   widgetClassPath=&Image
   parent=Overlay_BottomBar
   widget=Image_Backing_Glow_2
   properties {
    visibility=SelfHitTestInvisible
    brush {
     bIsDynamicallyLoaded=false
     drawAs=Box
     imageType=NoImage
     mirroring=NoMirror
     resourceName=None
     resourceObject=/PlayerAugmentsUI/Materials/MI_UI_PlayerAugment_Font_Glow
     tiling=NoTile
     imageSize=<88 88>
     margin=<ltrb 0.5 0.5 0.5 0.5>
     outlineSettings {
      bUseBrushTransparency=false
      roundingType=HalfHeightRadius
      width=0
      color {colorUseRule=UseColor_Specified, specifiedColor=<rgba 0 0 0 0>}
      cornerRadii=<0 0 0 0>
     }
     tintColor {colorUseRule=UseColor_Specified, specifiedColor=<rgba 1 1 1 1>}
     uVRegion {bIsValid=false, max=<0 0>, min=<0 0>}
    }
   }
   slot {class=OverlaySlot, properties {horizontalAlignment=HAlign_Fill, verticalAlignment=VAlign_Fill, padding=<ltrb 0 -16 0 -16>}}
  }
  HorizontalBox_0 {
   widgetClassPath=&HorizontalBox
   parent=Overlay_BottomBar
   widget=HorizontalBox_0
   slot {class=OverlaySlot, properties {horizontalAlignment=HAlign_Center, verticalAlignment=VAlign_Fill}}
  }
  WBP_ActiveAugments_Timer {
   widgetClassPath=/PlayerAugmentsUI/UI/ActiveAugmentsPanel/WBP_ActiveAugments_Timer
   bIsVariable=true
   parent=HorizontalBox_0
   widget=WBP_ActiveAugments_Timer
   slot {class=HorizontalBoxSlot, properties {horizontalAlignment=HAlign_Left, verticalAlignment=VAlign_Center}}
  }
  Overlay_ReadyInput {
   widgetClassPath=&Overlay
   bIsVariable=true
   parent=HorizontalBox_0
   widget=Overlay_ReadyInput
   properties {visibility=Hidden}
   slot {class=HorizontalBoxSlot, properties {horizontalAlignment=HAlign_Center, verticalAlignment=VAlign_Center, padding=<ltrb 25 0 0 0>}}
  }
  HardwareVisibilityBorder_TouchOnlyLeaveActiveAugments {
   widgetClassPath=&CommonHardwareVisibilityBorder
   parent=Overlay_ReadyInput
   widget=HardwareVisibilityBorder_TouchOnlyLeaveActiveAugments
   properties {
    visibleType=Visible
    visibilityQuery {
     autoDescription=" ANY( Input.Touch )"
     tokenStreamVersion=0
     userDescription=touch-only
     queryTokenStream [0,1,1,1,0]
     tagDictionary [{tagName=Input.Touch}]
    }
   }
   slot {class=OverlaySlot, properties {horizontalAlignment=HAlign_Fill, verticalAlignment=VAlign_Fill}}
  }
  FortTouchUserWidget_LeaveActiveAugments {
   widgetClassPath=/Script/FortniteUI.FortTouchUserWidget
   bIsVariable=true
   parent=HardwareVisibilityBorder_TouchOnlyLeaveActiveAugments
   widget=FortTouchUserWidget_LeaveActiveAugments
   properties {visibility=Visible, actionNames [AugmentSelectionViewSwap]}
   slot {class=BorderSlot, properties {padding=<ltrb 0 0 0 0>}}
  }
  Image_Ready_SolidBG {
   widgetClassPath=&Image
   parent=Overlay_ReadyInput
   widget=Image_Ready_SolidBG
   properties {
    visibility=SelfHitTestInvisible
    brush {
     bIsDynamicallyLoaded=false
     drawAs=Box
     imageType=NoImage
     mirroring=NoMirror
     resourceName=None
     resourceObject=/PlayerAugmentsUI/Materials/MI_UI_PlayerAugment_ActivationButton_Backing
     tiling=NoTile
     imageSize=<48 48>
     margin=<ltrb 0.5 0.5 0.5 0.5>
     outlineSettings {
      bUseBrushTransparency=false
      roundingType=HalfHeightRadius
      width=0
      color {colorUseRule=UseColor_Specified, specifiedColor=<rgba 0 0 0 0>}
      cornerRadii=<0 0 0 0>
     }
     tintColor {colorUseRule=UseColor_Specified, specifiedColor=<rgba 1 1 1 1>}
     uVRegion {bIsValid=false, max=<0 0>, min=<0 0>}
    }
   }
   slot {class=OverlaySlot, properties {horizontalAlignment=HAlign_Fill, verticalAlignment=VAlign_Fill}}
  }
  Image_ReadyButton_Glow {
   widgetClassPath=&Image
   bIsVariable=true
   parent=Overlay_ReadyInput
   widget=Image_ReadyButton_Glow
   properties {
    visibility=SelfHitTestInvisible
    brush {
     bIsDynamicallyLoaded=false
     drawAs=Box
     imageType=NoImage
     mirroring=NoMirror
     resourceName=None
     resourceObject=/PlayerAugmentsUI/Materials/MI_UI_PlayerAugment_Button_Glow
     tiling=NoTile
     imageSize=<104 104>
     margin=<ltrb 0.5 0.5 0.5 0.5>
     outlineSettings {
      bUseBrushTransparency=false
      roundingType=HalfHeightRadius
      width=0
      color {colorUseRule=UseColor_Specified, specifiedColor=<rgba 0 0 0 0>}
      cornerRadii=<0 0 0 0>
     }
     tintColor {colorUseRule=UseColor_Specified, specifiedColor=<rgba 1 1 1 1>}
     uVRegion {bIsValid=false, max=<0 0>, min=<0 0>}
    }
   }
   slot {class=OverlaySlot, properties {horizontalAlignment=HAlign_Fill, verticalAlignment=VAlign_Fill, padding=<ltrb -32 -32 -32 -32>}}
  }
  HorizontalBox_2 {
   widgetClassPath=&HorizontalBox
   parent=Overlay_ReadyInput
   widget=HorizontalBox_2
   slot {class=OverlaySlot, properties {horizontalAlignment=HAlign_Center, verticalAlignment=VAlign_Center, padding=<ltrb 20 5 20 5>}}
  }
  Keybind_SwapAction {
   widgetClassPath=&KeybindWidget
   bIsVariable=true
   parent=HorizontalBox_2
   widget=Keybind_SwapAction
   properties {boundAction=AugmentSelectionViewSwap}
   slot {class=HorizontalBoxSlot, properties {horizontalAlignment=HAlign_Center, verticalAlignment=VAlign_Center, padding=<ltrb 0 0 5 0>}}
  }
  Text_SelectionReady {
   widgetClassPath=&CommonTextBlock
   parent=HorizontalBox_2
   widget=Text_SelectionReady
   properties {
    mobileFontSizeMultiplier=1.4
    text="SELECTION READY!"
    font {
     bForceMonospaced=false
     bMaterialIsStencil=false
     fontMaterial=/PlayerAugmentsUI/Materials/MI_UI_PlayerAugment_Button_Font
     fontObject=&BurbankSmall
     letterSpacing=0
     monospacedWidth=1
     size=16
     skewAmount=0
     typefaceFontName=Black
     outlineSettings {
      bApplyOutlineToDropShadows=false
      bMiteredCorners=false
      bSeparateFillAlpha=true
      outlineMaterial=&M-FontShadow
      outlineSize=2
      outlineColor=<rgba 0 0.1667 1 0.75>
     }
    }
    shadowColorAndOpacity=<rgba 0 0.8333 1 0.125>
    shadowOffset=<4 2>
   }
   slot {class=HorizontalBoxSlot, properties {horizontalAlignment=HAlign_Center, verticalAlignment=VAlign_Center}}
  }
  Text_ReadyCount {
   widgetClassPath=&CommonTextBlock
   bIsVariable=true
   parent=HorizontalBox_2
   widget=Text_ReadyCount
   properties {
    mobileFontSizeMultiplier=1.4
    text="x##"
    textTransformPolicy=ToLower
    font {
     bForceMonospaced=false
     bMaterialIsStencil=false
     fontMaterial=&MI_UI_PlayerAugment_Button_Font
     fontObject=&BurbankSmall
     letterSpacing=0
     monospacedWidth=1
     size=16
     skewAmount=0
     typefaceFontName=Black
     outlineSettings {
      bApplyOutlineToDropShadows=false
      bMiteredCorners=false
      bSeparateFillAlpha=true
      outlineMaterial=&M-FontShadow
      outlineSize=2
      outlineColor=<rgba 0 0.1667 1 0.75>
     }
    }
    shadowColorAndOpacity=<rgba 0 0.8333 1 0.125>
    shadowOffset=<4 2>
   }
   slot {class=HorizontalBoxSlot, properties {horizontalAlignment=HAlign_Center, verticalAlignment=VAlign_Center}}
  }
  TutorialHighlight {widgetClassPath=&TutorialHighlight}
 ]
}
