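# Augment info panel, presumably for interaction indicators, showing a player augment's icon, name, category and description over glowing material-driven backing panels.
# SetAugmentInfo stores the augment definition, calls UpdateAugment on the FortPlayerAugmentInfoEntryWidget parent, then passes the definition's HUD effect image to SetAugmentIcon.
# SetAugmentIcon async-loads that soft texture and writes it to the Icon parameter of Image_AugmentIcon's dynamic material.
# Inherited rarity, category indicator and lender widgets are absent from this layout.
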WidgetBlueprint /PlayerAugmentsUI/UI/WBP_Augments_InteractionIndicatorInfo
graphs [
 EventGraph {nodes [
   n1 = SetAugmentIcon!() -> (OutputDelegate, then, Asset:Soft Object Reference) {is=title K2Node_CustomEvent}
   > n2 = Async Load Asset(Asset=@n1.Asset)
   n3 = Cast To Texture2D(execute=[@n2.Completed, @n2.then], Object=@n2)
   > n4 = GetDynamicMaterial(self=Get Image_AugmentIcon() -> (Image_AugmentIcon:Image Object Reference))
   > SetTextureParameterValue(self=@n4, ParameterName=Icon, Value=@n3)
  ]}
 SetAugmentInfo {nodes [
   n1 = SetAugmentInfo!() -> (then, InAugmentItemDefinition:Fort Player Augment Item Definition Object Reference) {is=title K2Node_FunctionEntry}
   > n2 = Set Augment Item Definition(Augment Item Definition:Fort Player Augment Item Definition Object Reference=@n1) -> (Output_Get:Fort Player Augment Item Definition Object Reference)
   > UpdateAugment!(
    execute
    self:Fort Player Augment Info Entry Widget Object Reference
    AugmentItemDefinition:Fort Player Augment Item Definition Object Reference=@n2
   ) -> (then)
   > SetAugmentIcon!(
    execute
    self
    Asset:Soft Object Reference=GetHUDEffectImage!(self:Fort Player Augment Item Definition Object Reference=@n2) -> (ReturnValue:Texture 2D Soft Object Reference)
   ) -> (then) {is=title K2Node_CallFunction}
  ]}
]
variables [Augment Item Definition {type=FortPlayerAugmentItemDefinition, default=null}]
widgets {
 info {
  inheritedWidgetCount=10
  parentClass=/Script/PlayerAugmentsUI.FortPlayerAugmentInfoEntryWidget
  rootWidgetClass=&Overlay
  widgetCount=38
 }
 widgets [
  All {widgetClassPath=&Overlay, widget=All}
  WidthConrol {
   widgetClassPath=&Spacer
   parent=All
   widget=WidthConrol
   properties {size=<288 85.3333>}
   slot {class=OverlaySlot, properties {horizontalAlignment=HAlign_Fill}}
  }
  Backing_Glow {
   widgetClassPath=&Image
   parent=All
   widget=Backing_Glow
   properties {
    visibility=SelfHitTestInvisible
    brush {
     bIsDynamicallyLoaded=false
     drawAs=Box
     imageType=NoImage
     mirroring=NoMirror
     resourceName=None
     resourceObject=/PlayerAugmentsUI/Materials/MI_UI_PlayerAugment_Panel_Glow
     tiling=NoTile
     imageSize=<300 100>
     margin=<ltrb 0.5 0.5 0.5 0.5>
     outlineSettings {
      bUseBrushTransparency=false
      roundingType=HalfHeightRadius
      width=0
      color {colorUseRule=UseColor_Specified, specifiedColor=<rgba 0 0 0 0>}
      cornerRadii=<0 0 0 0>
     }
     tintColor {colorUseRule=UseColor_Specified, specifiedColor=<rgba 1 1 1 1>}
     uVRegion {bIsValid=false, max=<0 0>, min=<0 0>}
    }
   }
   slot {class=OverlaySlot, properties {horizontalAlignment=HAlign_Fill, verticalAlignment=VAlign_Fill}}
  }
  Backing_Panel {
   widgetClassPath=&Image
   parent=All
   widget=Backing_Panel
   properties {
    visibility=SelfHitTestInvisible
    brush {
     bIsDynamicallyLoaded=false
     drawAs=Box
     imageType=NoImage
     mirroring=NoMirror
     resourceName=None
     resourceObject=/PlayerAugmentsUI/Materials/MI_UI_PlayerAugment_BackingPanel
     tiling=NoTile
     imageSize=<64 64>
     margin=<ltrb 0.5 0.5 0.5 0.5>
     outlineSettings {
      bUseBrushTransparency=false
      roundingType=HalfHeightRadius
      width=0
      color {colorUseRule=UseColor_Specified, specifiedColor=<rgba 0 0 0 0>}
      cornerRadii=<0 0 0 0>
     }
     tintColor {colorUseRule=UseColor_Specified, specifiedColor=<rgba 1 1 1 1>}
     uVRegion {bIsValid=false, max=<0 0>, min=<0 0>}
    }
   }
   slot {class=OverlaySlot, properties {horizontalAlignment=HAlign_Fill, verticalAlignment=VAlign_Fill}}
  }
  Content {widgetClassPath=&HorizontalBox, parent=All, widget=Content, slot {class=OverlaySlot, properties {verticalAlignment=VAlign_Fill}}}
  Icon {
   widgetClassPath=&Overlay
   parent=Content
   widget=Icon
   slot {class=HorizontalBoxSlot, properties {horizontalAlignment=HAlign_Center, verticalAlignment=VAlign_Center}}
  }
  IconSize {widgetClassPath=&Spacer, parent=Icon, widget=IconSize, properties {size=<42.6667 42.6667>}, slot {class=OverlaySlot}}
  Icon_BackingPanel_Reflection {
   widgetClassPath=&Image
   parent=Icon
   widget=Icon_BackingPanel_Reflection
   properties {
    visibility=SelfHitTestInvisible
    brush {
     bIsDynamicallyLoaded=false
     drawAs=Box
     imageType=NoImage
     mirroring=NoMirror
     resourceName=None
     resourceObject=/PlayerAugmentsUI/Materials/MI_UI_PlayerAugment_BackingPanel_NoNoise_Reflection
     tiling=NoTile
     imageSize=<64 64>
     margin=<ltrb 0.5 0.5 0.5 0.5>
     outlineSettings {
      bUseBrushTransparency=false
      roundingType=HalfHeightRadius
      width=0
      color {colorUseRule=UseColor_Specified, specifiedColor=<rgba 0 0 0 0>}
      cornerRadii=<0 0 0 0>
     }
     tintColor {colorUseRule=UseColor_Specified, specifiedColor=<rgba 1 1 1 1>}
     uVRegion {bIsValid=false, max=<0 0>, min=<0 0>}
    }
    renderTransform {angle=0, scale=<1 1>, shear=<0 0>, translation=<10 16>}
   }
   slot {class=OverlaySlot, properties {horizontalAlignment=HAlign_Fill, verticalAlignment=VAlign_Fill}}
  }
  Icon_ConfirmationFlare {
   widgetClassPath=&Image
   parent=Icon
   widget=Icon_ConfirmationFlare
   properties {brush {
     bIsDynamicallyLoaded=false
     drawAs=Image
     imageType=NoImage
     mirroring=NoMirror
     resourceName=None
     resourceObject=/PlayerAugmentsUI/Materials/M_UI_PlayerAugment_Gradient_Flare
     tiling=NoTile
     imageSize=<32 32>
     margin=<ltrb 0 0 0 0>
     outlineSettings {
      bUseBrushTransparency=false
      roundingType=HalfHeightRadius
      width=0
      color {colorUseRule=UseColor_Specified, specifiedColor=<rgba 0 0 0 0>}
      cornerRadii=<0 0 0 0>
     }
     tintColor {colorUseRule=UseColor_Specified, specifiedColor=<rgba 1 1 1 1>}
     uVRegion {bIsValid=false, max=<0 0>, min=<0 0>}
    }}
   slot {class=OverlaySlot, properties {horizontalAlignment=HAlign_Fill, verticalAlignment=VAlign_Fill}}
  }
  Icon_BackingPanel {
   widgetClassPath=&Image
   parent=Icon
   widget=Icon_BackingPanel
   properties {
    visibility=SelfHitTestInvisible
    brush {
     bIsDynamicallyLoaded=false
     drawAs=Box
     imageType=NoImage
     mirroring=NoMirror
     resourceName=None
     resourceObject=/PlayerAugmentsUI/Materials/MI_UI_PlayerAugment_BackingPanel_NoNoise_Icon
     tiling=NoTile
     imageSize=<64 64>
     margin=<ltrb 0.5 0.5 0.5 0.5>
     outlineSettings {
      bUseBrushTransparency=false
      roundingType=HalfHeightRadius
      width=0
      color {colorUseRule=UseColor_Specified, specifiedColor=<rgba 0 0 0 0>}
      cornerRadii=<0 0 0 0>
     }
     tintColor {colorUseRule=UseColor_Specified, specifiedColor=<rgba 1 1 1 1>}
     uVRegion {bIsValid=false, max=<0 0>, min=<0 0>}
    }
   }
   slot {class=OverlaySlot, properties {horizontalAlignment=HAlign_Fill, verticalAlignment=VAlign_Fill}}
  }
  Icon_BackingPanel_UnlockedFX {
   widgetClassPath=&Image
   parent=Icon
   widget=Icon_BackingPanel_UnlockedFX
   properties {
    visibility=SelfHitTestInvisible
    brush {
     bIsDynamicallyLoaded=false
     drawAs=Box
     imageType=NoImage
     mirroring=NoMirror
     resourceName=None
     resourceObject=/PlayerAugmentsUI/Materials/M_UI_AugmentUnlocked_Pattern
     tiling=NoTile
     imageSize=<64 64>
     margin=<ltrb 0.5 0.5 0.5 0.5>
     outlineSettings {
      bUseBrushTransparency=false
      roundingType=HalfHeightRadius
      width=0
      color {colorUseRule=UseColor_Specified, specifiedColor=<rgba 0 0 0 0>}
      cornerRadii=<0 0 0 0>
     }
     tintColor {colorUseRule=UseColor_Specified, specifiedColor=<rgba 1 1 1 1>}
     uVRegion {bIsValid=false, max=<0 0>, min=<0 0>}
    }
   }
   slot {class=OverlaySlot, properties {horizontalAlignment=HAlign_Fill, verticalAlignment=VAlign_Fill}}
  }
  Icon_Border_UnlockedGlow {
   widgetClassPath=&Image
   parent=Icon
   widget=Icon_Border_UnlockedGlow
   properties {
    visibility=SelfHitTestInvisible
    brush {
     bIsDynamicallyLoaded=false
     drawAs=Box
     imageType=NoImage
     mirroring=NoMirror
     resourceName=None
     resourceObject=/PlayerAugmentsUI/Materials/MI_UI_PlayerAugment_Unlocked_IconBacking_Glow_Border
     tiling=NoTile
     imageSize=<64 64>
     margin=<ltrb 0.5 0.5 0.5 0.5>
     outlineSettings {
      bUseBrushTransparency=false
      roundingType=HalfHeightRadius
      width=0
      color {colorUseRule=UseColor_Specified, specifiedColor=<rgba 0 0 0 0>}
      cornerRadii=<0 0 0 0>
     }
     tintColor {colorUseRule=UseColor_Specified, specifiedColor=<rgba 1 1 1 1>}
     uVRegion {bIsValid=false, max=<0 0>, min=<0 0>}
    }
   }
   slot {class=OverlaySlot, properties {horizontalAlignment=HAlign_Fill, verticalAlignment=VAlign_Fill, padding=<ltrb -16 -16 -16 -16>}}
  }
  Icon_Border_ConfirmationPulse {
   widgetClassPath=&Image
   parent=Icon
   widget=Icon_Border_ConfirmationPulse
   properties {
    visibility=SelfHitTestInvisible
    brush {
     bIsDynamicallyLoaded=false
     drawAs=Box
     imageType=NoImage
     mirroring=NoMirror
     resourceName=None
     resourceObject=/PlayerAugmentsUI/Materials/MI_UI_PlayerAugmment_Border_ConfirmationPulse
     tiling=NoTile
     imageSize=<64 64>
     margin=<ltrb 0.5 0.5 0.5 0.5>
     outlineSettings {
      bUseBrushTransparency=false
      roundingType=HalfHeightRadius
      width=0
      color {colorUseRule=UseColor_Specified, specifiedColor=<rgba 0 0 0 0>}
      cornerRadii=<0 0 0 0>
     }
     tintColor {colorUseRule=UseColor_Specified, specifiedColor=<rgba 1 1 1 1>}
     uVRegion {bIsValid=false, max=<0 0>, min=<0 0>}
    }
   }
   slot {class=OverlaySlot, properties {horizontalAlignment=HAlign_Fill, verticalAlignment=VAlign_Fill}}
  }
  Icon_Border {
   widgetClassPath=&Image
   parent=Icon
   widget=Icon_Border
   properties {
    visibility=SelfHitTestInvisible
    brush {
     bIsDynamicallyLoaded=false
     drawAs=Box
     imageType=NoImage
     mirroring=NoMirror
     resourceName=None
     resourceObject=/PlayerAugmentsUI/Materials/MI_UI_PlayerAugment_Border
     tiling=NoTile
     imageSize=<64 64>
     margin=<ltrb 0.5 0.5 0.5 0.5>
     outlineSettings {
      bUseBrushTransparency=false
      roundingType=HalfHeightRadius
      width=0
      color {colorUseRule=UseColor_Specified, specifiedColor=<rgba 0 0 0 0>}
      cornerRadii=<0 0 0 0>
     }
     tintColor {colorUseRule=UseColor_Specified, specifiedColor=<rgba 1 1 1 1>}
     uVRegion {bIsValid=false, max=<0 0>, min=<0 0>}
    }
    renderTransformPivot=<1 1>
   }
   slot {class=OverlaySlot, properties {horizontalAlignment=HAlign_Fill, verticalAlignment=VAlign_Fill}}
  }
  Icon_Border_Unlocked {
   widgetClassPath=&Image
   parent=Icon
   widget=Icon_Border_Unlocked
   properties {
    visibility=SelfHitTestInvisible
    brush {
     bIsDynamicallyLoaded=false
     drawAs=Box
     imageType=NoImage
     mirroring=NoMirror
     resourceName=None
     resourceObject=/PlayerAugmentsUI/Materials/MI_UI_PlayerAugment_Border_Unlcoked
     tiling=NoTile
     imageSize=<64 64>
     margin=<ltrb 0.5 0.5 0.5 0.5>
     outlineSettings {
      bUseBrushTransparency=false
      roundingType=HalfHeightRadius
      width=0
      color {colorUseRule=UseColor_Specified, specifiedColor=<rgba 0 0 0 0>}
      cornerRadii=<0 0 0 0>
     }
     tintColor {colorUseRule=UseColor_Specified, specifiedColor=<rgba 1 1 1 1>}
     uVRegion {bIsValid=false, max=<0 0>, min=<0 0>}
    }
    renderTransformPivot=<1 1>
   }
   slot {class=OverlaySlot, properties {horizontalAlignment=HAlign_Fill, verticalAlignment=VAlign_Fill}}
  }
  Image_AugmentIcon {
   widgetClassPath=&Image
   bIsVariable=true
   parent=Icon
   widget=Image_AugmentIcon
   properties {
    visibility=SelfHitTestInvisible
    brush {
     bIsDynamicallyLoaded=false
     drawAs=Image
     imageType=NoImage
     mirroring=NoMirror
     resourceName=None
     resourceObject=/PlayerAugmentsUI/Materials/MI_UI_PlayerAugment_HUD_Icon
     tiling=NoTile
     imageSize=<96 96>
     margin=<ltrb 0 0 0 0>
     outlineSettings {
      bUseBrushTransparency=false
      roundingType=HalfHeightRadius
      width=0
      color {colorUseRule=UseColor_Specified, specifiedColor=<rgba 0 0 0 0>}
      cornerRadii=<0 0 0 0>
     }
     tintColor {colorUseRule=UseColor_Specified, specifiedColor=<rgba 1 1 1 1>}
     uVRegion {bIsValid=false, max=<0 0>, min=<0 0>}
    }
   }
   slot {class=OverlaySlot, properties {horizontalAlignment=HAlign_Center, verticalAlignment=VAlign_Center}}
  }
  Text {widgetClassPath=&VerticalBox, parent=Content, widget=Text, slot {class=HorizontalBoxSlot, properties {padding=<ltrb 24 32 0 0>}}}
  Overlay_0 {
   widgetClassPath=&Overlay
   parent=Text
   widget=Overlay_0
   slot {class=VerticalBoxSlot, properties {horizontalAlignment=HAlign_Left, verticalAlignment=VAlign_Center}}
  }
  Overlay_Header {widgetClassPath=&Overlay, parent=Overlay_0, widget=Overlay_Header, slot {class=OverlaySlot}}
  Text_Glow {
   widgetClassPath=&Image
   parent=Overlay_Header
   widget=Text_Glow
   properties {brush {
     bIsDynamicallyLoaded=false
     drawAs=Box
     imageType=NoImage
     mirroring=NoMirror
     resourceName=None
     resourceObject=/PlayerAugmentsUI/Materials/MI_UI_PlayerAugment_Font_Glow
     tiling=NoTile
     imageSize=<64 64>
     margin=<ltrb 0.5 0.5 0.5 0.5>
     outlineSettings {
      bUseBrushTransparency=false
      roundingType=HalfHeightRadius
      width=0
      color {colorUseRule=UseColor_Specified, specifiedColor=<rgba 0 0 0 0>}
      cornerRadii=<0 0 0 0>
     }
     tintColor {colorUseRule=UseColor_Specified, specifiedColor=<rgba 1 1 1 1>}
     uVRegion {bIsValid=false, max=<0 0>, min=<0 0>}
    }}
   slot {class=OverlaySlot, properties {horizontalAlignment=HAlign_Fill, verticalAlignment=VAlign_Fill, padding=<ltrb -28 -16 -28 -16>}}
  }
  Text_AugmentName {
   widgetClassPath=&CommonTextBlock
   bInherited=true
   parent=Overlay_Header
   widget=Text_AugmentName
   properties {
    lineHeightPercentage=0.9
    mobileFontSizeMultiplier=1.2
    text="[PH] Augment Name"
    textTransformPolicy=ToUpper
    font {
     bForceMonospaced=false
     bMaterialIsStencil=false
     fontMaterial=null
     fontObject=&Burbank
     letterSpacing=50
     monospacedWidth=1
     size=20
     skewAmount=0
     typefaceFontName=BurbankRegular-Black
     outlineSettings {
      bApplyOutlineToDropShadows=false
      bMiteredCorners=false
      bSeparateFillAlpha=true
      outlineMaterial=&M-FontShadow
      outlineSize=3
      outlineColor=<rgba 0.6667 0 1 1>
     }
    }
    margin=<ltrb 0 -2 0 0>
    shadowColorAndOpacity=<rgba 0 0.8333 1 0.125>
    shadowOffset=<6 6>
   }
   slot {class=OverlaySlot}
  }
  DescriptionTextOverlay {
   widgetClassPath=&Overlay
   parent=Text
   widget=DescriptionTextOverlay
   slot {class=VerticalBoxSlot, properties {horizontalAlignment=HAlign_Left, verticalAlignment=VAlign_Top, padding=<ltrb 0 0 0 32>}}
  }
  DescriptionTextSize {
   widgetClassPath=&Spacer
   parent=DescriptionTextOverlay
   widget=DescriptionTextSize
   properties {size=<224 1>}
   slot {class=OverlaySlot, properties {horizontalAlignment=HAlign_Fill, verticalAlignment=VAlign_Fill}}
  }
  Overlay_1 {widgetClassPath=&Overlay, parent=DescriptionTextOverlay, widget=Overlay_1, slot {class=OverlaySlot}}
  Image_SolidBG_AugmentDescription_Reflection {
   widgetClassPath=&Image
   parent=Overlay_1
   widget=Image_SolidBG_AugmentDescription_Reflection
   properties {
    visibility=SelfHitTestInvisible
    brush {
     bIsDynamicallyLoaded=false
     drawAs=Box
     imageType=NoImage
     mirroring=NoMirror
     resourceName=None
     resourceObject=/PlayerAugmentsUI/Materials/MI_UI_PlayerAugment_BackingPanel_NoNoise_Reflection_TextPanel
     tiling=NoTile
     imageSize=<32 32>
     margin=<ltrb 0.5 0.5 0.5 0.5>
     outlineSettings {
      bUseBrushTransparency=false
      roundingType=HalfHeightRadius
      width=0
      color {colorUseRule=UseColor_Specified, specifiedColor=<rgba 0 0 0 0>}
      cornerRadii=<0 0 0 0>
     }
     tintColor {colorUseRule=UseColor_Specified, specifiedColor=<rgba 1 1 1 1>}
     uVRegion {bIsValid=false, max=<0 0>, min=<0 0>}
    }
    renderTransform {angle=0, scale=<1 1>, shear=<0 0>, translation=<8 12>}
   }
   slot {class=OverlaySlot, properties {horizontalAlignment=HAlign_Fill, verticalAlignment=VAlign_Fill}}
  }
  Image_SolidBG_AugmentDescription {
   widgetClassPath=&Image
   parent=Overlay_1
   widget=Image_SolidBG_AugmentDescription
   properties {
    visibility=SelfHitTestInvisible
    brush {
     bIsDynamicallyLoaded=false
     drawAs=Box
     imageType=NoImage
     mirroring=NoMirror
     resourceName=None
     resourceObject=/PlayerAugmentsUI/Materials/MI_UI_PlayerAugment_BackingPanel_NoNoise
     tiling=NoTile
     imageSize=<32 32>
     margin=<ltrb 0.5 0.5 0.5 0.5>
     outlineSettings {
      bUseBrushTransparency=false
      roundingType=HalfHeightRadius
      width=0
      color {colorUseRule=UseColor_Specified, specifiedColor=<rgba 0 0 0 0>}
      cornerRadii=<0 0 0 0>
     }
     tintColor {colorUseRule=UseColor_Specified, specifiedColor=<rgba 1 1 1 1>}
     uVRegion {bIsValid=false, max=<0 0>, min=<0 0>}
    }
   }
   slot {class=OverlaySlot, properties {horizontalAlignment=HAlign_Fill, verticalAlignment=VAlign_Fill}}
  }
  TextPanelBrackets {
   widgetClassPath=&Image
   parent=Overlay_1
   widget=TextPanelBrackets
   properties {
    visibility=SelfHitTestInvisible
    brush {
     bIsDynamicallyLoaded=false
     drawAs=Box
     imageType=NoImage
     mirroring=NoMirror
     resourceName=None
     resourceObject=/PlayerAugmentsUI/Materials/MI_UI_PlayerAugment_Brackets
     tiling=NoTile
     imageSize=<32 32>
     margin=<ltrb 0.5 0.5 0.5 0.5>
     outlineSettings {
      bUseBrushTransparency=false
      roundingType=HalfHeightRadius
      width=0
      color {colorUseRule=UseColor_Specified, specifiedColor=<rgba 0 0 0 0>}
      cornerRadii=<0 0 0 0>
     }
     tintColor {colorUseRule=UseColor_Specified, specifiedColor=<rgba 1 1 1 1>}
     uVRegion {bIsValid=false, max=<0 0>, min=<0 0>}
    }
   }
   slot {class=OverlaySlot, properties {horizontalAlignment=HAlign_Fill, verticalAlignment=VAlign_Fill}}
  }
  VerticalBox_0 {
   widgetClassPath=&VerticalBox
   parent=Overlay_1
   widget=VerticalBox_0
   slot {class=OverlaySlot, properties {horizontalAlignment=HAlign_Fill}}
  }
  HorizontalBox_0 {widgetClassPath=&HorizontalBox, parent=VerticalBox_0, widget=HorizontalBox_0, slot {class=VerticalBoxSlot}}
  Text_AugmentCategory {
   widgetClassPath=&CommonTextBlock
   bInherited=true
   parent=HorizontalBox_0
   widget=Text_AugmentCategory
   properties {
    lineHeightPercentage=0.9
    mobileFontSizeMultiplier=1.45
    text="[PH] Augment Category"
    textTransformPolicy=ToUpper
    wrapTextAt=754
    font {
     bForceMonospaced=false
     bMaterialIsStencil=false
     fontMaterial=null
     fontObject=&BurbankSmall
     letterSpacing=0
     monospacedWidth=1
     size=11
     skewAmount=0.2
     typefaceFontName=Bold
     outlineSettings {
      bApplyOutlineToDropShadows=false
      bMiteredCorners=false
      bSeparateFillAlpha=true
      outlineMaterial=&M-FontShadow
      outlineSize=2
      outlineColor=<rgba 0.6667 0 1 1>
     }
    }
    shadowColorAndOpacity=<rgba 0 0.8333 1 0.125>
    shadowOffset=<6 4>
   }
   slot {class=HorizontalBoxSlot}
  }
  Text_AugmentDescription {
   widgetClassPath=&CommonTextBlock
   bInherited=true
   parent=VerticalBox_0
   widget=Text_AugmentDescription
   properties {
    lineHeightPercentage=0.9
    mobileFontSizeMultiplier=1.45
    text="[PH] Augment Description. Example text to see how long it will go before wrapping and test multi lines. Where are the multiple lines"
    wrapTextAt=432
    font {
     bForceMonospaced=false
     bMaterialIsStencil=false
     fontMaterial=/PlayerAugmentsUI/Materials/MI_UI_PlayerAugment_Body_Small_Font
     fontObject=&BurbankSmall
     letterSpacing=0
     monospacedWidth=1
     size=13
     skewAmount=0
     typefaceFontName=Bold
     outlineSettings {
      bApplyOutlineToDropShadows=false
      bMiteredCorners=false
      bSeparateFillAlpha=true
      outlineMaterial=&M-FontShadow
      outlineSize=2
      outlineColor=<rgba 0.1255 0 0.7217 1>
     }
    }
    shadowColorAndOpacity=<rgba 0 0.8333 1 0.125>
    shadowOffset=<6 4>
   }
   slot {class=VerticalBoxSlot}
  }
  Text_AugmentRarity {widgetClassPath=&CommonTextBlock, bInherited=true, widget=null}
  ItemCategoryIndicator {widgetClassPath=/Script/FortniteUI.FortItemCategoryIndicator, bInherited=true, widget=null}
  Image_CategoryAsImage {widgetClassPath=&FortLazyImage, bInherited=true, widget=null}
  Icon_LenderAvatarIcon {widgetClassPath=/Script/FortniteUI.FortSocialAvatarIcon, bInherited=true, widget=null}
  Overlay_LenderAvatarIcon {widgetClassPath=/Script/UMG.Widget, bInherited=true, widget=null}
  Image_LenderSquadColor {widgetClassPath=&FortLazyImage, bInherited=true, widget=null}
  Text_BorrowingText {widgetClassPath=&CommonTextBlock, bInherited=true, widget=null}
 ]
}
