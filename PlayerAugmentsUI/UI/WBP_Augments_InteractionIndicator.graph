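# Fortnite interaction indicator widget (FortInteractContextInfoWidget) for player augment pickups, with interact text from a variable defaulting to "Activate".
# On context update, it casts the receiving pickup's item to an augment definition and passes it to the child InteractionIndicatorInfo widget.
# It then plays AnimIntro and sends Event.CollectedPickup to the owning player, ending in an empty rarity switch.
# The older inline panel is collapsed, and its name, description and async icon-loading chain is unreached.
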WidgetBlueprint /PlayerAugmentsUI/UI/WBP_Augments_InteractionIndicator
animations [AnimIntro {
  binding_count=2
  length_seconds=1.15
  object=&self:AnimIntro
  sequence {bindings [
    Backing_Panel {
     id=#1
     possessed_class=Image
     tracks [{
       type=MovieSceneWidgetMaterialTrack
       sections [{
         type=MovieSceneParameterSection
         channels [
          Anim_ScanLines_Amount {type=MovieSceneScriptingFloatChannel, keys [[0,0],[3,0],[6,16],[16,0]]}
          ScanLines_Speed {type=MovieSceneScriptingFloatChannel, keys [[3,0.333],[6,2],[10,0.333]]}
         ]
        }]
      }]
    }
    Text {
     id=#2
     possessed_class=VerticalBox
     tracks [
      {
       type=MovieScene2DTransformTrack
       sections [{
         type=MovieScene2DTransformSection
         channels [
          Translation.X {type=MovieSceneScriptingFloatChannel, keys [[0,16],[16,0]]}
          Translation.Y {type=MovieSceneScriptingFloatChannel, keys [[0,16],[16,0]]}
          Scale.X {type=MovieSceneScriptingFloatChannel, keys [[0,0.9],[16,1]]}
          Scale.Y {type=MovieSceneScriptingFloatChannel, keys [[0,0.9],[16,1]]}
         ]
        }]
      }
      {
       type=MovieSceneFloatTrack
       display_name=Render Opacity
       sections [{type=MovieSceneFloatSection, channels [None {type=MovieSceneScriptingFloatChannel, keys [[0,0],[23,1]]}]}]
      }
     ]
    }
   ]}
 }]
graphs [
 EventGraph {nodes [
   n1 = Event OnContextInfoUpdated!() -> (OutputDelegate, then, ContextInfo:Fort Interact Context Info Object Reference) {is=title K2Node_Event}
   n2 = Cast To FortPickup!(
    execute=@n11.Is Valid
    Object=Get ReceivingActor!(self:Fort Interact Context Info Object Reference=@n1.ContextInfo) -> (ReceivingActor:Actor Weak object)
   ) -> (then, CastFailed, AsFort Pickup:Fort Pickup Object Reference) {is=title K2Node_DynamicCast}
   > n3 = Get!(execute:Exec, self:Fort Pickup Object Reference=@n2) -> (then:Exec, else:Exec, PrimaryPickupDummyItem:Fort Item Object Reference)
   > n4 = Cast To FortPlayerAugmentItemDefinition!(execute, Object=GetItemDefinition!(self:Fort Item Object Reference=@n3) -> (ReturnValue:Fort Item Definition Object Reference)) -> (then, CastFailed, AsFort Player Augment Item Definition:Fort Player Augment Item Definition Object Reference) {is=title K2Node_DynamicCast}
   "SetText (Text)"(
    self=Get Text_AugmentName() -> (Text_AugmentName:Common Text Object Reference)
    InText=GetDisplayName!(self:Item Definition Base Object Reference=@n4) -> (ReturnValue:Text)
   )
   > "SetText (Text)"(
    self=Get Text_AugmentDescription() -> (Text_AugmentDescription:Common Text Object Reference)
    InText=GetItemShortDescription!(self:Item Definition Base Object Reference=@n4) -> (ReturnValue:Text)
   )
   > n5 = SetAugmentIcon!(
    execute
    self
    Asset:Soft Object Reference=GetHUDEffectImage!(self:Fort Player Augment Item Definition Object Reference=@n4) -> (ReturnValue:Texture 2D Soft Object Reference)
   ) -> (then)
   PlayAnimation(execute=[@n5.then, @n10.then], InAnimation=Get AnimIntro() -> (AnimIntro:Widget Animation Object Reference), PlaybackSpeed=2)
   > SendGameplayEventToActor!(
    execute
    Actor:Actor Object Reference=GetOwningPlayer()
    EventTag:Gameplay Tag Structure=(TagName="Event.CollectedPickup")
    Payload:Gameplay Event Data Structure=Make Gameplay Event Data!(
     EventTag:Gameplay Tag Structure=(TagName="")
     Instigator:Actor Object Reference
     Target:Actor Object Reference
     OptionalObject:Object Reference=@n4
     OptionalObject2:Object Reference
     ContextHandle:Gameplay Effect Context Handle Structure=()
     InstigatorTags:Gameplay Tag Container Structure=(GameplayTags=)
     TargetTags:Gameplay Tag Container Structure=(GameplayTags=)
     EventMagnitude:float=0
     TargetData:Gameplay Ability Target Data Handle Structure=()
    ) -> (GameplayEventData:Gameplay Event Data Structure) {is=title K2Node_MakeStruct}
   ) -> (then)
   > Switch on EFortRarity!(execute, Selection:EFortRarity Enum=GetRarity!(self:Fort Item Definition Object Reference=@n4) -> (ReturnValue:EFortRarity Enum=Common)) -> (Common:Exec, Uncommon:Exec, Rare:Exec, Epic:Exec, Legendary:Exec, Mythic:Exec, Transcendent:Exec, Unattainable:Exec) {is=title K2Node_SwitchEnum}
   n6 = SetAugmentIcon!() -> (OutputDelegate, then, Asset:Soft Object Reference) {is=title K2Node_CustomEvent}
   > n7 = Async Load Asset(Asset=@n6.Asset)
   n8 = Cast To Texture2D(execute=[@n7.Completed, @n7.then], Object=@n7)
   > n9 = GetDynamicMaterial(self=Get Image_AugmentIcon() -> (Image_AugmentIcon:Image Object Reference))
   > SetTextureParameterValue(self=@n9, ParameterName=Icon, Value=@n8)
   n10 = SetAugmentInfo!(
    execute=@n4.then
    self:WBP Augments Interaction Indicator Info Object Reference=Get WBP_Augments_InteractionIndicatorInfo() -> (WBP_Augments_InteractionIndicatorInfo:WBP Augments Interaction Indicator Info Object Reference)
    InAugmentItemDefinition:Fort Player Augment Item Definition Object Reference=@n4
   ) -> (then)
   n11 = Is Valid(exec=@n1.then, InputObject=@n1.ContextInfo)
  ]}
 GetInteractText {nodes [
   GetInteractText() {is=title K2Node_FunctionEntry}
   > Parent: GetInteractText!(execute) -> (then, InteractText:Text, Result:bool) {is=title K2Node_CallParentFunction}
   > Return Node!(execute, InteractText:Text=Get InteractText() -> (InteractText:Text), Result:bool=true)
  ]}
 GetShowErrorsUsingDefaultWidget {nodes [
   GetShowErrorsUsingDefaultWidget() {is=title K2Node_FunctionEntry}
   > Return Node!(execute, ReturnValue:bool=true)
  ]}
]
variables [InteractText {type=string, default=Activate}]
widgets {
 info {
  inheritedWidgetCount=0
  parentClass=/Script/FortniteGame.FortInteractContextInfoWidget
  rootWidgetClass=&VerticalBox
  widgetCount=33
 }
 widgets [
  VerticalBox_1 {widgetClassPath=&VerticalBox, widget=VerticalBox_1}
  AnimBound_Border {
   widgetClassPath=&Border
   parent=VerticalBox_1
   widget=AnimBound_Border
   properties {
    visibility=Collapsed
    background {
     bIsDynamicallyLoaded=false
     drawAs=NoDrawType
     imageType=NoImage
     mirroring=NoMirror
     resourceName=None
     resourceObject=null
     tiling=NoTile
     imageSize=<32 32>
     margin=<ltrb 0 0 0 0>
     outlineSettings {
      bUseBrushTransparency=false
      roundingType=HalfHeightRadius
      width=0
      color {colorUseRule=UseColor_Specified, specifiedColor=<rgba 0 0 0 0>}
      cornerRadii=<0 0 0 0>
     }
     tintColor {colorUseRule=UseColor_Specified, specifiedColor=<rgba 1 1 1 1>}
     uVRegion {bIsValid=false, max=<0 0>, min=<0 0>}
    }
    padding=<ltrb 0 0 0 0>
   }
   slot {class=VerticalBoxSlot}
  }
  All {widgetClassPath=&Overlay, parent=AnimBound_Border, widget=All, slot {class=BorderSlot, properties {padding=<ltrb 0 0 0 0>}}}
  WidthConrol {
   widgetClassPath=&Spacer
   bIsVariable=true
   parent=All
   widget=WidthConrol
   properties {size=<288 85.3333>}
   slot {class=OverlaySlot, properties {horizontalAlignment=HAlign_Fill}}
  }
  Backing_Glow {
   widgetClassPath=&Image
   parent=All
   widget=Backing_Glow
   properties {
    visibility=SelfHitTestInvisible
    brush {
     bIsDynamicallyLoaded=false
     drawAs=Box
     imageType=NoImage
     mirroring=NoMirror
     resourceName=None
     resourceObject=/PlayerAugmentsUI/Materials/MI_UI_PlayerAugment_Panel_Glow
     tiling=NoTile
     imageSize=<300 100>
     margin=<ltrb 0.5 0.5 0.5 0.5>
     outlineSettings {
      bUseBrushTransparency=false
      roundingType=HalfHeightRadius
      width=0
      color {colorUseRule=UseColor_Specified, specifiedColor=<rgba 0 0 0 0>}
      cornerRadii=<0 0 0 0>
     }
     tintColor {colorUseRule=UseColor_Specified, specifiedColor=<rgba 1 1 1 1>}
     uVRegion {bIsValid=false, max=<0 0>, min=<0 0>}
    }
   }
   slot {class=OverlaySlot, properties {horizontalAlignment=HAlign_Fill, verticalAlignment=VAlign_Fill}}
  }
  Backing_Panel {
   widgetClassPath=&Image
   parent=All
   widget=Backing_Panel
   properties {
    visibility=SelfHitTestInvisible
    brush {
     bIsDynamicallyLoaded=false
     drawAs=Box
     imageType=NoImage
     mirroring=NoMirror
     resourceName=None
     resourceObject=/PlayerAugmentsUI/Materials/MI_UI_PlayerAugment_BackingPanel
     tiling=NoTile
     imageSize=<64 64>
     margin=<ltrb 0.5 0.5 0.5 0.5>
     outlineSettings {
      bUseBrushTransparency=false
      roundingType=HalfHeightRadius
      width=0
      color {colorUseRule=UseColor_Specified, specifiedColor=<rgba 0 0 0 0>}
      cornerRadii=<0 0 0 0>
     }
     tintColor {colorUseRule=UseColor_Specified, specifiedColor=<rgba 1 1 1 1>}
     uVRegion {bIsValid=false, max=<0 0>, min=<0 0>}
    }
   }
   slot {class=OverlaySlot, properties {horizontalAlignment=HAlign_Fill, verticalAlignment=VAlign_Fill}}
  }
  Content {widgetClassPath=&HorizontalBox, parent=All, widget=Content, slot {class=OverlaySlot, properties {verticalAlignment=VAlign_Fill}}}
  Icon {
   widgetClassPath=&Overlay
   parent=Content
   widget=Icon
   slot {class=HorizontalBoxSlot, properties {horizontalAlignment=HAlign_Center, verticalAlignment=VAlign_Center}}
  }
  IconSize {widgetClassPath=&Spacer, parent=Icon, widget=IconSize, properties {size=<42.6667 42.6667>}, slot {class=OverlaySlot}}
  Icon_BackingPanel_Reflection {
   widgetClassPath=&Image
   parent=Icon
   widget=Icon_BackingPanel_Reflection
   properties {
    visibility=SelfHitTestInvisible
    brush {
     bIsDynamicallyLoaded=false
     drawAs=Box
     imageType=NoImage
     mirroring=NoMirror
     resourceName=None
     resourceObject=/PlayerAugmentsUI/Materials/MI_UI_PlayerAugment_BackingPanel_NoNoise_Reflection
     tiling=NoTile
     imageSize=<64 64>
     margin=<ltrb 0.5 0.5 0.5 0.5>
     outlineSettings {
      bUseBrushTransparency=false
      roundingType=HalfHeightRadius
      width=0
      color {colorUseRule=UseColor_Specified, specifiedColor=<rgba 0 0 0 0>}
      cornerRadii=<0 0 0 0>
     }
     tintColor {colorUseRule=UseColor_Specified, specifiedColor=<rgba 1 1 1 1>}
     uVRegion {bIsValid=false, max=<0 0>, min=<0 0>}
    }
    renderTransform {angle=0, scale=<1 1>, shear=<0 0>, translation=<10 16>}
   }
   slot {class=OverlaySlot, properties {horizontalAlignment=HAlign_Fill, verticalAlignment=VAlign_Fill}}
  }
  Icon_ConfirmationFlare {
   widgetClassPath=&Image
   parent=Icon
   widget=Icon_ConfirmationFlare
   properties {brush {
     bIsDynamicallyLoaded=false
     drawAs=Image
     imageType=NoImage
     mirroring=NoMirror
     resourceName=None
     resourceObject=/PlayerAugmentsUI/Materials/M_UI_PlayerAugment_Gradient_Flare
     tiling=NoTile
     imageSize=<32 32>
     margin=<ltrb 0 0 0 0>
     outlineSettings {
      bUseBrushTransparency=false
      roundingType=HalfHeightRadius
      width=0
      color {colorUseRule=UseColor_Specified, specifiedColor=<rgba 0 0 0 0>}
      cornerRadii=<0 0 0 0>
     }
     tintColor {colorUseRule=UseColor_Specified, specifiedColor=<rgba 1 1 1 1>}
     uVRegion {bIsValid=false, max=<0 0>, min=<0 0>}
    }}
   slot {class=OverlaySlot, properties {horizontalAlignment=HAlign_Fill, verticalAlignment=VAlign_Fill}}
  }
  Icon_BackingPanel {
   widgetClassPath=&Image
   parent=Icon
   widget=Icon_BackingPanel
   properties {
    visibility=SelfHitTestInvisible
    brush {
     bIsDynamicallyLoaded=false
     drawAs=Box
     imageType=NoImage
     mirroring=NoMirror
     resourceName=None
     resourceObject=/PlayerAugmentsUI/Materials/MI_UI_PlayerAugment_BackingPanel_NoNoise_Icon
     tiling=NoTile
     imageSize=<64 64>
     margin=<ltrb 0.5 0.5 0.5 0.5>
     outlineSettings {
      bUseBrushTransparency=false
      roundingType=HalfHeightRadius
      width=0
      color {colorUseRule=UseColor_Specified, specifiedColor=<rgba 0 0 0 0>}
      cornerRadii=<0 0 0 0>
     }
     tintColor {colorUseRule=UseColor_Specified, specifiedColor=<rgba 1 1 1 1>}
     uVRegion {bIsValid=false, max=<0 0>, min=<0 0>}
    }
   }
   slot {class=OverlaySlot, properties {horizontalAlignment=HAlign_Fill, verticalAlignment=VAlign_Fill}}
  }
  Icon_BackingPanel_UnlockedFX {
   widgetClassPath=&Image
   bIsVariable=true
   parent=Icon
   widget=Icon_BackingPanel_UnlockedFX
   properties {
    visibility=SelfHitTestInvisible
    brush {
     bIsDynamicallyLoaded=false
     drawAs=Box
     imageType=NoImage
     mirroring=NoMirror
     resourceName=None
     resourceObject=/PlayerAugmentsUI/Materials/M_UI_AugmentUnlocked_Pattern
     tiling=NoTile
     imageSize=<64 64>
     margin=<ltrb 0.5 0.5 0.5 0.5>
     outlineSettings {
      bUseBrushTransparency=false
      roundingType=HalfHeightRadius
      width=0
      color {colorUseRule=UseColor_Specified, specifiedColor=<rgba 0 0 0 0>}
      cornerRadii=<0 0 0 0>
     }
     tintColor {colorUseRule=UseColor_Specified, specifiedColor=<rgba 1 1 1 1>}
     uVRegion {bIsValid=false, max=<0 0>, min=<0 0>}
    }
   }
   slot {class=OverlaySlot, properties {horizontalAlignment=HAlign_Fill, verticalAlignment=VAlign_Fill}}
  }
  Icon_Border_UnlockedGlow {
   widgetClassPath=&Image
   bIsVariable=true
   parent=Icon
   widget=Icon_Border_UnlockedGlow
   properties {
    visibility=SelfHitTestInvisible
    brush {
     bIsDynamicallyLoaded=false
     drawAs=Box
     imageType=NoImage
     mirroring=NoMirror
     resourceName=None
     resourceObject=/PlayerAugmentsUI/Materials/MI_UI_PlayerAugment_Unlocked_IconBacking_Glow_Border
     tiling=NoTile
     imageSize=<64 64>
     margin=<ltrb 0.5 0.5 0.5 0.5>
     outlineSettings {
      bUseBrushTransparency=false
      roundingType=HalfHeightRadius
      width=0
      color {colorUseRule=UseColor_Specified, specifiedColor=<rgba 0 0 0 0>}
      cornerRadii=<0 0 0 0>
     }
     tintColor {colorUseRule=UseColor_Specified, specifiedColor=<rgba 1 1 1 1>}
     uVRegion {bIsValid=false, max=<0 0>, min=<0 0>}
    }
   }
   slot {class=OverlaySlot, properties {horizontalAlignment=HAlign_Fill, verticalAlignment=VAlign_Fill, padding=<ltrb -16 -16 -16 -16>}}
  }
  Icon_Border_ConfirmationPulse {
   widgetClassPath=&Image
   parent=Icon
   widget=Icon_Border_ConfirmationPulse
   properties {
    visibility=SelfHitTestInvisible
    brush {
     bIsDynamicallyLoaded=false
     drawAs=Box
     imageType=NoImage
     mirroring=NoMirror
     resourceName=None
     resourceObject=/PlayerAugmentsUI/Materials/MI_UI_PlayerAugmment_Border_ConfirmationPulse
     tiling=NoTile
     imageSize=<64 64>
     margin=<ltrb 0.5 0.5 0.5 0.5>
     outlineSettings {
      bUseBrushTransparency=false
      roundingType=HalfHeightRadius
      width=0
      color {colorUseRule=UseColor_Specified, specifiedColor=<rgba 0 0 0 0>}
      cornerRadii=<0 0 0 0>
     }
     tintColor {colorUseRule=UseColor_Specified, specifiedColor=<rgba 1 1 1 1>}
     uVRegion {bIsValid=false, max=<0 0>, min=<0 0>}
    }
   }
   slot {class=OverlaySlot, properties {horizontalAlignment=HAlign_Fill, verticalAlignment=VAlign_Fill}}
  }
  Icon_Border {
   widgetClassPath=&Image
   parent=Icon
   widget=Icon_Border
   properties {
    visibility=SelfHitTestInvisible
    brush {
     bIsDynamicallyLoaded=false
     drawAs=Box
     imageType=NoImage
     mirroring=NoMirror
     resourceName=None
     resourceObject=/PlayerAugmentsUI/Materials/MI_UI_PlayerAugment_Border
     tiling=NoTile
     imageSize=<64 64>
     margin=<ltrb 0.5 0.5 0.5 0.5>
     outlineSettings {
      bUseBrushTransparency=false
      roundingType=HalfHeightRadius
      width=0
      color {colorUseRule=UseColor_Specified, specifiedColor=<rgba 0 0 0 0>}
      cornerRadii=<0 0 0 0>
     }
     tintColor {colorUseRule=UseColor_Specified, specifiedColor=<rgba 1 1 1 1>}
     uVRegion {bIsValid=false, max=<0 0>, min=<0 0>}
    }
    renderTransformPivot=<1 1>
   }
   slot {class=OverlaySlot, properties {horizontalAlignment=HAlign_Fill, verticalAlignment=VAlign_Fill}}
  }
  Icon_Border_Unlocked {
   widgetClassPath=&Image
   bIsVariable=true
   parent=Icon
   widget=Icon_Border_Unlocked
   properties {
    visibility=SelfHitTestInvisible
    brush {
     bIsDynamicallyLoaded=false
     drawAs=Box
     imageType=NoImage
     mirroring=NoMirror
     resourceName=None
     resourceObject=/PlayerAugmentsUI/Materials/MI_UI_PlayerAugment_Border_Unlcoked
     tiling=NoTile
     imageSize=<64 64>
     margin=<ltrb 0.5 0.5 0.5 0.5>
     outlineSettings {
      bUseBrushTransparency=false
      roundingType=HalfHeightRadius
      width=0
      color {colorUseRule=UseColor_Specified, specifiedColor=<rgba 0 0 0 0>}
      cornerRadii=<0 0 0 0>
     }
     tintColor {colorUseRule=UseColor_Specified, specifiedColor=<rgba 1 1 1 1>}
     uVRegion {bIsValid=false, max=<0 0>, min=<0 0>}
    }
    renderTransformPivot=<1 1>
   }
   slot {class=OverlaySlot, properties {horizontalAlignment=HAlign_Fill, verticalAlignment=VAlign_Fill}}
  }
  Image_AugmentIcon {
   widgetClassPath=&Image
   bIsVariable=true
   parent=Icon
   widget=Image_AugmentIcon
   properties {
    visibility=SelfHitTestInvisible
    brush {
     bIsDynamicallyLoaded=false
     drawAs=Image
     imageType=NoImage
     mirroring=NoMirror
     resourceName=None
     resourceObject=/PlayerAugmentsUI/Materials/MI_UI_PlayerAugment_HUD_Icon
     tiling=NoTile
     imageSize=<96 96>
     margin=<ltrb 0 0 0 0>
     outlineSettings {
      bUseBrushTransparency=false
      roundingType=HalfHeightRadius
      width=0
      color {colorUseRule=UseColor_Specified, specifiedColor=<rgba 0 0 0 0>}
      cornerRadii=<0 0 0 0>
     }
     tintColor {colorUseRule=UseColor_Specified, specifiedColor=<rgba 1 1 1 1>}
     uVRegion {bIsValid=false, max=<0 0>, min=<0 0>}
    }
   }
   slot {class=OverlaySlot, properties {horizontalAlignment=HAlign_Center, verticalAlignment=VAlign_Center}}
  }
  Text {widgetClassPath=&VerticalBox, parent=Content, widget=Text, slot {class=HorizontalBoxSlot, properties {padding=<ltrb 24 32 0 0>}}}
  Overlay_0 {
   widgetClassPath=&Overlay
   parent=Text
   widget=Overlay_0
   slot {class=VerticalBoxSlot, properties {horizontalAlignment=HAlign_Left, verticalAlignment=VAlign_Center}}
  }
  Overlay_Header {widgetClassPath=&Overlay, parent=Overlay_0, widget=Overlay_Header, slot {class=OverlaySlot}}
  Text_Glow {
   widgetClassPath=&Image
   parent=Overlay_Header
   widget=Text_Glow
   properties {brush {
     bIsDynamicallyLoaded=false
     drawAs=Box
     imageType=NoImage
     mirroring=NoMirror
     resourceName=None
     resourceObject=/PlayerAugmentsUI/Materials/MI_UI_PlayerAugment_Font_Glow
     tiling=NoTile
     imageSize=<64 64>
     margin=<ltrb 0.5 0.5 0.5 0.5>
     outlineSettings {
      bUseBrushTransparency=false
      roundingType=HalfHeightRadius
      width=0
      color {colorUseRule=UseColor_Specified, specifiedColor=<rgba 0 0 0 0>}
      cornerRadii=<0 0 0 0>
     }
     tintColor {colorUseRule=UseColor_Specified, specifiedColor=<rgba 1 1 1 1>}
     uVRegion {bIsValid=false, max=<0 0>, min=<0 0>}
    }}
   slot {class=OverlaySlot, properties {horizontalAlignment=HAlign_Fill, verticalAlignment=VAlign_Fill, padding=<ltrb -28 -16 -28 -16>}}
  }
  Text_AugmentName {
   widgetClassPath=&CommonTextBlock
   bIsVariable=true
   parent=Overlay_Header
   widget=Text_AugmentName
   properties {
    lineHeightPercentage=0.9
    mobileFontSizeMultiplier=1.2
    text="[PH] Augment Name"
    textTransformPolicy=ToUpper
    font {
     bForceMonospaced=false
     bMaterialIsStencil=false
     fontMaterial=null
     fontObject=&Burbank
     letterSpacing=50
     monospacedWidth=1
     size=20
     skewAmount=0
     typefaceFontName=BurbankRegular-Black
     outlineSettings {
      bApplyOutlineToDropShadows=false
      bMiteredCorners=false
      bSeparateFillAlpha=true
      outlineMaterial=&M-FontShadow
      outlineSize=3
      outlineColor=<rgba 0 0.1667 1 1>
     }
    }
    margin=<ltrb 0 -2 0 0>
    shadowColorAndOpacity=<rgba 0 0.8333 1 0.125>
    shadowOffset=<6 6>
   }
   slot {class=OverlaySlot}
  }
  DescriptionTextOverlay {
   widgetClassPath=&Overlay
   parent=Text
   widget=DescriptionTextOverlay
   slot {class=VerticalBoxSlot, properties {horizontalAlignment=HAlign_Left, verticalAlignment=VAlign_Top, padding=<ltrb 0 0 0 32>}}
  }
  DescriptionTextSize {
   widgetClassPath=&Spacer
   parent=DescriptionTextOverlay
   widget=DescriptionTextSize
   properties {size=<224 1>}
   slot {class=OverlaySlot, properties {horizontalAlignment=HAlign_Fill, verticalAlignment=VAlign_Fill}}
  }
  Overlay_1 {widgetClassPath=&Overlay, parent=DescriptionTextOverlay, widget=Overlay_1, slot {class=OverlaySlot}}
  Image_SolidBG_AugmentDescription_Reflection {
   widgetClassPath=&Image
   parent=Overlay_1
   widget=Image_SolidBG_AugmentDescription_Reflection
   properties {
    visibility=SelfHitTestInvisible
    brush {
     bIsDynamicallyLoaded=false
     drawAs=Box
     imageType=NoImage
     mirroring=NoMirror
     resourceName=None
     resourceObject=/PlayerAugmentsUI/Materials/MI_UI_PlayerAugment_BackingPanel_NoNoise_Reflection_TextPanel
     tiling=NoTile
     imageSize=<32 32>
     margin=<ltrb 0.5 0.5 0.5 0.5>
     outlineSettings {
      bUseBrushTransparency=false
      roundingType=HalfHeightRadius
      width=0
      color {colorUseRule=UseColor_Specified, specifiedColor=<rgba 0 0 0 0>}
      cornerRadii=<0 0 0 0>
     }
     tintColor {colorUseRule=UseColor_Specified, specifiedColor=<rgba 1 1 1 1>}
     uVRegion {bIsValid=false, max=<0 0>, min=<0 0>}
    }
    renderTransform {angle=0, scale=<1 1>, shear=<0 0>, translation=<8 12>}
   }
   slot {class=OverlaySlot, properties {horizontalAlignment=HAlign_Fill, verticalAlignment=VAlign_Fill}}
  }
  Image_SolidBG_AugmentDescription {
   widgetClassPath=&Image
   parent=Overlay_1
   widget=Image_SolidBG_AugmentDescription
   properties {
    visibility=SelfHitTestInvisible
    brush {
     bIsDynamicallyLoaded=false
     drawAs=Box
     imageType=NoImage
     mirroring=NoMirror
     resourceName=None
     resourceObject=/PlayerAugmentsUI/Materials/MI_UI_PlayerAugment_BackingPanel_NoNoise
     tiling=NoTile
     imageSize=<32 32>
     margin=<ltrb 0.5 0.5 0.5 0.5>
     outlineSettings {
      bUseBrushTransparency=false
      roundingType=HalfHeightRadius
      width=0
      color {colorUseRule=UseColor_Specified, specifiedColor=<rgba 0 0 0 0>}
      cornerRadii=<0 0 0 0>
     }
     tintColor {colorUseRule=UseColor_Specified, specifiedColor=<rgba 1 1 1 1>}
     uVRegion {bIsValid=false, max=<0 0>, min=<0 0>}
    }
   }
   slot {class=OverlaySlot, properties {horizontalAlignment=HAlign_Fill, verticalAlignment=VAlign_Fill}}
  }
  TextPanelBrackets {
   widgetClassPath=&Image
   parent=Overlay_1
   widget=TextPanelBrackets
   properties {
    visibility=SelfHitTestInvisible
    brush {
     bIsDynamicallyLoaded=false
     drawAs=Box
     imageType=NoImage
     mirroring=NoMirror
     resourceName=None
     resourceObject=/PlayerAugmentsUI/Materials/MI_UI_PlayerAugment_Brackets
     tiling=NoTile
     imageSize=<32 32>
     margin=<ltrb 0.5 0.5 0.5 0.5>
     outlineSettings {
      bUseBrushTransparency=false
      roundingType=HalfHeightRadius
      width=0
      color {colorUseRule=UseColor_Specified, specifiedColor=<rgba 0 0 0 0>}
      cornerRadii=<0 0 0 0>
     }
     tintColor {colorUseRule=UseColor_Specified, specifiedColor=<rgba 1 1 1 1>}
     uVRegion {bIsValid=false, max=<0 0>, min=<0 0>}
    }
   }
   slot {class=OverlaySlot, properties {horizontalAlignment=HAlign_Fill, verticalAlignment=VAlign_Fill}}
  }
  VerticalBox_0 {
   widgetClassPath=&VerticalBox
   parent=Overlay_1
   widget=VerticalBox_0
   slot {class=OverlaySlot, properties {horizontalAlignment=HAlign_Fill}}
  }
  Text_AugmentCategory {
   widgetClassPath=&CommonTextBlock
   bIsVariable=true
   parent=VerticalBox_0
   widget=Text_AugmentCategory
   properties {
    lineHeightPercentage=0.9
    mobileFontSizeMultiplier=1.45
    text="[PH] Augment Category"
    textTransformPolicy=ToUpper
    wrapTextAt=754
    font {
     bForceMonospaced=false
     bMaterialIsStencil=false
     fontMaterial=null
     fontObject=&BurbankSmall
     letterSpacing=0
     monospacedWidth=1
     size=11
     skewAmount=0.2
     typefaceFontName=Bold
     outlineSettings {
      bApplyOutlineToDropShadows=false
      bMiteredCorners=false
      bSeparateFillAlpha=true
      outlineMaterial=&M-FontShadow
      outlineSize=2
      outlineColor=<rgba 0 0.1667 1 1>
     }
    }
    shadowColorAndOpacity=<rgba 0 0.8333 1 0.125>
    shadowOffset=<6 4>
   }
   slot {class=VerticalBoxSlot}
  }
  Text_AugmentDescription {
   widgetClassPath=&CommonTextBlock
   bIsVariable=true
   parent=VerticalBox_0
   widget=Text_AugmentDescription
   properties {
    lineHeightPercentage=0.9
    mobileFontSizeMultiplier=1.45
    text="[PH] Augment Description."
    wrapTextAt=432
    font {
     bForceMonospaced=false
     bMaterialIsStencil=false
     fontMaterial=/PlayerAugmentsUI/Materials/MI_UI_PlayerAugment_Body_Small_Font
     fontObject=&BurbankSmall
     letterSpacing=0
     monospacedWidth=1
     size=13
     skewAmount=0
     typefaceFontName=Bold
     outlineSettings {
      bApplyOutlineToDropShadows=false
      bMiteredCorners=false
      bSeparateFillAlpha=true
      outlineMaterial=&M-FontShadow
      outlineSize=2
      outlineColor=<rgba 0 0.1667 1 1>
     }
    }
    shadowColorAndOpacity=<rgba 0 0.8333 1 0.125>
    shadowOffset=<6 4>
   }
   slot {class=VerticalBoxSlot}
  }
  WBP_Augments_InteractionIndicatorInfo {
   widgetClassPath=&/WBP_Augments_InteractionIndicatorInfo
   bIsVariable=true
   parent=VerticalBox_1
   widget=WBP_Augments_InteractionIndicatorInfo
   slot {class=VerticalBoxSlot}
  }
 ]
}
