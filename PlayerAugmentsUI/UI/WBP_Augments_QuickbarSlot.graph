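# Quickbar slot widget for player augments, switching between an empty state and a filled augment state.
# SetAugmentInfo stores the augment definition, async-loads its HUD image into the icon's dynamic material and clears the "new" badge.
# SetAsEmptySlot highlights the next-augment slot with animated borders, arrow and looping pulse.
# A FortLiveSpectatorController check hides those highlight images and skips their animations for live spectators.
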WidgetBlueprint /PlayerAugmentsUI/UI/WBP_Augments_QuickbarSlot
animations [
 Anim_Activate {
  binding_count=1
  length_seconds=0.95
  object=&self:Anim_Activate
  sequence {bindings [Image_Glow {
     id=#1
     possessed_class=Image
     tracks [
      {
       type=MovieSceneColorTrack
       display_name=Color and Opacity
       sections [{
         type=MovieSceneColorSection
         end_frame=19
         start_frame=0
         channels [
          Color.R {type=MovieSceneScriptingFloatChannel, const [1,1]}
          Color.G {type=MovieSceneScriptingFloatChannel, const [1,1]}
          Color.B {type=MovieSceneScriptingFloatChannel, const [1,1]}
          Color.A {type=MovieSceneScriptingFloatChannel, keys [[0,0],[9,2],[19,0]]}
         ]
        }]
      }
      {
       type=MovieScene2DTransformTrack
       sections [{
         type=MovieScene2DTransformSection
         end_frame=19
         start_frame=0
         channels [
          Translation.X {type=MovieSceneScriptingFloatChannel, const [0,1]}
          Translation.Y {type=MovieSceneScriptingFloatChannel, const [0,1]}
          Angle {type=MovieSceneScriptingFloatChannel, const [0,1]}
          Scale.X {type=MovieSceneScriptingFloatChannel, keys [[0,1],[9,2],[19,0]]}
          Scale.Y {type=MovieSceneScriptingFloatChannel, keys [[0,1],[9,2],[19,0]]}
          Shear.X {type=MovieSceneScriptingFloatChannel, const [0,1]}
          Shear.Y {type=MovieSceneScriptingFloatChannel, const [0,1]}
         ]
        }]
      }
     ]
    }]}
 }
 Anim_NextAugment {
  binding_count=5
  length_seconds=0.1
  object=&self:Anim_NextAugment
  sequence {bindings [
    Image_EmptySlotBacking_Reflection {
     id=#2
     possessed_class=Image
     tracks [
      {
       type=MovieSceneWidgetMaterialTrack
       sections [{type=MovieSceneParameterSection, channels [Anim_BackingAlpha {type=MovieSceneScriptingFloatChannel, keys [[0,0],[2,1]]}]}]
      }
      {
       type=MovieScene2DTransformTrack
       sections [{
         type=MovieScene2DTransformSection
         channels [
          Translation.X {type=MovieSceneScriptingFloatChannel, keys [[0,0],[2,4]]}
          Translation.Y {type=MovieSceneScriptingFloatChannel, keys [[0,0],[2,4]]}
         ]
        }]
      }
     ]
    }
    Image_BorderGlow {
     id=#3
     possessed_class=Image
     tracks [{
       type=MovieSceneWidgetMaterialTrack
       sections [{type=MovieSceneParameterSection, channels [Anim_BorderAlpha {type=MovieSceneScriptingFloatChannel, keys [[0,0],[2,2]]}]}]
      }]
    }
    Overlay_EmptySlot {
     id=#4
     possessed_class=Overlay
     tracks [{
       type=MovieScene2DTransformTrack
       sections [{
         type=MovieScene2DTransformSection
         channels [
          Translation.X {type=MovieSceneScriptingFloatChannel, keys [[0,0],[2,-2]]}
          Translation.Y {type=MovieSceneScriptingFloatChannel, keys [[0,0],[2,-2]]}
         ]
        }]
      }]
    }
    Image_EmptySlotIcon {
     id=#5
     possessed_class=Image
     tracks [{
       type=MovieSceneWidgetMaterialTrack
       sections [{type=MovieSceneParameterSection, channels [Anim_BorderAlpha {type=MovieSceneScriptingFloatChannel, keys [[0,0],[2,1]]}]}]
      }]
    }
    Image_EmptySlotBacking {
     id=#6
     possessed_class=Image
     tracks [{
       type=MovieSceneWidgetMaterialTrack
       sections [{type=MovieSceneParameterSection, channels [Anim_BackingAlpha {type=MovieSceneScriptingFloatChannel, keys [[0,-0.5],[2,0]]}]}]
      }]
    }
   ]}
 }
 Anim_NextAugmentPulse {
  binding_count=1
  length_seconds=1
  object=&self:Anim_NextAugmentPulse
  sequence {
   bindings [Image_EmptySlotPulse {
     id=#7
     possessed_class=Image
     tracks [{
       type=MovieSceneWidgetMaterialTrack
       sections [{type=MovieSceneParameterSection, channels [Anim_BorderAlpha {type=MovieSceneScriptingFloatChannel, keys [[0,0],[15,1],[45,0],[60,0]]}]}]
      }]
    }]
   display_rate {numerator=60}
  }
 }
 Anim_NextAugmentPulse_Scale {
  binding_count=3
  length_seconds=1.5
  object=&self:Anim_NextAugmentPulse_Scale
  sequence {
   bindings [
    Image_Arrow {
     id=#8
     possessed_class=Image
     tracks [
      {
       type=MovieScene2DTransformTrack
       sections [{
         type=MovieScene2DTransformSection
         channels [
          Translation.X {type=MovieSceneScriptingFloatChannel, const [0,1]}
          Translation.Y {type=MovieSceneScriptingFloatChannel, keys [[0,-16],[45,-24],[75,-8]]}
          Scale.X {type=MovieSceneScriptingFloatChannel, keys [[0,1],[45,1.5],[75,1]]}
          Scale.Y {type=MovieSceneScriptingFloatChannel, keys [[0,1],[45,1.5],[75,1]]}
         ]
        }]
      }
      {
       type=MovieSceneWidgetMaterialTrack
       sections [{type=MovieSceneParameterSection, channels [Alpha {type=MovieSceneScriptingFloatChannel, keys [[45,0],[54,0.75],[75,1],[90,0]]}]}]
      }
     ]
    }
    Image_EmptySlotPulse_Scale {
     id=#9
     possessed_class=Image
     tracks [
      {
       type=MovieSceneWidgetMaterialTrack
       sections [{
         type=MovieSceneParameterSection
         channels [
          Anim_BorderAlpha {type=MovieSceneScriptingFloatChannel, keys [[0,-1],[15,1],[90,-0.5]]}
          Border_Scale_Y {type=MovieSceneScriptingFloatChannel, keys [[30,0.9],[60,0.95]]}
          Border_Scale_X {type=MovieSceneScriptingFloatChannel, keys [[30,0.9],[60,0.95]]}
         ]
        }]
      }
      {
       type=MovieScene2DTransformTrack
       sections [{
         type=MovieScene2DTransformSection
         channels [
          Scale.X {type=MovieSceneScriptingFloatChannel, keys [[0,1],[60,1.25]]}
          Scale.Y {type=MovieSceneScriptingFloatChannel, keys [[0,1],[60,1.25]]}
         ]
        }]
      }
     ]
    }
    Image_EmptySlotPulse {
     id=#7
     possessed_class=Image
     tracks [{
       type=MovieSceneWidgetMaterialTrack
       sections [{type=MovieSceneParameterSection, channels [Anim_BorderAlpha {type=MovieSceneScriptingFloatChannel, keys [[0,0],[15,1],[45,0],[60,0]]}]}]
      }]
    }
   ]
   display_rate {numerator=60}
  }
 }
]
graphs [
 EventGraph {nodes [
   n1 = LoadAndSetIcon!() -> (OutputDelegate, then, IconSoftReference:Texture 2D Soft Object Reference) {is=title K2Node_CustomEvent}
   > n2 = Async Load Asset(Asset=@n1.IconSoftReference)
   n3 = Cast To Texture!(execute=@n2.Completed, Object=@n2) -> (then, CastFailed, AsTexture:Texture Object Reference) {is=title K2Node_DynamicCast}
   > UpdateAugmentIcon!(execute, self, AugmentIcon:Texture Object Reference=@n3) -> (then)
  ]}
 SetAsEmptySlot {nodes [
   SetAsEmptySlot!() -> (then, IsNextAugment:bool=~) {is=title K2Node_FunctionEntry}
   > SetActiveWidget(
    self=Get Switcher_Slot() -> (Switcher_Slot:Common Visibility Switcher Object Reference)
    Widget=Get Overlay_EmptySlot() -> (Overlay_EmptySlot:Overlay Object Reference)
   )
   > n1 = Cast To FortLiveSpectatorController!(execute, Object=GetPlayerController!(PlayerIndex:int=0) -> (ReturnValue:Player Controller Object Reference)) -> (then, CastFailed, AsFort Live Spectator Controller:Fort Live Spectator Controller Object Reference) {is=title K2Node_DynamicCast}
   > n2 = Set Is Live Spectator(Is Live Spectator:bool=true) -> (Output_Get:bool)
   n3 = NOT Boolean(A=Get Is Live Spectator() -> (Is Live Spectator:bool))
   n4 = Get IsNextAugment() -> (IsNextAugment:bool)
   SetVisibility(
    execute=[@n8.then, @n2.then]
    self=[
     Get Image_BorderGlow() -> (Image_BorderGlow:Image Object Reference)
     Get Image_Arrow() -> (Image_Arrow:Image Object Reference)
     Get Image_EmptySlotPulse_Scale() -> (Image_EmptySlotPulse_Scale:Image Object Reference)
     Get Image_EmptySlotPulse() -> (Image_EmptySlotPulse:Image Object Reference)
     Get Image_EmptySlotBacking() -> (Image_EmptySlotBacking:Image Object Reference)
    ]
    InVisibility=Select(Option 0=Hidden, Option 1=SelfHitTestInvisible, Index=AND Boolean(A=@n4, B=@n3))
   )
   > n5 = Branch(Condition=Get Is Live Spectator())
   PlayAnimation(
    execute=@n5.else
    InAnimation=Get Anim_NextAugment() -> (Anim_NextAugment:Widget Animation Object Reference)
    PlayMode=Select(Option 0:EUMGSequencePlayMode Enum=Reverse, Option 1:EUMGSequencePlayMode Enum, Index=Get IsNextAugment()) -> (ReturnValue:EUMGSequencePlayMode Enum)
   )
   > n6 = Branch(Condition=Get IsNextAugment())
   n7 = Get Anim_NextAugmentPulse_Scale() -> (Anim_NextAugmentPulse_Scale:Widget Animation Object Reference)
   > PlayAnimation(InAnimation=@n7, NumLoopsToPlay=0)
   PlayAnimation(execute=@n6.else, InAnimation=@n7, StartAtTime=GetEndTime(self=@n7), PlayMode=Reverse)
   n8 = Set Is Live Spectator(execute=@n1.CastFailed, Is Live Spectator=false)
  ]}
 SetAsNewAugment {nodes [
   n1 = SetAsNewAugment!() -> (then, IsNewAugment:bool=~) {is=title K2Node_FunctionEntry}
   > SetVisibility(
    self=Get Overlay_NewVisuals() -> (Overlay_NewVisuals:Overlay Object Reference)
    InVisibility=Select(Option 0:ESlateVisibility Enum=Collapsed, Option 1:ESlateVisibility Enum=SelfHitTestInvisible, Index=@n1) -> (ReturnValue:ESlateVisibility Enum)
   )
  ]}
 SetAugmentInfo {nodes [
   n1 = SetAugmentInfo!() -> (then, InAugmentData:Fort Player Augment Item Definition Object Reference) {is=title K2Node_FunctionEntry}
   > n2 = Set Augment Data(Augment Data:Fort Player Augment Item Definition Object Reference=@n1) -> (Output_Get:Fort Player Augment Item Definition Object Reference)
   > SetActiveWidget(self=Get Switcher_Slot(), Widget=Get Overlay_AugmentSlot() -> (Overlay_AugmentSlot:Overlay Object Reference))
   > LoadAndSetIcon!(
    execute
    self
    IconSoftReference:Texture 2D Soft Object Reference=GetHUDEffectImage!(self:Fort Player Augment Item Definition Object Reference=@n2) -> (ReturnValue:Texture 2D Soft Object Reference)
   ) -> (then) {is=title K2Node_CallFunction}
   > SetAsNewAugment!(execute, self, IsNewAugment:bool=false) -> (then) {is=title K2Node_CallFunction}
  ]}
 UpdateAugmentIcon {nodes [
   n1 = UpdateAugmentIcon!() -> (then, AugmentIcon:Texture Object Reference) {is=title K2Node_FunctionEntry}
   > n2 = GetDynamicMaterial(self=Get Image_AugmentIcon() -> (Image_AugmentIcon:Image Object Reference))
   > SetTextureParameterValue(self=@n2, ParameterName=Icon, Value=@n1)
  ]}
]
variables [Augment Data {type=FortPlayerAugmentItemDefinition, default=null}]
widgets {
 info {inheritedWidgetCount=0, parentClass=&UserWidget, rootWidgetClass=&Overlay, widgetCount=17}
 widgets [
  Overlay_Root {widgetClassPath=&Overlay, widget=Overlay_Root}
  Switcher_Slot {
   widgetClassPath=&CommonVisibilitySwitcher
   bIsVariable=true
   parent=Overlay_Root
   widget=Switcher_Slot
   slot {class=OverlaySlot, properties {horizontalAlignment=HAlign_Center, verticalAlignment=VAlign_Center, padding=<ltrb 10 10 10 10>}}
  }
  Overlay_EmptySlot {
   widgetClassPath=&Overlay
   bIsVariable=true
   parent=Switcher_Slot
   widget=Overlay_EmptySlot
   slot {class=CommonVisibilitySwitcherSlot, properties {horizontalAlignment=HAlign_Center, verticalAlignment=VAlign_Center}}
  }
  Image_EmptySlotBacking_Reflection {
   widgetClassPath=&Image
   bIsVariable=true
   parent=Overlay_EmptySlot
   widget=Image_EmptySlotBacking_Reflection
   properties {
    clipping=ClipToBoundsWithoutIntersecting
    visibility=SelfHitTestInvisible
    brush {
     bIsDynamicallyLoaded=false
     drawAs=Box
     imageType=NoImage
     mirroring=NoMirror
     resourceName=None
     resourceObject=/PlayerAugmentsUI/Materials/MI_UI_PlayerAugment_BackingPanel_NoNoise_Reflection
     tiling=NoTile
     imageSize=<16 16>
     margin=<ltrb 0.5 0.5 0.5 0.5>
     outlineSettings {
      bUseBrushTransparency=false
      roundingType=HalfHeightRadius
      width=0
      color {colorUseRule=UseColor_Specified, specifiedColor=<rgba 0 0 0 0>}
      cornerRadii=<0 0 0 0>
     }
     tintColor {colorUseRule=UseColor_Specified, specifiedColor=<rgba 1 1 1 1>}
     uVRegion {bIsValid=false, max=<0 0>, min=<0 0>}
    }
   }
   slot {class=OverlaySlot, properties {horizontalAlignment=HAlign_Fill, verticalAlignment=VAlign_Fill}}
  }
  Image_EmptySlotBacking {
   widgetClassPath=&Image
   bIsVariable=true
   parent=Overlay_EmptySlot
   widget=Image_EmptySlotBacking
   properties {
    clipping=ClipToBoundsWithoutIntersecting
    visibility=SelfHitTestInvisible
    brush {
     bIsDynamicallyLoaded=false
     drawAs=Image
     imageType=NoImage
     mirroring=NoMirror
     resourceName=None
     resourceObject=/PlayerAugmentsUI/Materials/MI_UI_PlayerAugment_ActiveSlot_Backing
     tiling=NoTile
     imageSize=<48 48>
     margin=<ltrb 0 0 0 0>
     outlineSettings {
      bUseBrushTransparency=false
      roundingType=HalfHeightRadius
      width=0
      color {colorUseRule=UseColor_Specified, specifiedColor=<rgba 0 0 0 0>}
      cornerRadii=<0 0 0 0>
     }
     tintColor {colorUseRule=UseColor_Specified, specifiedColor=<rgba 1 1 1 1>}
     uVRegion {bIsValid=false, max=<0 0>, min=<0 0>}
    }
   }
   slot {class=OverlaySlot}
  }
  Image_EmptySlotPulse_Scale {
   widgetClassPath=&Image
   bIsVariable=true
   parent=Overlay_EmptySlot
   widget=Image_EmptySlotPulse_Scale
   properties {
    clipping=ClipToBoundsWithoutIntersecting
    renderOpacity=0.125
    visibility=SelfHitTestInvisible
    brush {
     bIsDynamicallyLoaded=false
     drawAs=Image
     imageType=NoImage
     mirroring=NoMirror
     resourceName=None
     resourceObject=/PlayerAugmentsUI/Materials/MI_UI_PlayerAugment_Border
     tiling=NoTile
     imageSize=<48 48>
     margin=<ltrb 0 0 0 0>
     outlineSettings {
      bUseBrushTransparency=false
      roundingType=HalfHeightRadius
      width=0
      color {colorUseRule=UseColor_Specified, specifiedColor=<rgba 0 0 0 0>}
      cornerRadii=<0 0 0 0>
     }
     tintColor {colorUseRule=UseColor_Specified, specifiedColor=<rgba 1 1 1 1>}
     uVRegion {bIsValid=false, max=<0 0>, min=<0 0>}
    }
   }
   slot {class=OverlaySlot}
  }
  Image_EmptySlotPulse {
   widgetClassPath=&Image
   bIsVariable=true
   parent=Overlay_EmptySlot
   widget=Image_EmptySlotPulse
   properties {
    clipping=ClipToBoundsWithoutIntersecting
    visibility=SelfHitTestInvisible
    brush {
     bIsDynamicallyLoaded=false
     drawAs=Image
     imageType=NoImage
     mirroring=NoMirror
     resourceName=None
     resourceObject=&MI_UI_PlayerAugment_Border
     tiling=NoTile
     imageSize=<48 48>
     margin=<ltrb 0 0 0 0>
     outlineSettings {
      bUseBrushTransparency=false
      roundingType=HalfHeightRadius
      width=0
      color {colorUseRule=UseColor_Specified, specifiedColor=<rgba 0 0 0 0>}
      cornerRadii=<0 0 0 0>
     }
     tintColor {colorUseRule=UseColor_Specified, specifiedColor=<rgba 1 1 1 1>}
     uVRegion {bIsValid=false, max=<0 0>, min=<0 0>}
    }
   }
   slot {class=OverlaySlot}
  }
  Image_EmptySlotIcon {
   widgetClassPath=&Image
   bIsVariable=true
   parent=Overlay_EmptySlot
   widget=Image_EmptySlotIcon
   properties {
    clipping=ClipToBoundsWithoutIntersecting
    visibility=SelfHitTestInvisible
    brush {
     bIsDynamicallyLoaded=false
     drawAs=Image
     imageType=NoImage
     mirroring=NoMirror
     resourceName=None
     resourceObject=&MI_UI_PlayerAugment_Border
     tiling=NoTile
     imageSize=<48 48>
     margin=<ltrb 0 0 0 0>
     outlineSettings {
      bUseBrushTransparency=false
      roundingType=HalfHeightRadius
      width=0
      color {colorUseRule=UseColor_Specified, specifiedColor=<rgba 0 0 0 0>}
      cornerRadii=<0 0 0 0>
     }
     tintColor {colorUseRule=UseColor_Specified, specifiedColor=<rgba 1 1 1 1>}
     uVRegion {bIsValid=false, max=<0 0>, min=<0 0>}
    }
   }
   slot {class=OverlaySlot}
  }
  Image_BorderGlow {
   widgetClassPath=&Image
   bIsVariable=true
   parent=Overlay_EmptySlot
   widget=Image_BorderGlow
   properties {
    clipping=ClipToBoundsWithoutIntersecting
    visibility=SelfHitTestInvisible
    brush {
     bIsDynamicallyLoaded=false
     drawAs=Box
     imageType=NoImage
     mirroring=NoMirror
     resourceName=None
     resourceObject=/PlayerAugmentsUI/Materials/MI_UI_PlayerAugment_Icon_BorderGlow
     tiling=NoTile
     imageSize=<64 64>
     margin=<ltrb 0.5 0.5 0.5 0.5>
     outlineSettings {
      bUseBrushTransparency=false
      roundingType=HalfHeightRadius
      width=0
      color {colorUseRule=UseColor_Specified, specifiedColor=<rgba 0 0 0 0>}
      cornerRadii=<0 0 0 0>
     }
     tintColor {colorUseRule=UseColor_Specified, specifiedColor=<rgba 1 1 1 1>}
     uVRegion {bIsValid=false, max=<0 0>, min=<0 0>}
    }
   }
   slot {class=OverlaySlot, properties {horizontalAlignment=HAlign_Fill, verticalAlignment=VAlign_Fill, padding=<ltrb -8 -8 -8 -8>}}
  }
  Image_Glow {
   widgetClassPath=&Image
   bIsVariable=true
   parent=Overlay_EmptySlot
   widget=Image_Glow
   properties {
    clipping=ClipToBoundsWithoutIntersecting
    visibility=SelfHitTestInvisible
    brush {
     bIsDynamicallyLoaded=false
     drawAs=Image
     imageType=NoImage
     mirroring=NoMirror
     resourceName=None
     resourceObject=/PlayerAugmentsUI/Materials/MI_UI_PlayerAugment_Timer_ReadyGlow
     tiling=NoTile
     imageSize=<32 32>
     margin=<ltrb 0 0 0 0>
     outlineSettings {
      bUseBrushTransparency=false
      roundingType=HalfHeightRadius
      width=0
      color {colorUseRule=UseColor_Specified, specifiedColor=<rgba 0 0 0 0>}
      cornerRadii=<0 0 0 0>
     }
     tintColor {colorUseRule=UseColor_Specified, specifiedColor=<rgba 1 1 1 1>}
     uVRegion {bIsValid=false, max=<0 0>, min=<0 0>}
    }
    colorAndOpacity=<rgba 1 1 1 0>
   }
   slot {class=OverlaySlot, properties {horizontalAlignment=HAlign_Fill, verticalAlignment=VAlign_Fill}}
  }
  Image_Arrow {
   widgetClassPath=&Image
   bIsVariable=true
   parent=Overlay_EmptySlot
   widget=Image_Arrow
   properties {
    clipping=ClipToBoundsWithoutIntersecting
    visibility=SelfHitTestInvisible
    brush {
     bIsDynamicallyLoaded=false
     drawAs=Image
     imageType=NoImage
     mirroring=NoMirror
     resourceName=None
     resourceObject=/PlayerAugmentsUI/Materials/MI_UI_AugmentArrow
     tiling=NoTile
     imageSize=<12 12>
     margin=<ltrb 0 0 0 0>
     outlineSettings {
      bUseBrushTransparency=false
      roundingType=HalfHeightRadius
      width=0
      color {colorUseRule=UseColor_Specified, specifiedColor=<rgba 0 0 0 0>}
      cornerRadii=<0 0 0 0>
     }
     tintColor {colorUseRule=UseColor_Specified, specifiedColor=<rgba 1 1 1 1>}
     uVRegion {bIsValid=false, max=<0 0>, min=<0 0>}
    }
    renderTransform {angle=0, scale=<1 1>, shear=<0 0>, translation=<0 -16>}
   }
   slot {class=OverlaySlot, properties {horizontalAlignment=HAlign_Center}}
  }
  Overlay_AugmentSlot {
   widgetClassPath=&Overlay
   bIsVariable=true
   parent=Switcher_Slot
   widget=Overlay_AugmentSlot
   slot {class=CommonVisibilitySwitcherSlot, properties {horizontalAlignment=HAlign_Center, verticalAlignment=VAlign_Center}}
  }
  Image_EmptySlotBacking_1 {
   widgetClassPath=&Image
   bIsVariable=true
   parent=Overlay_AugmentSlot
   widget=Image_EmptySlotBacking_1
   properties {
    clipping=ClipToBoundsWithoutIntersecting
    visibility=SelfHitTestInvisible
    brush {
     bIsDynamicallyLoaded=false
     drawAs=Image
     imageType=NoImage
     mirroring=NoMirror
     resourceName=None
     resourceObject=/PlayerAugmentsUI/Materials/MI_UI_PlayerAugment_ActiveSlot_Backing_Active
     tiling=NoTile
     imageSize=<48 48>
     margin=<ltrb 0 0 0 0>
     outlineSettings {
      bUseBrushTransparency=false
      roundingType=HalfHeightRadius
      width=0
      color {colorUseRule=UseColor_Specified, specifiedColor=<rgba 0 0 0 0>}
      cornerRadii=<0 0 0 0>
     }
     tintColor {colorUseRule=UseColor_Specified, specifiedColor=<rgba 1 1 1 1>}
     uVRegion {bIsValid=false, max=<0 0>, min=<0 0>}
    }
   }
   slot {class=OverlaySlot}
  }
  Image_EmptySlotIcon_1 {
   widgetClassPath=&Image
   bIsVariable=true
   parent=Overlay_AugmentSlot
   widget=Image_EmptySlotIcon_1
   properties {
    clipping=ClipToBoundsWithoutIntersecting
    visibility=SelfHitTestInvisible
    brush {
     bIsDynamicallyLoaded=false
     drawAs=Image
     imageType=NoImage
     mirroring=NoMirror
     resourceName=None
     resourceObject=&MI_UI_PlayerAugment_Border
     tiling=NoTile
     imageSize=<48 48>
     margin=<ltrb 0 0 0 0>
     outlineSettings {
      bUseBrushTransparency=false
      roundingType=HalfHeightRadius
      width=0
      color {colorUseRule=UseColor_Specified, specifiedColor=<rgba 0 0 0 0>}
      cornerRadii=<0 0 0 0>
     }
     tintColor {colorUseRule=UseColor_Specified, specifiedColor=<rgba 1 1 1 1>}
     uVRegion {bIsValid=false, max=<0 0>, min=<0 0>}
    }
   }
   slot {class=OverlaySlot}
  }
  Image_AugmentIcon {
   widgetClassPath=&Image
   bIsVariable=true
   parent=Overlay_AugmentSlot
   widget=Image_AugmentIcon
   properties {
    visibility=SelfHitTestInvisible
    brush {
     bIsDynamicallyLoaded=false
     drawAs=Image
     imageType=NoImage
     mirroring=NoMirror
     resourceName=None
     resourceObject=/PlayerAugmentsUI/Materials/MI_UI_PlayerAugment_HUD_Icon
     tiling=NoTile
     imageSize=<40 40>
     margin=<ltrb 0 0 0 0>
     outlineSettings {
      bUseBrushTransparency=false
      roundingType=HalfHeightRadius
      width=0
      color {colorUseRule=UseColor_Specified, specifiedColor=<rgba 0 0 0 0>}
      cornerRadii=<0 0 0 0>
     }
     tintColor {colorUseRule=UseColor_Specified, specifiedColor=<rgba 1 1 1 1>}
     uVRegion {bIsValid=false, max=<0 0>, min=<0 0>}
    }
   }
   slot {class=OverlaySlot, properties {horizontalAlignment=HAlign_Center, verticalAlignment=VAlign_Center}}
  }
  Overlay_NewVisuals {
   widgetClassPath=&Overlay
   bIsVariable=true
   parent=Overlay_AugmentSlot
   widget=Overlay_NewVisuals
   properties {visibility=Collapsed}
   slot {class=OverlaySlot, properties {horizontalAlignment=HAlign_Right, padding=<ltrb 0 -10 -10 0>}}
  }
  Image_NewIcon {
   widgetClassPath=&Image
   parent=Overlay_NewVisuals
   widget=Image_NewIcon
   properties {brush {
     bIsDynamicallyLoaded=false
     drawAs=Image
     imageType=NoImage
     mirroring=NoMirror
     resourceName=None
     resourceObject=/Game/Athena/UI/Frontend/Parties/Art/T_UI_PlusSign_64x
     tiling=NoTile
     imageSize=<32 32>
     margin=<ltrb 0 0 0 0>
     outlineSettings {
      bUseBrushTransparency=false
      roundingType=HalfHeightRadius
      width=0
      color {colorUseRule=UseColor_Specified, specifiedColor=<rgba 0 0 0 0>}
      cornerRadii=<0 0 0 0>
     }
     tintColor {colorUseRule=UseColor_Specified, specifiedColor=<rgba 0 1 0.6654 1>}
     uVRegion {bIsValid=false, max=<0 0>, min=<0 0>}
    }}
   slot {class=OverlaySlot}
  }
 ]
}
