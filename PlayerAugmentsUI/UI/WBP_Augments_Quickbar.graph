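# HUD quickbar widget for Fortnite player augments, showing a capped row of active augment slots with a glowing "+N" overflow counter.
# HandleAugmentsUpdated calls PopulateAugments, which rebuilds the row from WBP_Augments_QuickbarSlot widgets, then either pads empty slots, flagging the first as next, or shows the overflow count.
# AugmentSelectionStatusChanged switches between slide-and-fade show and hide animations.
# PreConstruct builds placeholder slots at design time; PopulateAugments retains a disconnected playlist-round-based loop.
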WidgetBlueprint /PlayerAugmentsUI/UI/WBP_Augments_Quickbar
animations [
 Anim_DisplayQuickbar {
  binding_count=1
  length_seconds=0.5333
  object=&self:Anim_DisplayQuickbar
  sequence {
   bindings [Overlay_AugmentQuickbar {
     id=#1
     possessed_class=Overlay
     tracks [
      {
       type=MovieScene2DTransformTrack
       sections [{
         type=MovieScene2DTransformSection
         end_frame=16
         start_frame=0
         channels [
          Translation.X {type=MovieSceneScriptingFloatChannel, const [0,1]}
          Translation.Y {type=MovieSceneScriptingFloatChannel, keys [[0,50],[10,-10],[13,5],[16,0]]}
          Angle {type=MovieSceneScriptingFloatChannel, const [0,1]}
          Scale.X {type=MovieSceneScriptingFloatChannel, const [1,1]}
          Scale.Y {type=MovieSceneScriptingFloatChannel, const [1,1]}
          Shear.X {type=MovieSceneScriptingFloatChannel, const [0,1]}
          Shear.Y {type=MovieSceneScriptingFloatChannel, const [0,1]}
         ]
        }]
      }
      {
       type=MovieSceneFloatTrack
       display_name=Render Opacity
       sections [{type=MovieSceneFloatSection, end_frame=16, start_frame=0, channels [None {type=MovieSceneScriptingFloatChannel, keys [[0,0],[10,1]]}]}]
      }
     ]
    }]
   display_rate {numerator=30}
  }
 }
 Anim_HideQuickbar {
  binding_count=1
  length_seconds=0.8304
  object=&self:Anim_HideQuickbar
  sequence {
   bindings [Overlay_AugmentQuickbar {
     id=#1
     possessed_class=Overlay
     tracks [
      {
       type=MovieScene2DTransformTrack
       sections [{
         type=MovieScene2DTransformSection
         end_frame=24
         start_frame=0
         channels [
          Translation.X {type=MovieSceneScriptingFloatChannel, const [0,1]}
          Translation.Y {type=MovieSceneScriptingFloatChannel, keys [[0,0],[24.913,50]]}
          Angle {type=MovieSceneScriptingFloatChannel, const [0,1]}
          Scale.X {type=MovieSceneScriptingFloatChannel, const [1,1]}
          Scale.Y {type=MovieSceneScriptingFloatChannel, const [1,1]}
          Shear.X {type=MovieSceneScriptingFloatChannel, const [0,1]}
          Shear.Y {type=MovieSceneScriptingFloatChannel, const [0,1]}
         ]
        }]
      }
      {
       type=MovieSceneFloatTrack
       display_name=Render Opacity
       sections [{type=MovieSceneFloatSection, end_frame=24, start_frame=0, channels [None {type=MovieSceneScriptingFloatChannel, keys [[0,1],[24.913,0]]}]}]
      }
     ]
    }]
   display_rate {numerator=30}
  }
 }
]
graphs [
 AugmentSelectionStatusChanged {nodes [
   n1 = AugmentSelectionStatusChanged!() -> (then, IsOpened:bool=~) {is=title K2Node_FunctionEntry}
   > n2 = Branch(Condition=@n1)
   > StopAnimation(InAnimation=Get Anim_HideQuickbar() -> (Anim_HideQuickbar:Widget Animation Object Reference))
   > PlayAnimation(InAnimation=Get Anim_DisplayQuickbar() -> (Anim_DisplayQuickbar:Widget Animation Object Reference))
   StopAnimation(execute=@n2.else, InAnimation=Get Anim_DisplayQuickbar())
   > PlayAnimation(InAnimation=Get Anim_HideQuickbar())
  ]}
 EventGraph {nodes [
   n1 = Event HandleAugmentsUpdated!() -> (OutputDelegate, then, ActiveAugments:Array of Fort Player Augment Item Definition Object References) {is=title K2Node_Event}
   > PopulateAugments!(execute, self, InActiveAugments:Array of Fort Player Augment Item Definition Object References=@n1.ActiveAugments) -> (then)
   Event Destruct()
   > Clear(TargetArray:Array of WBP Augments Quickbar Slot Object References=Get ActiveQuickbarSlots() -> (ActiveQuickbarSlots:Array of WBP Augments Quickbar Slot Object References))
   n2 = Event PreConstruct()
   > Branch(Condition=@n2.IsDesignTime)
   > ClearChildren(self=Get HBox_CurrentAugmentList() -> (HBox_CurrentAugmentList:Horizontal Box Object Reference))
   > n3 = For Loop(LastIndex=Get MaxAugmentIndex_ToShowInUI() -> (MaxAugmentIndex_ToShowInUI:int=0))
   > n4 = Create Widget!(
    execute:Exec
    Class:User Widget Class Reference=&/WBP_Augments_QuickbarSlot.WBP_Augments_QuickbarSlot_C
    OwningPlayer:Player Controller Object Reference
   ) -> (then:Exec, ReturnValue:WBP Augments Quickbar Slot Object Reference)
   > AddChildToHorizontalBox!(execute, self:Horizontal Box Object Reference=Get HBox_CurrentAugmentList(), Content:Widget Object Reference=@n4) -> (then, ReturnValue:Horizontal Box Slot Object Reference)
   > SetAsEmptySlot!(execute, self:WBP Augments Quickbar Slot Object Reference=@n4, IsNextAugment:bool="Equal (Integer)"(A=@n3)) -> (then)
  ]}
 PopulateAugments {nodes [
   n1 = PopulateAugments!() -> (then, InActiveAugments:Array of Fort Player Augment Item Definition Object References) {is=title K2Node_FunctionEntry}
   > ClearChildren(self=Get HBox_CurrentAugmentList())
   > n2 = For Each Loop with Break!(Exec:Exec, Array:Array of Fort Player Augment Item Definition Object References=@n1, Break:Exec=@n3.else) -> (LoopBody:Exec, Array Element:Fort Player Augment Item Definition Object Reference, Array Index:int, Completed:Exec)
   > n3 = Branch(Condition="integer <= integer"(A=@n2.Array Index, B=Get MaxAugmentIndex_ToShowInUI() -> (MaxAugmentIndex_ToShowInUI=0)))
   > n4 = Create Widget(Class=&WBP_Augments_QuickbarSlot_C)
   > AddChildToHorizontalBox(self=Get HBox_CurrentAugmentList(), Content=@n4)
   > SetAugmentInfo!(execute, self:WBP Augments Quickbar Slot Object Reference=@n4, InAugmentData:Fort Player Augment Item Definition Object Reference=@n2) -> (then)
   n5 = Get InActiveAugments() -> (InActiveAugments:Array of Fort Player Augment Item Definition Object References)
   n6 = Branch(
    execute=@n2.Completed
    Condition="integer > integer"(
     A=Last Index!(TargetArray:Array of Fort Player Augment Item Definition Object References=@n5) -> (ReturnValue=0)
     B=Get MaxAugmentIndex_ToShowInUI() -> (MaxAugmentIndex_ToShowInUI=0)
    )
   )
   n7 = int - int(A=Last Index(TargetArray=Get InActiveAugments()) -> (ReturnValue=0), B=Get MaxAugmentIndex_ToShowInUI() -> (MaxAugmentIndex_ToShowInUI=0))
   > "SetText (Text)"(
    self=Get Text_AdditionalAugments() -> (Text_AdditionalAugments:Common Text Object Reference)
    InText=Format Text!(Format="LOCTEXT(\"+{Num}\")", Num:Text="To Text (Integer)"(Value=@n7)) -> (Result)
   )
   > SetVisibility(self=Get Overlay_AdditionalAugments() -> (Overlay_AdditionalAugments:Overlay Object Reference), InVisibility=SelfHitTestInvisible)
   n8 = int + int(A=Last Index(TargetArray=Get InActiveAugments()) -> (ReturnValue=0))
   n9 = For Loop(execute=@n6.else, FirstIndex=@n8, LastIndex=Get MaxAugmentIndex_ToShowInUI() -> (MaxAugmentIndex_ToShowInUI=0))
   > n10 = Create Widget(Class=&WBP_Augments_QuickbarSlot_C)
   > AddChildToHorizontalBox(self=Get HBox_CurrentAugmentList(), Content=@n10)
   > SetAsEmptySlot(self=@n10, IsNextAugment="Equal (Integer)"(A=@n9, B=@n8))
   > SetVisibility(self=Get Overlay_AdditionalAugments(), InVisibility=Collapsed)
   n11 = For Each Loop with Break()
   n12 = Last Index(TargetArray=Get InActiveAugments()) -> (ReturnValue=0)
   n13 = Get NumAugmentGrantRoundsForPlaylist() -> (NumAugmentGrantRoundsForPlaylist:Scalable Float Structure)
   n14 = int - int(A=Truncate(A=GetValueAtLevel!(Input:"Scalable Float Structure (by ref)"=@n13, Level:float=0) -> (ReturnValue:float)))
   Branch(execute=@n11.Completed, Condition="integer >= integer"(A=@n14, B=@n12))
   > n15 = Increment Int(Value=@n12)
   > n16 = For Loop(FirstIndex=@n15, LastIndex=@n14)
   > n17 = Create Widget(Class=&WBP_Augments_QuickbarSlot_C)
   > AddChildToHorizontalBox(self=Get HBox_CurrentAugmentList(), Content=@n17)
   > SetAsEmptySlot(self=@n17, IsNextAugment="Equal (Integer)"(A=@n16, B=@n15))
   Branch(
    execute=@n11.LoopBody
    Condition="integer <= integer"(A=@n11.Array Index, B=Get MaxAugmentIndex_ToShowInUI() -> (MaxAugmentIndex_ToShowInUI=0))
   )
   > n18 = Create Widget(Class=&WBP_Augments_QuickbarSlot_C)
   > AddChildToHorizontalBox(self=Get HBox_CurrentAugmentList(), Content=@n18)
   > SetAugmentInfo(self=@n18, InAugmentData=@n11)
  ]}
]
variables [
 ActiveQuickbarSlots {type="array<WBP_Augments_QuickbarSlot_C>", default []}
 MaxAugmentIndex_ToShowInUI {type=integer, default=3}
]
widgets {
 info {
  inheritedWidgetCount=0
  parentClass=/Script/PlayerAugmentsUI.FortPlayerAugmentHUDQuickbarWidget
  rootWidgetClass=&Overlay
  widgetCount=9
 }
 widgets [
  Overlay_Root {widgetClassPath=&Overlay, widget=Overlay_Root}
  Backing_Glow_1 {
   widgetClassPath=&Image
   parent=Overlay_Root
   widget=Backing_Glow_1
   properties {
    visibility=SelfHitTestInvisible
    brush {
     bIsDynamicallyLoaded=false
     drawAs=Box
     imageType=NoImage
     mirroring=NoMirror
     resourceName=None
     resourceObject=/PlayerAugmentsUI/Materials/MI_UI_PlayerAugment_Font_Glow
     tiling=NoTile
     imageSize=<88 88>
     margin=<ltrb 0.5 0.5 0.5 0.5>
     outlineSettings {
      bUseBrushTransparency=false
      roundingType=HalfHeightRadius
      width=0
      color {colorUseRule=UseColor_Specified, specifiedColor=<rgba 0 0 0 0>}
      cornerRadii=<0 0 0 0>
     }
     tintColor {colorUseRule=UseColor_Specified, specifiedColor=<rgba 1 1 1 1>}
     uVRegion {bIsValid=false, max=<0 0>, min=<0 0>}
    }
   }
   slot {class=OverlaySlot, properties {horizontalAlignment=HAlign_Fill, verticalAlignment=VAlign_Fill, padding=<ltrb -32 -16 -32 -16>}}
  }
  HorizontalBox_71 {
   widgetClassPath=&HorizontalBox
   parent=Overlay_Root
   widget=HorizontalBox_71
   slot {class=OverlaySlot, properties {verticalAlignment=VAlign_Center}}
  }
  Overlay_AugmentQuickbar {
   widgetClassPath=&Overlay
   bIsVariable=true
   parent=HorizontalBox_71
   widget=Overlay_AugmentQuickbar
   properties {clipping=ClipToBounds}
   slot {class=HorizontalBoxSlot, properties {horizontalAlignment=HAlign_Right, verticalAlignment=VAlign_Bottom}}
  }
  HBox_CurrentAugmentList {
   widgetClassPath=&HorizontalBox
   bIsVariable=true
   parent=Overlay_AugmentQuickbar
   widget=HBox_CurrentAugmentList
   slot {class=OverlaySlot}
  }
  Spacer_177 {widgetClassPath=&Spacer, parent=HorizontalBox_71, widget=Spacer_177, properties {size=<10 1>}, slot {class=HorizontalBoxSlot}}
  Overlay_AdditionalAugments {
   widgetClassPath=&Overlay
   bIsVariable=true
   parent=HorizontalBox_71
   widget=Overlay_AdditionalAugments
   properties {visibility=Collapsed}
   slot {class=HorizontalBoxSlot, properties {verticalAlignment=VAlign_Center}}
  }
  Backing_Glow {
   widgetClassPath=&Image
   parent=Overlay_AdditionalAugments
   widget=Backing_Glow
   properties {
    visibility=SelfHitTestInvisible
    brush {
     bIsDynamicallyLoaded=false
     drawAs=Box
     imageType=NoImage
     mirroring=NoMirror
     resourceName=None
     resourceObject=&MI_UI_PlayerAugment_Font_Glow
     tiling=NoTile
     imageSize=<88 88>
     margin=<ltrb 0.5 0.5 0.5 0.5>
     outlineSettings {
      bUseBrushTransparency=false
      roundingType=HalfHeightRadius
      width=0
      color {colorUseRule=UseColor_Specified, specifiedColor=<rgba 0 0 0 0>}
      cornerRadii=<0 0 0 0>
     }
     tintColor {colorUseRule=UseColor_Specified, specifiedColor=<rgba 1 1 1 1>}
     uVRegion {bIsValid=false, max=<0 0>, min=<0 0>}
    }
   }
   slot {class=OverlaySlot, properties {horizontalAlignment=HAlign_Fill, verticalAlignment=VAlign_Fill, padding=<ltrb -32 -16 -32 -16>}}
  }
  Text_AdditionalAugments {
   widgetClassPath=&CommonTextBlock
   bIsVariable=true
   parent=Overlay_AdditionalAugments
   widget=Text_AdditionalAugments
   properties {
    mobileFontSizeMultiplier=1.4
    text="+##"
    textTransformPolicy=ToUpper
    font {
     bForceMonospaced=false
     bMaterialIsStencil=false
     fontMaterial=/PlayerAugmentsUI/Materials/MI_UI_PlayerAugment_RerollText
     fontObject=&BurbankSmall
     letterSpacing=0
     monospacedWidth=1
     size=20
     skewAmount=0
     typefaceFontName=Black
     outlineSettings {
      bApplyOutlineToDropShadows=false
      bMiteredCorners=false
      bSeparateFillAlpha=true
      outlineMaterial=&M-FontShadow
      outlineSize=2
      outlineColor=<rgba 0.6667 0 1 0.75>
     }
    }
    shadowColorAndOpacity=<rgba 0 0.8333 1 0.125>
    shadowOffset=<4 2>
   }
   slot {class=OverlaySlot, properties {horizontalAlignment=HAlign_Center, verticalAlignment=VAlign_Center}}
  }
 ]
}
