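# Starter-augment HUD widget derived from FortPlayerAugmentInGameCarriedWidget, laying out a placeholder "Active" label, augment icon and augment name in one row.
# The root overlay starts collapsed, and nothing in this Blueprint reveals it.
# SetAugmentInfo holds nodes to set the icon from a soft texture and set the name text, but shows no execution links or supplied values.
# The event graph is empty.
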WidgetBlueprint /PlayerAugmentsUI/UI/WBP_Augments_StarterAugment
graphs [
 EventGraph {links [], nodes []}
 SetAugmentInfo {nodes [
   SetBrushFromSoftTexture!(
    execute
    self:Image Object Reference=Get Image_AugmentIcon() -> (Image_AugmentIcon:Image Object Reference)
    SoftTexture:Texture 2D Soft Object Reference
    bMatchSize:bool=false
   ) -> (then)
   "SetText (Text)"(self=Get Text_AugmentName() -> (Text_AugmentName:Common Text Object Reference))
   SetAugmentInfo() {is=title K2Node_FunctionEntry}
  ]}
]
variables []
widgets {
 info {
  inheritedWidgetCount=1
  parentClass=/Script/PlayerAugmentsUI.FortPlayerAugmentInGameCarriedWidget
  rootWidgetClass=&Overlay
  widgetCount=6
 }
 widgets [
  Overlay_Root {widgetClassPath=&Overlay, widget=Overlay_Root, properties {visibility=Collapsed}}
  HorizontalBox_121 {
   widgetClassPath=&HorizontalBox
   parent=Overlay_Root
   widget=HorizontalBox_121
   properties {flowDirectionPreference=Culture}
   slot {class=OverlaySlot, properties {horizontalAlignment=HAlign_Fill, verticalAlignment=VAlign_Fill}}
  }
  Text_Active {
   widgetClassPath=&CommonTextBlock
   parent=HorizontalBox_121
   widget=Text_Active
   properties {
    text="[PH] Active: "
    textTransformPolicy=ToUpper
    colorAndOpacity {colorUseRule=UseColor_Specified, specifiedColor=<rgba 0.7835 0.7835 0.7835 1>}
    font {
     bForceMonospaced=false
     bMaterialIsStencil=false
     fontMaterial=null
     fontObject=&BurbankSmall
     letterSpacing=0
     monospacedWidth=1
     size=16
     skewAmount=0.2
     typefaceFontName=Bold
     outlineSettings {
      bApplyOutlineToDropShadows=false
      bMiteredCorners=false
      bSeparateFillAlpha=true
      outlineMaterial=&M-FontShadow
      outlineSize=0
      outlineColor=<rgba 0 0 0 0.5>
     }
    }
    shadowOffset=<0 0>
   }
   slot {class=HorizontalBoxSlot, properties {horizontalAlignment=HAlign_Left, verticalAlignment=VAlign_Center}}
  }
  Image_AugmentIcon {
   widgetClassPath=&Image
   bIsVariable=true
   parent=HorizontalBox_121
   widget=Image_AugmentIcon
   properties {
    visibility=SelfHitTestInvisible
    brush {
     bIsDynamicallyLoaded=false
     drawAs=Image
     imageType=NoImage
     mirroring=NoMirror
     resourceName=None
     resourceObject=/Game/UI/Icons/Icon-RandomDice/Icon-RandomDice-128
     tiling=NoTile
     imageSize=<32 32>
     margin=<ltrb 0 0 0 0>
     outlineSettings {
      bUseBrushTransparency=false
      roundingType=HalfHeightRadius
      width=0
      color {colorUseRule=UseColor_Specified, specifiedColor=<rgba 0 0 0 0>}
      cornerRadii=<0 0 0 0>
     }
     tintColor {colorUseRule=UseColor_Specified, specifiedColor=<rgba 1 1 1 1>}
     uVRegion {bIsValid=false, max=<0 0>, min=<0 0>}
    }
   }
   slot {class=HorizontalBoxSlot, properties {horizontalAlignment=HAlign_Center, verticalAlignment=VAlign_Center, padding=<ltrb 5 0 5 0>}}
  }
  Text_AugmentName {
   widgetClassPath=&CommonTextBlock
   bIsVariable=true
   parent=HorizontalBox_121
   widget=Text_AugmentName
   properties {
    text="[PH] Augment Name"
    textTransformPolicy=ToUpper
    font {
     bForceMonospaced=false
     bMaterialIsStencil=false
     fontMaterial=null
     fontObject=&BurbankSmall
     letterSpacing=0
     monospacedWidth=1
     size=18
     skewAmount=0.2
     typefaceFontName=Bold
     outlineSettings {
      bApplyOutlineToDropShadows=false
      bMiteredCorners=false
      bSeparateFillAlpha=true
      outlineMaterial=&M-FontShadow
      outlineSize=0
      outlineColor=<rgba 0 0 0 0.5>
     }
    }
    shadowOffset=<0 0>
   }
   slot {class=HorizontalBoxSlot, properties {horizontalAlignment=HAlign_Left, verticalAlignment=VAlign_Center}}
  }
  TutorialHighlight {widgetClassPath=&TutorialHighlight}
 ]
}
