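# Sound cue for the Player Augments UI, named as a reminder that an augment is ready.
# A single wave player feeds the reminder wave straight to the output, with boosted volume and no randomization or modulation.
# It uses the non-critical in-game UI sound class with virtualization disabled.
# A concurrency override caps simultaneous copies at two, stopping the oldest first.
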SoundCue /PlayerAugmentsCode/Sound/PlayerAugments_UI_Ready_Reminder_Cue
root=WavePlayer_0
nodes [{id=WavePlayer_0, properties {soundWaveAssetPtr=&/PlayerAugments_UI_Reminder_01}}]
properties {
 bOverrideConcurrency=true
 maxDistance=2097152
 soundClassObject=&Fort_UI_InGame_NonCritical
 virtualizationMode=Disabled
 volumeMultiplier=2
 concurrencyOverrides {maxCount=2, resolutionRule=StopOldest}
}
